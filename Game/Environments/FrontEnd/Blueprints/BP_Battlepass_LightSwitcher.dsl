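; BP_Battlepass_LightSwitcher turns a front-end scene's lighting set on or off, covering lights, height fog and particles.
; Activation picks high-detail or low-detail lights and fog by mobile or debug flag and sets MPC_Cozy's isLowDetail.
; BP_DetailLevelMesh visibility follows shadow quality, platform and per-mesh flags.
; BeginPlay hides inactive sets, reactivates active backdrop cameras against a race, then binds settings updates; the construction script previews lighting without setting isActive.
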(event Custom|UpdateSettings
  (bind _directionlight_high_channel_2 (Variables|HighDetailLights|GetDirectionLight_HIgh_Channel_2))
  (bind _directionlight_high_channel_2_2 (Variables|HighDetailLights|GetDirectionLight_HIgh_Channel_2))
  (bind _directionallight_lowdetailmode_channel2 (Variables|MobileLightSetup|GetDirectionalLight_LowDetailMode_Channel2))
  (bind _directionallight_lowdetailmode_channel2_2 (Variables|MobileLightSetup|GetDirectionalLight_LowDetailMode_Channel2))
  (bind _directionallight_lowdetailmode_channel1 (Variables|MobileLightSetup|GetDirectionalLight_LowDetailMode_Channel1))
  (bind _directionallight_lowdetailmode_channel1_2 (Variables|MobileLightSetup|GetDirectionalLight_LowDetailMode_Channel1))
  ;[1 1 1] On Setting change Event
  (if (Variables|Default|GetisActive)
    (Utilities|FlowControl|Sequence
      (:then_0
        (if (TouchPlatform|IsMobileGame)
          (Rendering|SetVisibility (Variables|/Script/Engine.Light|GetLightComponent (Variables|HighDetailLights|GetDirectionLight)) false false)
          (Utilities|IsValid _directionlight_high_channel_2_2
            (:Is%20Valid
              (Rendering|SetVisibility (Variables|/Script/Engine.Light|GetLightComponent _directionlight_high_channel_2_2) false false))
            (:Is%20Not%20Valid
              _))
          (Rendering|SetVisibility (Variables|/Script/Engine.Light|GetLightComponent (Variables|MobileLightSetup|GetDirectionalLight_LowDetailMode)) true false)
          (Rendering|Material|SetScalarParameterValue "/Game/Environments/FrontEnd/Materials/Cozy/MPC_Cozy.MPC_Cozy" "isLowDetail" 1.0 :!class /Script/Engine.KismetMaterialLibrary)
          (Development|PrintString "LowDetail enabled")
          (Utilities|IsValid _directionallight_lowdetailmode_channel1
            (:Is%20Valid
              (Rendering|SetVisibility (Variables|/Script/Engine.Light|GetLightComponent _directionallight_lowdetailmode_channel1) true false))
            (:Is%20Not%20Valid
              _))
          (Utilities|IsValid _directionallight_lowdetailmode_channel2_2
            (:Is%20Valid
              (Rendering|SetVisibility (Variables|/Script/Engine.Light|GetLightComponent _directionallight_lowdetailmode_channel2_2) true false)))
          (else
            (Rendering|SetVisibility (Variables|/Script/Engine.Light|GetLightComponent (Variables|HighDetailLights|GetDirectionLight)) true false)
            (Utilities|IsValid _directionlight_high_channel_2
              (:Is%20Valid
                (Rendering|SetVisibility (Variables|/Script/Engine.Light|GetLightComponent _directionlight_high_channel_2) true false))
              (:Is%20Not%20Valid
                _))
            (Rendering|SetVisibility (Variables|/Script/Engine.Light|GetLightComponent (Variables|MobileLightSetup|GetDirectionalLight_LowDetailMode)) false false)
            (Rendering|Material|SetScalarParameterValue "/Game/Environments/FrontEnd/Materials/Cozy/MPC_Cozy.MPC_Cozy" "isLowDetail" :!class /Script/Engine.KismetMaterialLibrary)
            (Development|PrintString "HiehgDetail Enabled")
            (Utilities|IsValid _directionallight_lowdetailmode_channel1_2
              (:Is%20Valid
                (Rendering|SetVisibility (Variables|/Script/Engine.Light|GetLightComponent _directionallight_lowdetailmode_channel1_2) false false))
              (:Is%20Not%20Valid
                _))
            (Utilities|IsValid _directionallight_lowdetailmode_channel2
              (:Is%20Valid
                (Rendering|SetVisibility (Variables|/Script/Engine.Light|GetLightComponent _directionallight_lowdetailmode_channel2) false false))))))
      (:then_1
        ;;[1 1 1] If Shadow quality is Medium or lower the Volumetric fog is disabled and it shows the simple fog cards
        (CallFunction|SetVisiblityofDetailSpecificMeshes)))))

(event EventBeginPlay
  (Variables|MobileLightSetup|SetDebugLOWQualityLighting false)
  (if (not (Variables|Default|GetisActive))
    (CallFunction|SetVisiblityofSceneLighting false))
  ;[1 1 1] Fix for sometimes no lighting if race condition enables lights before begin play was triggered
  (bind _outactors (Actor|GetAllActorsOfClass "/Game/Environments/FrontEnd/Blueprints/FortnitePartyBackdrop_Camera.FortnitePartyBackdrop_Camera_C"))
  (bind _array_element (Utilities|Array|ForEachLoopwithBreak _outactors)
    (:LoopBody
      (if (Variables|/Game/Environments/FrontEnd/Blueprints/FortnitePartyBackdrop_Camera.FortnitePartyBackdrop_Camera_C|GetActive _array_element)
        (Class|FortnitePartyBackdropCamera|ManualActivate _array_element)
        (Development|PrintString (Utilities|String|ToString(Boolean) (Variables|/Game/Environments/FrontEnd/Blueprints/FortnitePartyBackdrop_Camera.FortnitePartyBackdrop_Camera_C|GetActive _array_element)))
        (bind _returnvalue (Settings|GetGameUserSettings))
        (EventDispatchers|BindEventtoOnGameUserSettingsUINeedsUpdate _returnvalue (delegate Custom|UpdateSettings))
        ;[1 1 1] If Shadow quality is Medium or lower the Volumetric fog is disabled and it shows the simple fog cards
        (CallFunction|SetVisiblityofDetailSpecificMeshes)))))

(event EventEnableLevelSetLighting
  ;[1 1 1] JON this need to be triggered on entering a new scene and exiting the scene
  (CallFunction|SetVisiblityofSceneLighting true)
  (Development|PrintString "enabled"))

(event Custom|EnableLobbySetLighting
  ;[1 1 1] JON this need to be triggered on entering a new scene and exiting the scene
  (CallFunction|SetVisiblityofSceneLighting true)
  (Development|PrintString "enabled"))

(event Custom|DisableLobbySetLighting
  ;[1 1 1] JON this need to be triggered on entering a new scene and exiting the scene
  (CallFunction|SetVisiblityofSceneLighting false))

(event EventDisableLevelSetLighting
  ;[1 1 1] JON this need to be triggered on entering a new scene and exiting the scene
  (CallFunction|SetVisiblityofSceneLighting false))

(fn SetVisiblityofDetailSpecificMeshes
  (bind _returnvalue (Settings|GetGameUserSettings))
  (bind _outactors (Actor|GetAllActorsOfClass "/Game/Environments/FrontEnd/Blueprints/BP_DetailLevelMesh.BP_DetailLevelMesh_C"))
  (for _array_element _outactors
    (bind _asbp_detail_level_mesh (Utilities|Casting|CastToBP_DetailLevelMesh _array_element))
    (if (and (Utilities|String|EqualExactly(String) (Game|GetPlatformName) "Switch") (Variables|/Game/Environments/FrontEnd/Blueprints/BP_DetailLevelMesh.BP_DetailLevelMesh_C|GetNotVisibleOnSwitch _asbp_detail_level_mesh))
      (Rendering|SetVisibility (Variables|/Game/Environments/FrontEnd/Blueprints/BP_DetailLevelMesh.BP_DetailLevelMesh_C|GetStaticMesh _array_element) false false)
      (elif (Variables|/Game/Environments/FrontEnd/Blueprints/BP_DetailLevelMesh.BP_DetailLevelMesh_C|GetAlwaysVisible _asbp_detail_level_mesh)
        (Rendering|SetVisibility (Variables|/Game/Environments/FrontEnd/Blueprints/BP_DetailLevelMesh.BP_DetailLevelMesh_C|GetStaticMesh _array_element) true false)
        (else
          (Rendering|SetVisibility (Variables|/Game/Environments/FrontEnd/Blueprints/BP_DetailLevelMesh.BP_DetailLevelMesh_C|GetStaticMesh _array_element) (or (or (<= (Settings|GetShadowQuality _returnvalue) 3) (TouchPlatform|IsMobileGame)) (Variables|MobileLightSetup|GetDebugLOWQualityLighting)) false))))))

(fn SetVisiblityofSceneLighting
  (params (Activate :type Boolean) (ConstructionScript :type Boolean))
  (bind _skylight_lowdetailmode (Variables|MobileLightSetup|GetSkyLight_LowDetailMode))
  (bind _directionlight_high_channel_2 (Variables|HighDetailLights|GetDirectionLight_HIgh_Channel_2))
  (bind _skylight (Variables|HighDetailLights|GetSkyLight))
  (bind _skylight_2 (Variables|HighDetailLights|GetSkyLight))
  (bind _skylight_3 (Variables|HighDetailLights|GetSkyLight))
  (bind _skylight_lowdetailmode_2 (Variables|MobileLightSetup|GetSkyLight_LowDetailMode))
  (bind _exponentialheightfog_lowdetailmode (Variables|MobileLightSetup|GetExponentialHeightfog_LowDetailMode))
  (bind _directionlight (Variables|HighDetailLights|GetDirectionLight))
  (bind _skylight_lowdetailmode_3 (Variables|MobileLightSetup|GetSkyLight_LowDetailMode))
  (bind _exponentialheightfog (Variables|HighDetailLights|GetExponentialHeightfog))
  (bind _exponentialheightfog_2 (Variables|HighDetailLights|GetExponentialHeightfog))
  (bind _directionallight_lowdetailmode_channel2 (Variables|MobileLightSetup|GetDirectionalLight_LowDetailMode_Channel2))
  (bind _exponentialheightfog_lowdetailmode_2 (Variables|MobileLightSetup|GetExponentialHeightfog_LowDetailMode))
  (bind _exponentialheightfog_3 (Variables|HighDetailLights|GetExponentialHeightfog))
  (bind _exponentialheightfog_lowdetailmode_3 (Variables|MobileLightSetup|GetExponentialHeightfog_LowDetailMode))
  (bind _directionallight_lowdetailmode_channel2_2 (Variables|MobileLightSetup|GetDirectionalLight_LowDetailMode_Channel2))
  (bind _directionlight_2 (Variables|HighDetailLights|GetDirectionLight))
  (bind _directionallight_lowdetailmode (Variables|MobileLightSetup|GetDirectionalLight_LowDetailMode))
  (bind _directionallight_lowdetailmode_2 (Variables|MobileLightSetup|GetDirectionalLight_LowDetailMode))
  (bind _directionallight_lowdetailmode_3 (Variables|MobileLightSetup|GetDirectionalLight_LowDetailMode))
  (bind _directionallight_lowdetailmode_channel1 (Variables|MobileLightSetup|GetDirectionalLight_LowDetailMode_Channel1))
  (bind _directionallight_lowdetailmode_channel1_2 (Variables|MobileLightSetup|GetDirectionalLight_LowDetailMode_Channel1))
  (bind _directionallight_lowdetailmode_channel1_3 (Variables|MobileLightSetup|GetDirectionalLight_LowDetailMode_Channel1))
  (bind _directionlight_3 (Variables|HighDetailLights|GetDirectionLight))
  (bind _directionlight_high_channel_2_2 (Variables|HighDetailLights|GetDirectionLight_HIgh_Channel_2))
  (bind _directionlight_high_channel_2_3 (Variables|HighDetailLights|GetDirectionLight_HIgh_Channel_2))
  (bind _directionallight_lowdetailmode_channel2_3 (Variables|MobileLightSetup|GetDirectionalLight_LowDetailMode_Channel2))
  (Utilities|FlowControl|Sequence
    (:then_0
      (if Activate
        ;[1 1 1] Check for low detail mode and enable alternative directional light
        (if (or (and (TouchPlatform|IsMobileGame) (Utilities|IsValid (Variables|MobileLightSetup|GetDirectionalLight_LowDetailMode) :!class /Script/Engine.KismetSystemLibrary)) (Variables|MobileLightSetup|GetDebugLOWQualityLighting))
          (Utilities|FlowControl|Sequence
            (:then_0
              (Utilities|IsValid _directionlight_3
                (:Is%20Valid
                  (Rendering|SetVisibility (Variables|/Script/Engine.Light|GetLightComponent _directionlight_3) false false))
                (:Is%20Not%20Valid
                  _))
              (Utilities|IsValid _directionlight_high_channel_2_3
                (:Is%20Valid
                  (Rendering|SetVisibility (Variables|/Script/Engine.Light|GetLightComponent _directionlight_high_channel_2_3) false false))))
            (:then_1
              (Utilities|IsValid _directionallight_lowdetailmode_2
                (:Is%20Valid
                  (Rendering|SetVisibility (Variables|/Script/Engine.Light|GetLightComponent _directionallight_lowdetailmode_2) true false))))
            (:then_2
              (Utilities|IsValid _directionallight_lowdetailmode_channel1_2
                (:Is%20Valid
                  (Rendering|SetVisibility (Variables|/Script/Engine.Light|GetLightComponent _directionallight_lowdetailmode_channel1_2) true false))))
            (:then_3
              (Utilities|IsValid _directionallight_lowdetailmode_channel2_3
                (:Is%20Valid
                  (Rendering|SetVisibility (Variables|/Script/Engine.Light|GetLightComponent _directionallight_lowdetailmode_channel2_3) true false))))
            (:then_4
              (Rendering|Material|SetScalarParameterValue "/Game/Environments/FrontEnd/Materials/Cozy/MPC_Cozy.MPC_Cozy" "isLowDetail" 1.0 :!class /Script/Engine.KismetMaterialLibrary)
              (Development|PrintString "LowDetail enabled")))
          (else
            (Utilities|FlowControl|Sequence
              (:then_0
                (Utilities|IsValid _directionallight_lowdetailmode_3
                  (:Is%20Valid
                    (Rendering|SetVisibility (Variables|/Script/Engine.Light|GetLightComponent _directionallight_lowdetailmode_3) false false))))
              (:then_1
                (Utilities|IsValid _directionallight_lowdetailmode_channel1_3
                  (:Is%20Valid
                    (Rendering|SetVisibility (Variables|/Script/Engine.Light|GetLightComponent _directionallight_lowdetailmode_channel1_3) false false))))
              (:then_2
                (Utilities|IsValid _directionallight_lowdetailmode_channel2_2
                  (:Is%20Valid
                    (Rendering|SetVisibility (Variables|/Script/Engine.Light|GetLightComponent _directionallight_lowdetailmode_channel2_2) false false))))
              (:then_3
                (Utilities|IsValid _directionlight
                  (:Is%20Valid
                    (Rendering|SetVisibility (Variables|/Script/Engine.Light|GetLightComponent _directionlight) true false))
                  (:Is%20Not%20Valid
                    _))
                (Utilities|IsValid _directionlight_high_channel_2
                  (:Is%20Valid
                    (Rendering|SetVisibility (Variables|/Script/Engine.Light|GetLightComponent _directionlight_high_channel_2) true false))))
              (:then_4
                (Rendering|Material|SetScalarParameterValue "/Game/Environments/FrontEnd/Materials/Cozy/MPC_Cozy.MPC_Cozy" "isLowDetail" :!class /Script/Engine.KismetMaterialLibrary)
                (Development|PrintString "HihgDetail Enabled")))))
        (else
          (Utilities|FlowControl|Sequence
            (:then_0
              (Utilities|IsValid _directionlight_2
                (:Is%20Valid
                  (Rendering|SetVisibility (Variables|/Script/Engine.Light|GetLightComponent _directionlight_2) Activate false))
                (:Is%20Not%20Valid
                  _))
              (Utilities|IsValid _directionlight_high_channel_2_2
                (:Is%20Valid
                  (Rendering|SetVisibility (Variables|/Script/Engine.Light|GetLightComponent _directionlight_high_channel_2_2) Activate false))))
            (:then_1
              (Utilities|IsValid _directionallight_lowdetailmode
                (:Is%20Valid
                  (Rendering|SetVisibility (Variables|/Script/Engine.Light|GetLightComponent _directionallight_lowdetailmode) Activate false))))
            (:then_2
              (Utilities|IsValid _directionallight_lowdetailmode_channel1
                (:Is%20Valid
                  (Rendering|SetVisibility (Variables|/Script/Engine.Light|GetLightComponent _directionallight_lowdetailmode_channel1) Activate false))))
            (:then_3
              (Utilities|IsValid _directionallight_lowdetailmode_channel2
                (:Is%20Valid
                  (Rendering|SetVisibility (Variables|/Script/Engine.Light|GetLightComponent _directionallight_lowdetailmode_channel2) Activate false))))))))
    (:then_1
      (if Activate
        (if (or (and (TouchPlatform|IsMobileGame) (Utilities|IsValid (Variables|MobileLightSetup|GetDirectionalLight_LowDetailMode) :!class /Script/Engine.KismetSystemLibrary)) (Variables|MobileLightSetup|GetDebugLOWQualityLighting))
          (Utilities|FlowControl|Sequence
            (:then_0
              (Utilities|IsValid _skylight_2
                (:Is%20Valid
                  (Rendering|SetVisibility (Variables|/Script/Engine.SkyLight|GetLightComponent _skylight_2) false false))))
            (:then_1
              (Utilities|IsValid _skylight_lowdetailmode_3
                (:Is%20Valid
                  (Rendering|SetVisibility (Variables|/Script/Engine.SkyLight|GetLightComponent _skylight_lowdetailmode_3) true false)))))
          (else
            (Utilities|FlowControl|Sequence
              (:then_0
                (Utilities|IsValid _skylight_lowdetailmode_2
                  (:Is%20Valid
                    (Rendering|SetVisibility (Variables|/Script/Engine.SkyLight|GetLightComponent _skylight_lowdetailmode_2) false false))))
              (:then_1
                (Utilities|IsValid _skylight_3
                  (:Is%20Valid
                    (Rendering|SetVisibility (Variables|/Script/Engine.SkyLight|GetLightComponent _skylight_3) true false)))))))
        (else
          (Utilities|FlowControl|Sequence
            (:then_0
              (Utilities|IsValid _skylight
                (:Is%20Valid
                  (Rendering|SetVisibility (Variables|/Script/Engine.SkyLight|GetLightComponent _skylight) Activate false))))
            (:then_1
              (Utilities|IsValid _skylight_lowdetailmode
                (:Is%20Valid
                  (Rendering|SetVisibility (Variables|/Script/Engine.SkyLight|GetLightComponent _skylight_lowdetailmode) Activate false))))))))
    (:then_2
      (for _array_element (Variables|HighDetailLights|GetPointlights)
        (Utilities|IsValid _array_element
          (:Is%20Valid
            (Rendering|SetVisibility (Variables|/Script/Engine.Light|GetLightComponent _array_element) Activate false)))))
    (:then_3
      (if Activate
        (if (or (and (TouchPlatform|IsMobileGame) (Utilities|IsValid (Variables|MobileLightSetup|GetDirectionalLight_LowDetailMode) :!class /Script/Engine.KismetSystemLibrary)) (Variables|MobileLightSetup|GetDebugLOWQualityLighting))
          (Utilities|FlowControl|Sequence
            (:then_0
              (Utilities|IsValid _exponentialheightfog_3
                (:Is%20Valid
                  (Rendering|SetVisibility (Variables|/Script/Engine.ExponentialHeightFog|GetComponent _exponentialheightfog_3) false false))))
            (:then_1
              (Utilities|IsValid _exponentialheightfog_lowdetailmode_3
                (:Is%20Valid
                  (Rendering|SetVisibility (Variables|/Script/Engine.ExponentialHeightFog|GetComponent _exponentialheightfog_lowdetailmode_3) true false)))))
          (else
            (Utilities|FlowControl|Sequence
              (:then_0
                (Utilities|IsValid _exponentialheightfog_lowdetailmode
                  (:Is%20Valid
                    (Rendering|SetVisibility (Variables|/Script/Engine.ExponentialHeightFog|GetComponent _exponentialheightfog_lowdetailmode) false false))))
              (:then_1
                (Utilities|IsValid _exponentialheightfog
                  (:Is%20Valid
                    (Rendering|SetVisibility (Variables|/Script/Engine.ExponentialHeightFog|GetComponent _exponentialheightfog) true false)))))))
        (else
          (Utilities|FlowControl|Sequence
            (:then_0
              (Utilities|IsValid _exponentialheightfog_2
                (:Is%20Valid
                  (Rendering|SetVisibility (Variables|/Script/Engine.ExponentialHeightFog|GetComponent _exponentialheightfog_2) Activate false))))
            (:then_1
              (Utilities|IsValid _exponentialheightfog_lowdetailmode_2
                (:Is%20Valid
                  (Rendering|SetVisibility (Variables|/Script/Engine.ExponentialHeightFog|GetComponent _exponentialheightfog_lowdetailmode_2) false false))))))))
    (:then_4
      (for _array_element_2 (Variables|AdvancedSetup|GetParticleSystems)
        (Utilities|IsValid _array_element_2
          (:Is%20Valid
            (if Activate
              (Components|Activation|Activate (Variables|/Script/CascadeRuntime.Emitter|GetParticleSystemComponent _array_element_2) false)
              (else
                (Components|Activation|Deactivate (Variables|/Script/CascadeRuntime.Emitter|GetParticleSystemComponent _array_element_2))))))))
    (:then_5
      ;[1 0 0.7184] Dont let the construction scrip SET active on the lights.... it can turn them on for edtior useage but we need to only set active state AFTER Beginplay
      (if ConstructionScript
        _
        (elif Activate
          (Variables|Default|SetisActive true)
          (else
            (Variables|Default|SetisActive false)))))))

(fn ConstructionScript
  (CallFunction|SetVisiblityofSceneLighting true true)
  (CallFunction|SetVisiblityofDetailSpecificMeshes))
