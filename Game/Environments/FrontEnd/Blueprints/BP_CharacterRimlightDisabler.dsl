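; Front-end environment Blueprint that switches off character rim lighting while it is in play.
; On BeginPlay it sets the bEnableCharacterRimLight scalar in the global FortniteMaterialParameters collection to its default of zero, and on EndPlay it sets it back to one.
; Both writes run only where cosmetic events can execute, and the restore does not preserve any earlier value.
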(event EventBeginPlay
  (Networking|CanExecuteCosmeticEvents
    (:True
      (Rendering|Material|SetScalarParameterValue "/Game/Packages/Fortress_SharedMaterials/GlobalMaterialParameters/FortniteMaterialParameters.FortniteMaterialParameters" "bEnableCharacterRimLight" :!class /Script/Engine.KismetMaterialLibrary))))

(event EventEndPlay
  (params (EndPlayReason :type EEndPlayReason%20Enum :default "Destroyed"))
  (Networking|CanExecuteCosmeticEvents
    (:True
      (Rendering|Material|SetScalarParameterValue "/Game/Packages/Fortress_SharedMaterials/GlobalMaterialParameters/FortniteMaterialParameters.FortniteMaterialParameters" "bEnableCharacterRimLight" 1.0 :!class /Script/Engine.KismetMaterialLibrary))))

(fn ConstructionScript)
