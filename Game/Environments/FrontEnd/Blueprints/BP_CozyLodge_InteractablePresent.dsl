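; Interactable Cozy Lodge front-end present that the player can shake, open and reset.
; Shakes stack intensity into timeline-damped sine rotation and material wobble, play the selected present's sound and duck menu music.
; Opening drives a bulge timeline, reversed faster on cancel, then on completion locks and plays fold-open with paper-tinted FX.
; Reset rebinds the dynamic material and clears locks; display-name string matches trigger special Doom audio and Galileo unwrap FX.
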(event EventBeginPlay)

(event Collision|EventActorBeginOverlap
  (params (OtherActor :type Actor%20Object%20Reference)))

(event EventOnShakePresent
  (params (ItemDef :type Fort%20Item%20Definition%20Object%20Reference))
  (bind _returnvalue (Game|GetGameState))
  (bind _returnvalue_2 (Actor|GetComponentsByClass _returnvalue "/Game/Blueprints/Components/GSC_DynamicLevelStreaming.GSC_DynamicLevelStreaming_C"))
  (bind _intensityshake (Variables|Default|GetIntensityShake))
  ;[1 1 1] Shake
  (if (not (Variables|Default|GetBlockRotation))
    (Variables|Default|SetIntensityShake (Math|Float|Clamp(Float) (+ (Variables|Default|GetIntensityShake) 1.0) 0.0 (Variables|Default|GetMaxShakeIntensity)))
    (Development|PrintString (Utilities|String|ToString(Text) (Variables|/Script/ItemizationCoreRuntime.ItemDefinitionBase|GetItemName ItemDef)))
    (bind _shake_audio_component (Variables|Default|GetShake%20Audio%20Component)
      (:then
        (Audio|Components|Audio|FadeOut _shake_audio_component 1.0 0.0 "Linear" :!class /Script/Engine.AudioComponent))
      (:else
        _))
    (if (> (Utilities|Array|Length _returnvalue_2) 0)
      (bind _asbp_cozy_lodge_selectable_present (Utilities|Casting|CastToBP_CozyLodge_SelectablePresent (Class|FortGameStateComponentEventLevel|GetLastNavObject (Utilities|Array|Get(acopy) _returnvalue_2)))
        (:then
          (Variables|Default|SetShakeSound (Variables|/Game/Environments/FrontEnd/Blueprints/BP_CozyLodge_SelectablePresent.BP_CozyLodge_SelectablePresent_C|GetShakeSound _asbp_cozy_lodge_selectable_present)))
        (:CastFailed
          _)))
    (bind _returnvalue_3 (Audio|SpawnSoundatLocation (Variables|Default|GetShakeSound) (Transformation|GetActorLocation)))
    (Variables|Default|SetShake%20Audio%20Component _returnvalue_3)
    (if (Utilities|String|EqualExactly(String) (Utilities|GetDisplayName _asbp_cozy_lodge_selectable_present) "BP_CozyLodge_SelectablePresent_NoOpenMesh_S23")
      (Variables|Default|GetDoomAudio
        (:then
          _)
        (:else
          (bind _returnvalue_4 (Audio|SpawnSound2D (Variables|Default|GetDoomAudioClip)))
          (Variables|Default|SetDoomAudio _returnvalue_4))))
    (Audio|PushSoundMixModifier "/Game/Sounds/FortSoundClassesAndModes/Fort_Present_Duck_MenuMusic.Fort_Present_Duck_MenuMusic")
    (Utilities|Time|SetTimerbyEvent (delegate Custom|FadeDoomAudio) 1.0)
    (bind (_direction _normalized_timeprogression _normalized_falloff) (Timeline|Timeline_Shake PlayFromStart)
      (:Update
        (Variables|Default|SetIntensityShake (Math|Float|Clamp(Float) (* _intensityshake _normalized_falloff) 0.0 (Variables|Default|GetMaxShakeIntensity)))
        (Transformation|AddWorldRotation (Variables|Transformation|GetRootComponent) (Math|Rotator|MakeRotator (Math|Float|float*float (Math|Trig|Sin(Radians) (* (Utilities|Time|GetTimeSeconds) (Math|Float|float*float (Math|Trig|GetPI) 2.0 (Variables|Default|GetRotateFrequency)))) (Variables|Default|GetIntensityShake) (Variables|Default|GetRotationBias)) (Math|Float|float*float (Math|Trig|Sin(Radians) (* (Utilities|Time|GetTimeSeconds) (Math|Float|float*float (Math|Trig|GetPI) 2.0 (Variables|Default|GetRotateFrequency)))) (Variables|Default|GetIntensityShake) (Variables|Default|GetRotationBias)) (Math|Float|float*float (Math|Trig|Sin(Radians) (* (Utilities|Time|GetTimeSeconds) (Math|Float|float*float (Math|Trig|GetPI) 2.0 (Variables|Default|GetRotateFrequency)))) (Variables|Default|GetIntensityShake) (Variables|Default|GetRotationBias))) false false)
        (Rendering|Material|SetScalarParameterValue (Variables|Default|GetMID_Present) "ShakeAmount" _intensityshake :!class /Script/Engine.MaterialInstanceDynamic)))))

(event Custom|FadeDoomAudio
  ;[1 1 1] Shake
  (bind _doomaudio (Variables|Default|GetDoomAudio)
    (:then
      (Audio|Components|Audio|FadeOut _doomaudio 0.2 0.0 "Linear" :!class /Script/Engine.AudioComponent))
    (:else)))

(event EventOnOpenPresent
  (params (bActionCancelled :type Boolean))
  (bind _returnvalue (Transformation|GetActorLocation))
  (bind _returnvalue_2 (Utilities|GetDisplayName 0))
  (bind _returnvalue_3 (Utilities|String|EqualExactly(String) _returnvalue_2 "BP_CozyLodge_OpenPresent.HighlightMesh Wrapped_Galileo_Rocket"))
  (bind _mid_present (Variables|Default|GetMID_Present))
  (bind _timeline_openpresent (Variables|BP_CozyLodge_InteractablePresent|GetTimeline_OpenPresent))
  (bind _mid_present_2 (Variables|Default|GetMID_Present))
  (bind _blockopening (Variables|Default|GetBlockOpening))
  ;[1 1 1] Open
  (if bActionCancelled
    (if (not _blockopening)
      (Variables|Default|SetBlockRotation false)
      (Components|Timeline|SetPlayRate _timeline_openpresent 2.0)
      (bind _opening_sound_component (Variables|Default|GetOpening%20Sound%20Component)
        (:then
          (Audio|Components|Audio|FadeOut _opening_sound_component 0.5 0.0 "Linear" :!class /Script/Engine.AudioComponent))
        (:else
          _))
      (if _returnvalue_3
        (Components|Activation|Deactivate (Variables|Default|GetGalileoUnwrapFX)))
      (bind (_direction _normalizedprogress) (Timeline|Timeline_OpenPresent Reverse)
        (:Update
          (Rendering|Material|SetScalarParameterValue (Variables|Default|GetMID_Present) "Progress_Buldge" _normalizedprogress :!class /Script/Engine.MaterialInstanceDynamic)
          (switch Utilities|FlowControl|Switch|SwitchonETimelineDirection _direction
            (:Forward
              (Rendering|Material|SetScalarParameterValue _mid_present_2 "Progress_Reverse_Blocker" 1.0 :!class /Script/Engine.MaterialInstanceDynamic)
              (Transformation|SetRelativeRotation (Variables|Transformation|GetRootComponent) (Math|Interpolation|RInterpTo (Variables|/Script/Engine.SceneComponent|GetRelativeRotation (Variables|Transformation|GetRootComponent)) "0, 0, 0" (Utilities|Time|GetWorldDeltaSeconds) 1.0) false false))
            (:Backward
              (Rendering|Material|SetScalarParameterValue _mid_present_2 "Progress_Reverse_Blocker" 0.0 :!class /Script/Engine.MaterialInstanceDynamic))))
        (:Finished
          ;[1 1 1] Explode or actually open package
          (switch Utilities|FlowControl|Switch|SwitchonETimelineDirection _direction
            (:Forward
              (Variables|Default|SetBlockOpening true)
              (bind _returnvalue_4 (Rendering|Material|GetVectorParameterValue (Variables|Default|GetMID_Present) "ColorPaper1" :!class /Script/Engine.MaterialInstanceDynamic))
              (Effects|Components|ParticleSystem|SetVectorParameter (Variables|Default|GetP_PresentOpen) "Color" (Math|Conversions|ToVector(LinearColor) _returnvalue_4))
              (Effects|Components|ParticleSystem|SetColorParameter (Variables|Default|GetNS_OpenPresent) "TestColor" _returnvalue_4)
              (Audio|PlaySoundatLocation (Variables|Sound|GetOpen%20End%20Sound) (Transformation|GetActorLocation))
              (bind _opening_sound_component_2 (Variables|Default|GetOpening%20Sound%20Component)
                (:then
                  (Audio|Components|Audio|FadeOut _opening_sound_component_2 0.0 0.0 "Linear" :!class /Script/Engine.AudioComponent))
                (:else
                  _))
              (bind (_direction_2 _progress _fade) (Timeline|Timeline_0 PlayFromStart)
                (:Update
                  (Rendering|Material|SetScalarParameterValue _mid_present "ProgressFoldOpen" _progress :!class /Script/Engine.MaterialInstanceDynamic)
                  (Rendering|Material|SetScalarParameterValue _mid_present "Progress_Buldge" _fade :!class /Script/Engine.MaterialInstanceDynamic))
                (:TriggerFX
                  (Components|Activation|Activate (Variables|Default|GetP_PresentOpen) true)
                  (Components|Activation|Activate (Variables|Default|GetNS_OpenPresent) true))))))))
    (elif _blockopening
      _
      (else
        (Variables|Default|SetBlockRotation true)
        (Components|Timeline|SetPlayRate _timeline_openpresent 1.0)
        (bind _returnvalue_5 (Audio|SpawnSoundatLocation (Variables|Sound|GetPlay%20Opening%20Sound) _returnvalue))
        (Variables|Default|SetOpening%20Sound%20Component _returnvalue_5)
        (if _returnvalue_3
          ;[1 1 1] Shake
          ;[1 1 1] Galileo Unwrap FX
          (bind _returnvalue_6 (Effects|SpawnEmitteratLocation "/Game/Environments/FrontEnd/FX/P_CozyLodg_Galileo_Unwrap.P_CozyLodg_Galileo_Unwrap" _returnvalue))
          (Variables|Default|SetGalileoUnwrapFX _returnvalue_6))
        (Timeline|Timeline_OpenPresent PlayFromStart)))))

(event Custom|TestOpen
  (bind _mid_present (Variables|Default|GetMID_Present))
  ;[1 1 1] Explode or actually open package
  (bind _returnvalue (Rendering|Material|GetVectorParameterValue (Variables|Default|GetMID_Present) "ColorPaper1" :!class /Script/Engine.MaterialInstanceDynamic))
  (Effects|Components|ParticleSystem|SetVectorParameter (Variables|Default|GetP_PresentOpen) "Color" (Math|Conversions|ToVector(LinearColor) _returnvalue))
  (Effects|Components|ParticleSystem|SetColorParameter (Variables|Default|GetNS_OpenPresent) "TestColor" _returnvalue)
  (Audio|PlaySoundatLocation (Variables|Sound|GetOpen%20End%20Sound) (Transformation|GetActorLocation))
  (bind _opening_sound_component (Variables|Default|GetOpening%20Sound%20Component)
    (:then
      (Audio|Components|Audio|FadeOut _opening_sound_component 0.0 0.0 "Linear" :!class /Script/Engine.AudioComponent))
    (:else
      _))
  (Timeline|Timeline_0 PlayFromStart))

(event EventOnResetPresent
  (bind _displaymesh (Variables|SceneNav|GetDisplayMesh))
  (bind _staticmesh (Variables|/Script/Engine.StaticMeshComponent|GetStaticMesh _displaymesh))
  ;[1 1 1] Reset Local Data for next use
  (Variables|SceneNav|GetDisplayMesh
    (:then
      (if (Utilities|Array|ContainsItem (Variables|Default|GetOverrideMaterialObjects) (Utilities|GetObjectName _staticmesh))
        (bind _returnvalue (Rendering|Material|CreateDynamicMaterialInstance (Variables|SceneNav|GetDisplayMesh) 0 (StaticMesh|GetMaterial _staticmesh 0) "None" :!class /Script/Engine.PrimitiveComponent))
        (Variables|Default|SetMID_Present _returnvalue)
        (else
          (bind _asmaterial_instance_dynamic (Utilities|Casting|CastToMaterialInstanceDynamic (Rendering|Material|GetMaterial (Variables|SceneNav|GetDisplayMesh) 0))
            (:then
              (Variables|Default|SetMID_Present _asmaterial_instance_dynamic))
            (:CastFailed
              (Development|PrintString "Failed to get a MID on reset" true true "(R=1.000000,G=0.000000,B=0.054508,A=1.000000)")))))
      (Variables|Default|SetBlockRotation false)
      (Variables|Default|SetBlockOpening false))
    (:else)))

(fn ConstructionScript)
