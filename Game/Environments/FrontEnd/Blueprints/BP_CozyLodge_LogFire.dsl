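; Front-end Cozy Lodge fireplace that drips rested XP and sends Winterfest quest credit while the player watches it.
; BeginPlay caches five FX sockets and loads XP tiers and timer tuning from tables.
; Fireplace interaction ducks audio, starts the game-state rested timer and, after delays, fires quest feedback and a self-rescheduling randomized tick rolling a tier for XP, sound and Niagara FX.
; Camera changes end the timer.
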(event EventBeginPlay
  ;[1 1 1] Initialize data
  ;;[1 1 1] Psuedo Construction Script (wasn't running correctly previously due to streaming)
  (bind _output_get (Variables|BPVars|Setup|SetXPSocketLocation (Utilities|Array|MakeArray (Transformation|GetSocketLocation (Variables|SceneNav|GetDisplayMesh) "FX_XP_01") (Transformation|GetSocketLocation (Variables|SceneNav|GetDisplayMesh) "FX_XP_02") (Transformation|GetSocketLocation (Variables|SceneNav|GetDisplayMesh) "FX_XP_03") (Transformation|GetSocketLocation (Variables|SceneNav|GetDisplayMesh) "FX_XP_04") (Transformation|GetSocketLocation (Variables|SceneNav|GetDisplayMesh) "FX_XP_05"))))
  ;[1 0 0] XP Particle Logging
  (for (_array_element _array_index) _output_get
    (Development|PrintString (Utilities|String|Append "XPParticleLog: Spawn location " (Utilities|String|ToString(Integer) _array_index) " = " (Utilities|String|ToString(Vector) _array_element)) true true "(R=1.000000,G=0.131528,B=0.000000,A=1.000000)"))
  ;;[1 1 1] XP Data
  (bind _returnvalue (DataTable|GetDataTableColumnasString (Variables|BPVars|Setup|GetXPData) "XPValue"))
  (for _index (range (Utilities|Array|Length _returnvalue))
    (bind _returnvalue_2 (Utilities|GetDataTableRow (Variables|BPVars|Setup|GetXPData) (Utilities|String|StringToName (Utilities|String|ToString(Text) (Utilities|Text|FormatText "INVTEXT(\"XP{XPIndex}\")" _index))))
      (:then
        (Utilities|Array|Add (Variables|BPVars|Data|GetXPRowData) _returnvalue_2)
        ;[1 1 1] Timer data
        (bind (_outxy _returnvalue_3) (Fortnite|EvaluateCurveTableRow (Variables|BPVars|Setup|GetAverageTimeBetweenTicksRow)))
        (if _returnvalue_3
          (Variables|BPVars|Data|SetAverageTimeBetweenTicks _outxy))
        (bind (_outxy_2 _returnvalue_4) (Fortnite|EvaluateCurveTableRow (Variables|BPVars|Setup|GetTimerVariancePctRow)))
        (if _returnvalue_4
          (Variables|BPVars|Data|SetTimerVariancePct _outxy_2)))
      (:RowNotFound
        (Development|PrintString (Utilities|String|ToString(Text) (Utilities|Text|FormatText "INVTEXT(\"BP_CozyLodge_LogFire: Can\\'t find {TableRowName} data table row\")" (Utilities|Text|FormatText "INVTEXT(\"XP{XPIndex}\")" _index))))))))

(event Custom|GrantXP
  (params (XPIndex :type Integer) (XPValue :type Integer))
  ;[1 1 1] Grant XP and spawn FX
  (CallFunction|SpawnXPFX (Utilities|Array|Get(acopy) (Variables|BPVars|Setup|GetXPSocketLocation) (Math|Random|RandomIntegerinRange 0 4)) XPIndex)
  ;[1 0 0] XP Particle Logging
  (Development|PrintString (Utilities|String|Append "XPParticleLog: Grant XP. XP Value = " (Utilities|String|ToString(Integer) XPValue)) true true "(R=1.000000,G=0.131528,B=0.000000,A=1.000000)")
  (Class|FortGameStateComponentEventLevel|NotifyRestedXpGenAmount (CallFunction|GetOwningGameStateComp) XPValue))

(event Custom|StartXPgain
  (bind _returnvalue (CallFunction|GetOwningGameStateComp))
  (bind _bresult (Class|FortGameStateComponentEventLevel|CanGainRestedXp _returnvalue))
  ;[1 1 1] XP Drip
  ;;[1 1 1] Start XP
  (Development|PrintString "START XP GAIN FIRED" true true "(R=1.000000,G=0.872107,B=0.000000,A=1.000000)" 1e+01)
  ;;;[1 0 0] XP Particle Logging
  (Development|PrintString (Utilities|String|Append "XPParticleLog: XP Start. Rested Cap = " (Utilities|String|ToString(Boolean) _bresult)) true true "(R=1.000000,G=0.131528,B=0.000000,A=1.000000)")
  (if _bresult
    (Variables|BPVars|Data|SetXpSizeRoll (Math|Random|RandomIntegerinRange 0 100))
    (bind (_array_element _array_index) (Utilities|Array|ForEachLoopwithBreak (Variables|BPVars|Data|GetXPRowData) :!label _label)
      (:LoopBody
        (bind (_array_element_xpvalue_2_3a6eb9e14da4e98737cc98a06f370422 _array_element_streakbreakerincrement_4_eaea867b4b87ef1c5e69eebb9b2e970b _array_element_pctthreshold_10_42ea227042e0cb8aa46e6faf865118f5) (Utilities|Struct|BreakLogFireXPData _array_element))
        (if (<= (Variables|BPVars|Data|GetXpSizeRoll) _array_element_pctthreshold_10_42ea227042e0cb8aa46e6faf865118f5)
          (CallFunction|GrantXP _array_index _array_element_xpvalue_2_3a6eb9e14da4e98737cc98a06f370422)
          (goto _label Break)))
      (:Completed
        _)))
  (bind _returnvalue_2 (Utilities|Time|SetTimerbyEvent (delegate Custom|StartXPgain) (Variables|BPVars|Data|GetAverageTimeBetweenTicks) false false 0.0 (* (Variables|BPVars|Data|GetAverageTimeBetweenTicks) (Variables|BPVars|Data|GetTimerVariancePct))))
  (Variables|BPVars|Data|SetXPTimer _returnvalue_2))

(event Custom|EndXPgain
  ;[1 1 1] XP Drip
  ;;[1 1 1] End XP
  ;;;[1 0 0] XP Particle Logging
  (Development|PrintString (Utilities|String|Append "XPParticleLog: End XP") true true "(R=1.000000,G=0.131528,B=0.000000,A=1.000000)")
  (Utilities|Time|ClearandInvalidateTimerbyHandle (Variables|BPVars|Data|GetXPTimer))
  (Variables|BPVars|Data|SetIs%20Watching false))

(event EventHandleObjectSpecialInteraction
  (params (ActionTag :type Gameplay%20Tag%20Structure) (CameraTag :type Gameplay%20Tag%20Structure))
  (bind _returnvalue (CallFunction|GetOwningGameStateComp))
  ;[0.5146 0.9688 0.5666] HandleObjectSpecialInteraction
  ; Start log fire, wintefest quest update, xp gain
  (if (and (GameplayTags|Equal(GameplayTag) ActionTag "(TagName=\"ERG.Interaction.Fireplace\")") (GameplayTags|Equal(GameplayTag) CameraTag "(TagName=\"ERG.Camera.Staging\")"))
    (Variables|BPVars|Data|SetIs%20Watching true)
    (Class|FortGameStateComponentEventLevel|StartRestedXpTimer _returnvalue)
    (Utilities|FlowControl|Sequence
      (:then_0
        ;;[0.07081 0.1832 1] Sequence 0: Update Winterfest quest
        (Utilities|FlowControl|Delay (Variables|Default|GetQuestUpdateDelayTime)) ; delay a moment for players to see the fire before triggering quest and feedback
        (if (Variables|BPVars|Data|GetIs%20Watching)
          (for _array_element (Fortnite|GetLocalFortPlayerControllers true)
            (Utilities|IsValid _array_element
              (:Is%20Valid
                (Utilities|IsValid (Quests|GetQuestManager _array_element "Invalid")
                  (:Is%20Valid
                    (Quests|SendClientStatEventWithTags :QuestManager (Quests|GetQuestManager _array_element "Invalid") :SourceTags (GameplayTags|MakeLiteralGameplayTagContainer "(GameplayTags=((TagName=\"Athena.Quests.Winterfest.Frontend.CozyFireplace\")))") :Count 1
                      (:OnSuccess
                        ;;[0.6704 0.1614 1] Play Quest feedback VFX and SFX
                        (Audio|PlaySound2D (Variables|Default|GetSoundToPlayOnQuestCredit))
                        (Effects|Components|ParticleSystem|SpawnFXAtLocation :FXSystemAsset (Variables|Default|GetSuccessEmitter) :Location (Transformation|GetActorLocation) :Scale (Variables|Default|GetSuccessEmitterScale) :bAutoDestroy true :PoolingMethod "None" :bAutoActivate true)
                        ;[0.6704 0.1614 1] Delegate Call for the UI "Quest Complete" Message
                        (Default|CallOnWinterfestQuestCompleted))))))))))
      (:then_1
        ;;[0.4279 1 0.4576] Sequence 1: Update XP gains
        (EventDispatchers|BindEventtoOnRestedXpDripStopped _returnvalue (EventDispatchers|CreateEvent :!event Custom|EndXPgain))
        (Utilities|FlowControl|Delay (Variables|BPVars|Setup|GetFirstXPDelay))
        (if (Variables|BPVars|Data|GetIs%20Watching)
          (CallFunction|StartXPgain)))
      (:then_2
        ;;[1 0.4265 0] Sequence 2: Bind Camera changed for ending XP drip
        (Utilities|FlowControl|DelayUntilNextTick)
        (EventDispatchers|BindEventtoActiveCameraChanged _returnvalue (EventDispatchers|CreateEvent :!event Custom|OnCameraChanged))
        (Audio|PushSoundMixModifier "/Game/Sounds/FortSoundClassesAndModes/Fort_YuleLog_Duck_Mix.Fort_YuleLog_Duck_Mix")))))

(event Custom|OnCameraChanged
  (params (CameraTag :type Gameplay%20Tag%20Structure))
  (bind _returnvalue (CallFunction|GetOwningGameStateComp))
  ;[1 0.4265 0] Stop xp drip when camera shifts away from fireplace
  (if (!= CameraTag (Variables|Default|GetFireplaceActiveCameraTag))
    (EventDispatchers|UnbindEventfromActiveCameraChanged _returnvalue (delegate Custom|OnCameraChanged))
    (Class|FortGameStateComponentEventLevel|EndRestXpTimer _returnvalue)
    (Audio|PopSoundMixModifier "/Game/Sounds/FortSoundClassesAndModes/Fort_YuleLog_Duck_Mix.Fort_YuleLog_Duck_Mix")))

(fn SpawnXPFX
  (params (SpawnLocation :type Vector) (XPIndex :type Integer))
  (bind _returnvalue (Audio|SpawnSoundatLocation (Variables|Default|GetFireplace_XP_GainedSound) SpawnLocation))
  (bind _output_get (Variables|Default|SetXP%20Pop%20Audio _returnvalue))
  (Audio|Parameter|SetIntegerParameter _output_get "XPGain" XPIndex :!class /Script/AudioExtensions.AudioParameterControllerInterface)
  (Audio|Components|Audio|Play _output_get 0.0)
  ;[1 0 0] XP Particle Logging
  (Development|PrintString (Utilities|String|Append "XPParticleLog: Pre FX Switch. XPIndex = " (Utilities|String|ToString(Integer) XPIndex)) true true "(R=1.000000,G=0.131528,B=0.000000,A=1.000000)")
  (switch int XPIndex
    (:0
      ;[1 0 0] XP Particle Logging
      (Development|PrintString (Utilities|String|Append "XPParticleLog: FX Spawn. XPIndex = " (Utilities|String|ToString(Integer) XPIndex) " Location = " (Utilities|String|ToString(Vector) SpawnLocation)) true true "(R=1.000000,G=0.131528,B=0.000000,A=1.000000)")
      (Niagara|SpawnSystematLocation "/Game/Environments/FrontEnd/FX/Cozy/NS_Cozy_Fire_10XP.NS_Cozy_Fire_10XP" SpawnLocation))
    (:1
      (Development|PrintString (Utilities|String|Append "XPParticleLog: FX Spawn. XPIndex = " (Utilities|String|ToString(Integer) XPIndex) " Location = " (Utilities|String|ToString(Vector) SpawnLocation)) true true "(R=1.000000,G=0.131528,B=0.000000,A=1.000000)")
      (Niagara|SpawnSystematLocation "/Game/Environments/FrontEnd/FX/Cozy/NS_Cozy_Fire_20XP.NS_Cozy_Fire_20XP" SpawnLocation))
    (:2
      (Development|PrintString (Utilities|String|Append "XPParticleLog: FX Spawn. XPIndex = " (Utilities|String|ToString(Integer) XPIndex) " Location = " (Utilities|String|ToString(Vector) SpawnLocation)) true true "(R=1.000000,G=0.131528,B=0.000000,A=1.000000)")
      (Niagara|SpawnSystematLocation "/Game/Environments/FrontEnd/FX/Cozy/NS_Cozy_Fire_100XP.NS_Cozy_Fire_100XP" SpawnLocation))
    (:3
      (Development|PrintString (Utilities|String|Append "XPParticleLog: FX Spawn. XPIndex = " (Utilities|String|ToString(Integer) XPIndex) " Location = " (Utilities|String|ToString(Vector) SpawnLocation)) true true "(R=1.000000,G=0.131528,B=0.000000,A=1.000000)")
      (Niagara|SpawnSystematLocation "/Game/Environments/FrontEnd/FX/Cozy/NS_Cozy_Fire_1000XP.NS_Cozy_Fire_1000XP" SpawnLocation))
    (:4
      (Development|PrintString (Utilities|String|Append "XPParticleLog: FX Spawn. XPIndex = " (Utilities|String|ToString(Integer) XPIndex) " Location = " (Utilities|String|ToString(Vector) SpawnLocation)) true true "(R=1.000000,G=0.131528,B=0.000000,A=1.000000)")
      (Niagara|SpawnSystematLocation "/Game/Environments/FrontEnd/FX/Cozy/NS_Cozy_Fire_5000XP.NS_Cozy_Fire_5000XP" SpawnLocation))))

(fn ConstructionScript
  ;[1 0 0] XP Particle Logging
  (Development|PrintString (Utilities|String|Append "XPParticleLog: Construction script runs") true true "(R=1.000000,G=0.131528,B=0.000000,A=1.000000)"))
