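; Selectable present in the Cozy Lodge front-end scene that updates its display mesh to show opened or no-longer-claimable states.
; The opened update runs once, swapping in the opened mesh if valid.
; CopyMaterialInfo builds dynamic instances of both materials and copies ribbon, paper and tag parameters, so the opened present keeps the wrapped look.
; The no-longer-claimable update only replaces the material with the unavailable one.
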(event EventBeginPlay)

(event Collision|EventActorBeginOverlap
  (params (OtherActor :type Actor%20Object%20Reference)))

(event EventTick
  (params (DeltaSeconds :type Float%20%28single-precision%29)))

(event EventUpdateDisplayMeshToOpened
  (bind _openedmesh (Variables|Default|GetOpenedMesh))
  (Utilities|FlowControl|DoOnce false
    (:Completed
      (Utilities|IsValid _openedmesh
        (:Is%20Valid
          (bind _mid_open (CallFunction|CopyMaterialInfo))
          (Components|StaticMesh|SetStaticMesh (Variables|SceneNav|GetDisplayMesh) _openedmesh)
          (Rendering|Material|SetMaterial (Variables|SceneNav|GetDisplayMesh) 0 _mid_open))))))

(event EventUpdateDisplayMeshToNoLongerClaimable
  (Rendering|Material|SetMaterial (Variables|SceneNav|GetDisplayMesh) 0 (Variables|Default|GetUnavailablePresentMaterial)))

(fn CopyMaterialInfo
  (bind _returnvalue (Rendering|Material|CreateDynamicMaterialInstance (Rendering|Material|GetMaterial (Variables|SceneNav|GetDisplayMesh) 0) "None" "None" :!class /Script/Engine.KismetMaterialLibrary))
  (Variables|Default|SetMID_Present _returnvalue)
  (bind _returnvalue_2 (Rendering|Material|CreateDynamicMaterialInstance (StaticMesh|GetMaterial (Variables|Default|GetOpenedMesh) 0) "None" "None" :!class /Script/Engine.KismetMaterialLibrary))
  (Variables|Default|SetMID_Open _returnvalue_2)
  (bind _returnvalue_3 (Rendering|Material|GetScalarParameterValue (Variables|Default|GetMID_Present) "RibbonType" :!class /Script/Engine.MaterialInstanceDynamic))
  (Rendering|Material|SetScalarParameterValue (Variables|Default|GetMID_Open) "RibbonType" _returnvalue_3 :!class /Script/Engine.MaterialInstanceDynamic)
  (bind _returnvalue_4 (Rendering|Material|GetVectorParameterValue (Variables|Default|GetMID_Present) "ColorRibbon" :!class /Script/Engine.MaterialInstanceDynamic))
  (Rendering|Material|SetVectorParameterValue (Variables|Default|GetMID_Open) "ColorRibbon" _returnvalue_4 :!class /Script/Engine.MaterialInstanceDynamic)
  (bind _returnvalue_5 (Rendering|Material|GetVectorParameterValue (Variables|Default|GetMID_Present) "ColorPaper1" :!class /Script/Engine.MaterialInstanceDynamic))
  (Rendering|Material|SetVectorParameterValue (Variables|Default|GetMID_Open) "ColorPaper1" _returnvalue_5 :!class /Script/Engine.MaterialInstanceDynamic)
  (bind _returnvalue_6 (Rendering|Material|GetVectorParameterValue (Variables|Default|GetMID_Present) "Tag" :!class /Script/Engine.MaterialInstanceDynamic))
  (Rendering|Material|SetVectorParameterValue (Variables|Default|GetMID_Open) "Tag" _returnvalue_6 :!class /Script/Engine.MaterialInstanceDynamic)
  (bind _returnvalue_7 (Rendering|Material|GetVectorParameterValue (Variables|Default|GetMID_Present) "PaperType" :!class /Script/Engine.MaterialInstanceDynamic))
  (Rendering|Material|SetVectorParameterValue (Variables|Default|GetMID_Open) "PaperType" _returnvalue_7 :!class /Script/Engine.MaterialInstanceDynamic)
  (return (Variables|Default|GetMID_Open)))

(fn ConstructionScript)
