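; Selectable present Blueprint for the tree in the front-end Cozy Lodge, which attempts to claim its event reward graph node when interacted with.
; The special interaction handler responds only when the action tag is the tree interaction and the camera tag is focus 3.
; It then finds the dynamic level streaming components on the GameState and asks each to claim the node identified by GetNodeTag.
; BeginPlay, overlap and Tick are disabled stubs, and construction just calls the parent.
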(event EventBeginPlay
  (|CallParentFunction) ; This node is disabled and will not be called.\nDrag off pins to build functionality.
)

(event Collision|EventActorBeginOverlap
  (params (OtherActor :type Actor%20Object%20Reference))
  (|CallParentFunction OtherActor) ; This node is disabled and will not be called.\nDrag off pins to build functionality.
)

(event EventTick
  (params (DeltaSeconds :type Float%20%28single-precision%29))
  (|CallParentFunction DeltaSeconds) ; This node is disabled and will not be called.\nDrag off pins to build functionality.
)

(event EventHandleObjectSpecialInteraction
  (params (ActionTag :type Gameplay%20Tag%20Structure) (CameraTag :type Gameplay%20Tag%20Structure))
  (if (and (GameplayTags|Equal(GameplayTag) ActionTag "(TagName=\"ERG.Interaction.Tree\")") (GameplayTags|Equal(GameplayTag) CameraTag "(TagName=\"ERG.Camera.Focus.3\")"))
    (for _array_element (Actor|GetComponentsByClass (Game|GetGameState) "/Game/Blueprints/Components/GSC_DynamicLevelStreaming.GSC_DynamicLevelStreaming_C")
      (Class|FortGameStateComponentEventLevel|TrytoClaimRewardGraphNode _array_element (CallFunction|GetNodeTag)))))

(fn ConstructionScript
  (|CallParentFunction))
