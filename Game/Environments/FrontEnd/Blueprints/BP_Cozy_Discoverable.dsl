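; BP_Cozy_Discoverable is a front-end environment Blueprint for a discoverable object that responds to special interactions.
; An interaction stores its action and camera tags, then commits if allowed, with an optional timer cooldown blocking repeats.
; The commit event only prints a debug string here, apparently as a hook for specific behavior.
; When notified, it relays the tags and start flag to each discoverable in its OtherDiscoverablesToNotify list.
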(event EventBeginPlay)

(event Custom|ResetCooldown
  (Variables|Default|SetIsInteractAllowed true))

(event EventHandleObjectSpecialInteraction
  (params (ActionTag :type Gameplay%20Tag%20Structure) (CameraTag :type Gameplay%20Tag%20Structure))
  (Development|PrintString "HandleObjectSpecialInteraction")
  (Variables|Default|SetInteracted_ActionTag ActionTag)
  (Variables|Default|SetInteracted_CameraTag CameraTag)
  (if (Variables|Default|GetIsInteractAllowed)
    (CallFunction|CommitonSpecialInteraction)
    (if (> (Variables|Default|GetInteractCooldown) 0.0)
      (Variables|Default|SetIsInteractAllowed false)
      (Utilities|Time|SetTimerbyEvent (delegate Custom|ResetCooldown) (Variables|Default|GetInteractCooldown)))))

(event Custom|Commit_OnSpecialInteraction
  (Development|PrintString "Commit_OnSpecialInteraction"))

(event Custom|NotifiedByDiscoverable
  (params (ActionTag :type Gameplay%20Tag%20Structure) (CameraTag :type Gameplay%20Tag%20Structure) (bIsStart :type Boolean))
  (for _array_element (Variables|Default|GetOtherDiscoverablesToNotify)
    (CallFunction|NotifiedbyDiscoverable _array_element ActionTag CameraTag bIsStart)))

(fn ConstructionScript)
