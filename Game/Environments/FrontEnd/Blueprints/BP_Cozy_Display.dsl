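; BP_Cozy_Display is a front-end display actor showing a configurable mesh, skeletal mesh or particle effect, with an optional unlock-attention sparkle.
; The construction script applies those assets, then feeds manual or component-derived bounds and spawn rate into UnlockSparkle's instance parameters.
; Unless sparkle is disabled, it also adds duplicate static and skeletal meshes using the UnlockAttention material as highlights.
; BeginPlay stops skeletal component ticking when no mesh is assigned.
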(event EventBeginPlay
  (bind _skeletalmesh (Variables|Default|GetSkeletalMesh))
  ;[1 0.4082 0] Disable skeletal mesh component tick if there's no mesh.
  (Variables|/Script/Engine.SkinnedMeshComponent|GetSkeletalMesh _skeletalmesh
    (:then)
    (:else
      (Components|Tick|SetComponentTickEnabled _skeletalmesh false))))

(event Collision|EventActorBeginOverlap
  (params (OtherActor :type Actor%20Object%20Reference)))

(event EventTick
  (params (DeltaSeconds :type Float%20%28single-precision%29)))

(fn ConstructionScript
  (bind (_min _max) (Components|StaticMesh|GetLocalBounds (Variables|Default|GetStaticMesh)))
  (bind (_origin _boxextent _sphereradius) (Collision|GetComponentBounds (Variables|Default|GetSkeletalMesh)))
  (bind (_origin_2 _boxextent_2 _sphereradius_2) (Collision|GetComponentBounds (Variables|Default|GetFX)))
  (bind _parameters (Variables|Default|GetParameters))
  (bind _staticmesh (Variables|Default|GetStaticMesh))
  (bind _skeletalmesh (Variables|Default|GetSkeletalMesh))
  (bind _skeletalmesh_2 (Variables|/Script/Engine.SkinnedMeshComponent|GetSkeletalMesh _skeletalmesh))
  (bind _disable_sparkle (Variables|Default|GetDisable%20Sparkle))
  (bind _unlocksparkle (Variables|Default|GetUnlockSparkle))
  (Utilities|FlowControl|Sequence
    (:then_0
      (Utilities|IsValid (Variables|Default|GetMesh)
        (:Is%20Valid
          (Components|StaticMesh|SetStaticMesh (Variables|Default|GetStaticMesh) (Variables|Default|GetMesh)))))
    (:then_1
      (Utilities|IsValid (Variables|Default|GetSkeletalMesh)
        (:Is%20Valid
          (Components|SkinnedMesh|SetSkinnedAssetandUpdate (Variables|Default|GetSkeletalMesh) (Variables|Default|GetSkelMesh) true))))
    (:then_2
      (Utilities|IsValid (Variables|Default|GetParticleSystem)
        (:Is%20Valid
          (Effects|Components|ParticleSystem|SetTemplate (Variables|Default|GetFX) (Variables|Default|GetParticleSystem)))))
    (:then_3
      (if (Variables|Default|GetDisable%20Sparkle)
        (Rendering|SetVisibility _unlocksparkle false false)
        (else
          (Transformation|SetWorldScale3D _unlocksparkle "1.000000,1.000000,1.000000")
          (Utilities|Array|SetArrayElem _parameters 0 (Utilities|Struct|MakeParticleSysParam "Location" "PSPT_VectorRand" 0.0 0.0 (Math|Vector|SelectVector (Variables|Default|GetMaxBounds) (Math|Vector|SelectVector _max (Math|Vector|SelectVector (Math|Vector|MakeVector _sphereradius _sphereradius _sphereradius) (Math|Vector|SelectVector (Math|Vector|MakeVector _sphereradius_2 _sphereradius_2 _sphereradius_2) "0, 0, 0" (Utilities|IsValid (Variables|Default|GetFX) :!class /Script/Engine.KismetSystemLibrary)) (Utilities|IsValid (Variables|Default|GetSkeletalMesh) :!class /Script/Engine.KismetSystemLibrary)) (Utilities|IsValid (Variables|Default|GetStaticMesh) :!class /Script/Engine.KismetSystemLibrary)) (Variables|Default|GetManualBounds)) (Math|Vector|SelectVector (Variables|Default|GetMinBounds) (Math|Vector|SelectVector _min (Math|Vector|SelectVector (* (Math|Vector|MakeVector _sphereradius _sphereradius _sphereradius) -1.0) (Math|Vector|SelectVector (Math|Vector|MakeVector _sphereradius_2 _sphereradius_2 _sphereradius_2) (* (Math|Vector|MakeVector _sphereradius_2 _sphereradius_2 _sphereradius_2) -1.0) (Utilities|IsValid (Variables|Default|GetFX) :!class /Script/Engine.KismetSystemLibrary)) (Utilities|IsValid (Variables|Default|GetSkeletalMesh) :!class /Script/Engine.KismetSystemLibrary)) (Utilities|IsValid (Variables|Default|GetStaticMesh) :!class /Script/Engine.KismetSystemLibrary)) (Variables|Default|GetManualBounds))))
          (Utilities|Array|SetArrayElem _parameters 1 (Utilities|Struct|MakeParticleSysParam "SpawnRateScale" "PSPT_ScalarRand" (Variables|Default|GetSpawnRate) (Variables|Default|GetSpawnRate)) true)
          (Effects|Components|ParticleSystem|SetInstanceParams _unlocksparkle _parameters))))
    (:then_4
      (if (not _disable_sparkle)
        (Utilities|IsValid _skeletalmesh_2
          (:Is%20Valid
            (bind _returnvalue (AddComponent|NODE_AddSkeletalMeshComponent-0))
            (Components|SkinnedMesh|SetSkinnedAssetandUpdate _returnvalue _skeletalmesh_2 true)
            (Rendering|Material|SetMaterial _returnvalue 0 "/Game/Environments/FrontEnd/Materials/Cozy/UnlockAttention.UnlockAttention")))))
    (:then_5
      (if (not _disable_sparkle)
        (Utilities|IsValid _staticmesh
          (:Is%20Valid
            (bind _returnvalue_2 (AddComponent|NODE_AddStaticMeshComponent-0))
            (Components|StaticMesh|SetStaticMesh _returnvalue_2 (Variables|/Script/Engine.StaticMeshComponent|GetStaticMesh _staticmesh))
            (Rendering|Material|SetMaterial _returnvalue_2 0 "/Game/Environments/FrontEnd/Materials/Cozy/UnlockAttention.UnlockAttention")))))))
