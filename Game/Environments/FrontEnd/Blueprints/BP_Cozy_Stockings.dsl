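; Front-end cozy lodge stocking that tracks whether the local player has an uncompleted Winterfest stocking quest and reacts to interaction.
; CheckStockingState, rerun on inventory updates or by timer when its first check fails, scans possible quests to set an active flag and call full or empty state events, both empty here.
; A timeline drives a material jiggle during interaction.
; Interacting with an active quest spawns celebration particles and sound, then rechecks state.
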(event Custom|Debug_CurrentQuestsTotal)

(event Custom|CheckStockingState
  (bind _returnvalue (Quests|GetQuestManager (Fortnite|GetFirstLocalFortPlayerController true) "Invalid"))
  (bind _returnvalue_2 (Utilities|IsValid _returnvalue :!class /Script/Engine.KismetSystemLibrary))
  (bind _returnvalue_3 (Fortnite|GetFirstLocalFortPlayerController true))
  ;[1 1 1] CheckStockingState
  (CallFunction|DebugCurrentQuestsTotal)
  (Variables|Default|SetbHasStockingQuestActive false)
  (bind _playerhasquest (Class|AthenaQuestFunctionLibrary|AthenaHasQuest _returnvalue_3 (Variables|Default|GetFirstStockingQuest)))
  (if (and _returnvalue_2 _playerhasquest)
    (Development|PrintString (Utilities|String|Append "QuestManager.IsValid() = " (Utilities|String|ToString(Boolean) _returnvalue_2)))
    (for _array_element (Variables|Default|GetPossibleStockingQuests)
      (bind _playerhasquest_2 (Class|AthenaQuestFunctionLibrary|AthenaHasQuest _returnvalue_3 _array_element))
      (if (and _playerhasquest_2 (not (Quest|HasCompletedObjective :self (Quests|GetQuestManager _returnvalue_3 "Invalid") :Definition _array_element :!class /Script/FortniteGame.FortQuestManager)))
        (Variables|Default|SetbHasStockingQuestActive true)))
    (if (Variables|Default|GetbHasStockingQuestActive)
      (CallFunction|SetStateFull)
      (else
        (CallFunction|SetStateEmpty)))
    (else
      (Utilities|Time|SetTimerbyEvent (EventDispatchers|CreateEvent :!event Custom|CheckStockingState) 0.1 (not _returnvalue_2)))))

(event EventBeginPlay)

(event Custom|BP_OnInventoryUpdated
  (CallFunction|CheckStockingState))

(event Custom|SetStateFull)

(event Custom|SetStateEmpty)

(event Custom|OnFrontendInteract
  (if (Variables|Default|GetbHasStockingQuestActive)
    (Effects|SpawnEmitteratLocation :EmitterTemplate "/Game/Environments/FrontEnd/FX/P_CozyLodge_Celebration.P_CozyLodge_Celebration" :Location (Transformation|GetWorldLocation (Variables|Default|GetFXLoc)) :Scale "0.300000,0.300000,0.300000" :bAutoDestroy true :PoolingMethod "None" :bAutoActivateSystem true)
    (Audio|PlaySoundatLocation "/Game/Athena/Sounds/Events/Winterfest/Winterfest_Stocking_Opening_End_Cue.Winterfest_Stocking_Opening_End_Cue" (Transformation|GetActorLocation))
    (CallFunction|CheckStockingState)))

(event Custom|StartInteraction
  (bind _mid (Variables|Default|GetMID))
  ;[1 1 1] Animated Jiggle when opening needs to get triggered from interaction
  (Components|Timeline|SetPlayRate (Variables|BP_Cozy_Stockings|GetTimeline_0) 1.0)
  (bind (_direction _progressjiggle) (Timeline|Timeline_0 PlayFromStart)
    (:Update
      (Rendering|Material|SetScalarParameterValue _mid "ProgressJiggle" (* _progressjiggle (Variables|Default|GetJiggleIntensity)) :!class /Script/Engine.MaterialInstanceDynamic))
    (:Finished
      (Rendering|Material|SetScalarParameterValue _mid "ProgressJiggle" 0.0 :!class /Script/Engine.MaterialInstanceDynamic))))

(event Custom|StopInteraction
  (bind _mid (Variables|Default|GetMID))
  ;[1 1 1] Animated Jiggle when opening needs to get triggered from interaction
  (Components|Timeline|SetPlayRate (Variables|BP_Cozy_Stockings|GetTimeline_0) 2.0)
  (Timeline|Timeline_0 Reverse))

(fn GetHasStockingQuestActive
  (return (Variables|Default|GetbHasStockingQuestActive)))

(fn ConstructionScript)
