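; Front-end Cozy Lodge exterior light switcher that turns the set's lights, fog and particles on or off through enable/disable events.
; Low shadow quality, mobile, optional low-effects or debug checks select low-detail directional, sky and fog actors.
; BeginPlay switches everything off, re-activates active backdrop cameras to avoid a race, then binds settings updates.
; Those updates swap only the directional light and refresh BP_DetailLevelMesh visibility by shadow quality and platform.
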(event Custom|UpdateSettings
  ;[1 1 1] On Setting change Event
  (if (Variables|Default|GetisActive)
    (Utilities|FlowControl|Sequence
      (:then_0
        (if (Math|Boolean|ORBoolean (== (Settings|GetShadowQuality (stale EventBeginPlay _returnvalue)) 0) (Fortnite|IsMobilePlatform) (and (Variables|Default|GetLOW_FX_Setting_Use_MobileLighting) (== (Settings|GetVisualEffectQuality (stale EventBeginPlay _returnvalue)) 0)))
          (Rendering|SetVisibility (Variables|/Script/Engine.Light|GetLightComponent (Variables|Default|GetDirectionLight)) false false)
          (Rendering|SetVisibility (Variables|/Script/Engine.Light|GetLightComponent (Variables|Default|GetDirectionalLight_LowDetailMode)) true false)
          (else
            (Rendering|SetVisibility (Variables|/Script/Engine.Light|GetLightComponent (Variables|Default|GetDirectionLight)) true false)
            (Rendering|SetVisibility (Variables|/Script/Engine.Light|GetLightComponent (Variables|Default|GetDirectionalLight_LowDetailMode)) false false))))
      (:then_1
        ;;[1 1 1] If Shadow quality is Medium or lower the Volumetric fog is disabled and it shows the simple fog cards
        (CallFunction|SetVisiblityofDetailSpecificMeshes)))))

(event EventBeginPlay
  (Variables|Default|SetDebugLOWQualityLighting false)
  (CallFunction|SetVisiblityofSceneLighting false)
  ;[1 1 1] Fix for sometimes no lighting if race condition enables lights before begin play was triggered
  (bind _outactors (Actor|GetAllActorsOfClass "/Game/Environments/FrontEnd/Blueprints/FortnitePartyBackdrop_Camera.FortnitePartyBackdrop_Camera_C"))
  (bind _array_element (Utilities|Array|ForEachLoopwithBreak _outactors)
    (:LoopBody
      (if (Variables|/Game/Environments/FrontEnd/Blueprints/FortnitePartyBackdrop_Camera.FortnitePartyBackdrop_Camera_C|GetActive _array_element)
        (Class|FortnitePartyBackdropCamera|ManualActivate _array_element)
        (Development|PrintString (Utilities|String|ToString(Boolean) (Variables|/Game/Environments/FrontEnd/Blueprints/FortnitePartyBackdrop_Camera.FortnitePartyBackdrop_Camera_C|GetActive _array_element)))
        (bind _returnvalue (Settings|GetGameUserSettings))
        (EventDispatchers|BindEventtoOnGameUserSettingsUINeedsUpdate _returnvalue (delegate Custom|UpdateSettings))
        ;[1 1 1] If Shadow quality is Medium or lower the Volumetric fog is disabled and it shows the simple fog cards
        (CallFunction|SetVisiblityofDetailSpecificMeshes)))))

(event EventEnableLevelSetLighting
  (CallFunction|SetVisiblityofSceneLighting true))

(event Custom|EnableLobbySetLighting
  (CallFunction|SetVisiblityofSceneLighting true))

(event EventDisableLevelSetLighting
  (CallFunction|SetVisiblityofSceneLighting false))

(event Custom|DisableLobbySetLighting
  (CallFunction|SetVisiblityofSceneLighting false))

(fn SetVisiblityofDetailSpecificMeshes
  (bind _returnvalue (Settings|GetGameUserSettings))
  (bind _outactors (Actor|GetAllActorsOfClass "/Game/Environments/FrontEnd/Blueprints/BP_DetailLevelMesh.BP_DetailLevelMesh_C"))
  (for _array_element _outactors
    (bind _asbp_detail_level_mesh (Utilities|Casting|CastToBP_DetailLevelMesh _array_element))
    (if (and (Utilities|String|EqualExactly(String) (Game|GetPlatformName) "Switch") (Variables|/Game/Environments/FrontEnd/Blueprints/BP_DetailLevelMesh.BP_DetailLevelMesh_C|GetNotVisibleOnSwitch _asbp_detail_level_mesh))
      (Rendering|SetVisibility (Variables|/Game/Environments/FrontEnd/Blueprints/BP_DetailLevelMesh.BP_DetailLevelMesh_C|GetStaticMesh _array_element) false false)
      (elif (Variables|/Game/Environments/FrontEnd/Blueprints/BP_DetailLevelMesh.BP_DetailLevelMesh_C|GetAlwaysVisible _asbp_detail_level_mesh)
        (Rendering|SetVisibility (Variables|/Game/Environments/FrontEnd/Blueprints/BP_DetailLevelMesh.BP_DetailLevelMesh_C|GetStaticMesh _array_element) true false)
        (else
          (Rendering|SetVisibility (Variables|/Game/Environments/FrontEnd/Blueprints/BP_DetailLevelMesh.BP_DetailLevelMesh_C|GetStaticMesh _array_element) (or (or (<= (Settings|GetShadowQuality _returnvalue) 3) (Fortnite|IsMobilePlatform)) (Variables|Default|GetDebugLOWQualityLighting)) false))))))

(fn SetVisiblityofSceneLighting
  (params (Activate :type Boolean))
  (bind _skylight_lowdetailmode (Variables|Default|GetSkyLight_LowDetailMode))
  (bind _skylight (Variables|Default|GetSkyLight))
  (bind _skylight_2 (Variables|Default|GetSkyLight))
  (bind _skylight_3 (Variables|Default|GetSkyLight))
  (bind _skylight_lowdetailmode_2 (Variables|Default|GetSkyLight_LowDetailMode))
  (bind _exponentialheightfog_lowdetailmode (Variables|Default|GetExponentialHeightfog_LowDetailMode))
  (bind _directionlight (Variables|Default|GetDirectionLight))
  (bind _skylight_lowdetailmode_3 (Variables|Default|GetSkyLight_LowDetailMode))
  (bind _exponentialheightfog (Variables|Default|GetExponentialHeightfog))
  (bind _exponentialheightfog_2 (Variables|Default|GetExponentialHeightfog))
  (bind _directionallight_lowdetailmode (Variables|Default|GetDirectionalLight_LowDetailMode))
  (bind _exponentialheightfog_lowdetailmode_2 (Variables|Default|GetExponentialHeightfog_LowDetailMode))
  (bind _exponentialheightfog_3 (Variables|Default|GetExponentialHeightfog))
  (bind _exponentialheightfog_lowdetailmode_3 (Variables|Default|GetExponentialHeightfog_LowDetailMode))
  (bind _directionallight_lowdetailmode_2 (Variables|Default|GetDirectionalLight_LowDetailMode))
  (bind _directionlight_2 (Variables|Default|GetDirectionLight))
  (bind _directionlight_3 (Variables|Default|GetDirectionLight))
  (bind _directionallight_lowdetailmode_3 (Variables|Default|GetDirectionalLight_LowDetailMode))
  (Utilities|FlowControl|Sequence
    (:then_0
      (if Activate
        ;[1 1 1] Check for low detail mode and enable alternative directional light
        (bind _returnvalue (Settings|GetGameUserSettings))
        (if (or (and (Math|Boolean|ORBoolean (== (Settings|GetShadowQuality _returnvalue) 0) (Fortnite|IsMobilePlatform) (and (Variables|Default|GetLOW_FX_Setting_Use_MobileLighting) (== (Settings|GetVisualEffectQuality _returnvalue) 0))) (Utilities|IsValid (Variables|Default|GetDirectionalLight_LowDetailMode) :!class /Script/Engine.KismetSystemLibrary)) (Variables|Default|GetDebugLOWQualityLighting))
          (Utilities|FlowControl|Sequence
            (:then_0
              (Utilities|IsValid _directionlight_3
                (:Is%20Valid
                  (Rendering|SetVisibility (Variables|/Script/Engine.Light|GetLightComponent _directionlight_3) false false))))
            (:then_1
              (Utilities|IsValid _directionallight_lowdetailmode_3
                (:Is%20Valid
                  (Rendering|SetVisibility (Variables|/Script/Engine.Light|GetLightComponent _directionallight_lowdetailmode_3) true false)))))
          (else
            (Utilities|FlowControl|Sequence
              (:then_0
                (Utilities|IsValid _directionallight_lowdetailmode_2
                  (:Is%20Valid
                    (Rendering|SetVisibility (Variables|/Script/Engine.Light|GetLightComponent _directionallight_lowdetailmode_2) false false))))
              (:then_1
                (Utilities|IsValid _directionlight
                  (:Is%20Valid
                    (Rendering|SetVisibility (Variables|/Script/Engine.Light|GetLightComponent _directionlight) true false)))))))
        (else
          (Utilities|FlowControl|Sequence
            (:then_0
              (Utilities|IsValid _directionlight_2
                (:Is%20Valid
                  (Rendering|SetVisibility (Variables|/Script/Engine.Light|GetLightComponent _directionlight_2) Activate false))))
            (:then_1
              (Utilities|IsValid _directionallight_lowdetailmode
                (:Is%20Valid
                  (Rendering|SetVisibility (Variables|/Script/Engine.Light|GetLightComponent _directionallight_lowdetailmode) Activate false))))))))
    (:then_1
      (if Activate
        (bind _returnvalue_2 (Settings|GetGameUserSettings))
        (if (or (and (Math|Boolean|ORBoolean (== (Settings|GetShadowQuality _returnvalue_2) 0) (Fortnite|IsMobilePlatform) (and (Variables|Default|GetLOW_FX_Setting_Use_MobileLighting) (== (Settings|GetVisualEffectQuality _returnvalue_2) 0))) (Utilities|IsValid (Variables|Default|GetDirectionalLight_LowDetailMode) :!class /Script/Engine.KismetSystemLibrary)) (Variables|Default|GetDebugLOWQualityLighting))
          (Utilities|FlowControl|Sequence
            (:then_0
              (Utilities|IsValid _skylight_2
                (:Is%20Valid
                  (Rendering|SetVisibility (Variables|/Script/Engine.SkyLight|GetLightComponent _skylight_2) false false))))
            (:then_1
              (Utilities|IsValid _skylight_lowdetailmode_3
                (:Is%20Valid
                  (Rendering|SetVisibility (Variables|/Script/Engine.SkyLight|GetLightComponent _skylight_lowdetailmode_3) true false)))))
          (else
            (Utilities|FlowControl|Sequence
              (:then_0
                (Utilities|IsValid _skylight_lowdetailmode_2
                  (:Is%20Valid
                    (Rendering|SetVisibility (Variables|/Script/Engine.SkyLight|GetLightComponent _skylight_lowdetailmode_2) false false))))
              (:then_1
                (Utilities|IsValid _skylight_3
                  (:Is%20Valid
                    (Rendering|SetVisibility (Variables|/Script/Engine.SkyLight|GetLightComponent _skylight_3) true false)))))))
        (else
          (Utilities|FlowControl|Sequence
            (:then_0
              (Utilities|IsValid _skylight
                (:Is%20Valid
                  (Rendering|SetVisibility (Variables|/Script/Engine.SkyLight|GetLightComponent _skylight) Activate false))))
            (:then_1
              (Utilities|IsValid _skylight_lowdetailmode
                (:Is%20Valid
                  (Rendering|SetVisibility (Variables|/Script/Engine.SkyLight|GetLightComponent _skylight_lowdetailmode) Activate false))))))))
    (:then_2
      (for _array_element (Variables|Default|GetPointlights)
        (Utilities|IsValid _array_element
          (:Is%20Valid
            (Rendering|SetVisibility (Variables|/Script/Engine.Light|GetLightComponent _array_element) Activate false)))))
    (:then_3
      (if Activate
        (bind _returnvalue_3 (Settings|GetGameUserSettings))
        (if (or (and (Math|Boolean|ORBoolean (== (Settings|GetShadowQuality _returnvalue_3) 0) (Fortnite|IsMobilePlatform) (and (Variables|Default|GetLOW_FX_Setting_Use_MobileLighting) (== (Settings|GetVisualEffectQuality _returnvalue_3) 0))) (Utilities|IsValid (Variables|Default|GetDirectionalLight_LowDetailMode) :!class /Script/Engine.KismetSystemLibrary)) (Variables|Default|GetDebugLOWQualityLighting))
          (Utilities|FlowControl|Sequence
            (:then_0
              (Utilities|IsValid _exponentialheightfog_3
                (:Is%20Valid
                  (Rendering|SetVisibility (Variables|/Script/Engine.ExponentialHeightFog|GetComponent _exponentialheightfog_3) false false))))
            (:then_1
              (Utilities|IsValid _exponentialheightfog_lowdetailmode_3
                (:Is%20Valid
                  (Rendering|SetVisibility (Variables|/Script/Engine.ExponentialHeightFog|GetComponent _exponentialheightfog_lowdetailmode_3) true false)))))
          (else
            (Utilities|FlowControl|Sequence
              (:then_0
                (Utilities|IsValid _exponentialheightfog_lowdetailmode
                  (:Is%20Valid
                    (Rendering|SetVisibility (Variables|/Script/Engine.ExponentialHeightFog|GetComponent _exponentialheightfog_lowdetailmode) false false))))
              (:then_1
                (Utilities|IsValid _exponentialheightfog
                  (:Is%20Valid
                    (Rendering|SetVisibility (Variables|/Script/Engine.ExponentialHeightFog|GetComponent _exponentialheightfog) true false)))))))
        (else
          (Utilities|FlowControl|Sequence
            (:then_0
              (Utilities|IsValid _exponentialheightfog_2
                (:Is%20Valid
                  (Rendering|SetVisibility (Variables|/Script/Engine.ExponentialHeightFog|GetComponent _exponentialheightfog_2) Activate false))))
            (:then_1
              (Utilities|IsValid _exponentialheightfog_lowdetailmode_2
                (:Is%20Valid
                  (Rendering|SetVisibility (Variables|/Script/Engine.ExponentialHeightFog|GetComponent _exponentialheightfog_lowdetailmode_2) false false))))))))
    (:then_4
      (for _array_element_2 (Variables|Default|GetParticleSystems)
        (Utilities|IsValid _array_element_2
          (:Is%20Valid
            (if Activate
              (Components|Activation|Activate (Variables|/Script/CascadeRuntime.Emitter|GetParticleSystemComponent _array_element_2) false)
              (else
                (Components|Activation|Deactivate (Variables|/Script/CascadeRuntime.Emitter|GetParticleSystemComponent _array_element_2))))))))
    (:then_5
      (Variables|Default|SetisActive Activate))))

(fn ConstructionScript
  (CallFunction|SetVisiblityofSceneLighting true)
  (CallFunction|SetVisiblityofDetailSpecificMeshes))
