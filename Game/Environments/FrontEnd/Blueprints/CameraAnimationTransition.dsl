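; Front-end Blueprint that switches the menu camera between the HeroSelect and MultipleSquads views.
; Custom events MoveCameraUp (MultipleSquads) and MoveCameraDown (HeroSelect) store the target in CameraState, then set the front-end camera with a timed ease-out view-target blend.
; Both are gated by CanChangeCamera, which requires the current camera to be a hero camera and both CameraBR and CameraBR16Player references to be valid.
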(event EventBeginPlay)

(event Collision|EventActorBeginOverlap
  (params (OtherActor :type Actor%20Object%20Reference)))

(event EventTick
  (params (DeltaSeconds :type Float%20%28single-precision%29)))

(event Custom|MoveCameraUp
  (bind _canchangecamera (CallFunction|CanChangeCamera))
  (if _canchangecamera
    (Variables|Default|SetCameraState "MultipleSquads")
    ;[1 1 1] Change Camera
    (FrontEndCamera|SetFrontEndCamera (BlueprintContext|GetFortFrontEndContext) (Variables|Default|GetCameraState) (Utilities|Struct|MakeViewTargetTransitionParams 1.5 "VTBlend_EaseOut"))))

(event Custom|MoveCameraDown
  (bind _canchangecamera (CallFunction|CanChangeCamera))
  (if _canchangecamera
    (Variables|Default|SetCameraState "HeroSelect")
    ;[1 1 1] Change Camera
    (FrontEndCamera|SetFrontEndCamera (BlueprintContext|GetFortFrontEndContext) (Variables|Default|GetCameraState) (Utilities|Struct|MakeViewTargetTransitionParams 1.5 "VTBlend_EaseOut"))))

(fn CanChangeCamera
  (return (Math|Boolean|ANDBoolean (FrontEndCamera|IsHeroCamera (FrontEndCamera|GetFrontEndCamera (BlueprintContext|GetFortFrontEndContext))) (Utilities|IsValid (Variables|Default|GetCameraBR) :!class /Script/Engine.KismetSystemLibrary) (Utilities|IsValid (Variables|Default|GetCameraBR16Player) :!class /Script/Engine.KismetSystemLibrary))))

(fn ConstructionScript)
