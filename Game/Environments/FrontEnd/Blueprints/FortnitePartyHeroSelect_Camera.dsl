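; Front-end camera for the party hero select screen, with logic to spin the hero pawn by mouse drag.
; Activation fires the camera-switch effect on all TVPostProcessBP actors and enables input; deactivation disables input.
; While MouseDown is set, Tick turns the cached pawn and its mesh by horizontal mouse delta.
; Press/release handlers cache or clear hero 0, but the mouse button events are empty and release also sets MouseDown true.
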(event Camera|Fortnite|EventOnActivated
  (params (PlayerController :type Fort%20Player%20Controller%20Object%20Reference))
  ;[0.5688 1 0] Camera switch post process anim
  (bind _outactors (Actor|GetAllActorsOfClass "/Game/UI/Foundation/Environment/Cameras/TVPostProcessBP.TVPostProcessBP_C"))
  (for _array_element _outactors
    (Class|TVPostProcessBP|FrontEndCameraSwitch _array_element))
  ;[1 1 1] Enable input
  (Input|EnableInput (Game|GetPlayerController 0)))

(event Camera|Fortnite|EventOnDeactivated
  (params (PlayerController :type Fort%20Player%20Controller%20Object%20Reference))
  ;[1 1 1] Disable input
  (Input|DisableInput (Game|GetPlayerController 0)))

(event LeftMouseButton
  (params (Key :type Key%20Structure)))

(event RightMouseButton
  (params (Key :type Key%20Structure)))

(event EventTick
  (params (DeltaSeconds :type Float%20%28single-precision%29))
  (bind _returnvalue (Game|GetPlayerController 0))
  (bind _deltax (Game|Player|GetInputMouseDelta _returnvalue))
  (bind _returnvalue_2 (* _deltax 10))
  (bind _cachedpawn (Variables|Default|GetCachedPawn))
  (bind _returnvalue_3 (Transformation|GetActorRotation _cachedpawn))
  (bind _yaw (.yaw _returnvalue_3))
  (bind _returnvalue_4 (- _yaw _returnvalue_2))
  (bind _roll (.roll _returnvalue_3))
  (bind _pitch (.pitch _returnvalue_3))
  (bind _returnvalue_5 (Math|Rotator|MakeRotator _roll _pitch _returnvalue_4))
  (if (Variables|Default|GetMouseDown)
    (Utilities|IsValid _cachedpawn
      (:Is%20Valid
        (Transformation|SetActorRotation _cachedpawn _returnvalue_5 false)
        (Transformation|SetRelativeRotation (Variables|/Script/Engine.Character|GetMesh (Variables|Default|GetCachedPawn)) _returnvalue_5 false false)))))

(fn HandleMousePress
  (Variables|Default|SetMouseDown true)
  (bind _returnvalue (FrontEndContext|GetHeroPlayerPawnbyIndex (BlueprintContext|GetFortFrontEndContext) 0))
  (Utilities|IsValid _returnvalue
    (:Is%20Valid
      (Variables|Default|SetCachedPawn _returnvalue))))

(fn HandleMouseRelease
  (Variables|Default|SetMouseDown true)
  (Variables|Default|SetCachedPawn 0))

(fn ConstructionScript)
