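; Front-end camera Blueprint for the Cozy Lodge lobby set that switches lighting and time of day with its active state.
; Activation, becoming view target or ManualActivate sets Active, sets time to noon at zero speed, hides the time-of-day manager and enables lobby lighting on all CozyLodge LightSwitcher actors.
; Deactivation or ending view target clears Active and disables that lighting, leaving time of day untouched.
; A custom event resets its transform to a saved one.
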(event Custom|Reset%20Location
  (Transformation|SetActorTransform (Variables|Default|GetSavedTransform)))

(event Camera|Fortnite|EventOnActivated
  (params (PlayerController :type Fort%20Player%20Controller%20Object%20Reference))
  (Development|PrintString "I am Activated")
  (Variables|Default|SetActive true)
  ;[1 1 1] TimeOfDayManager settings
  (TimeOfDay|SetTimeOfDay 12 :!class /Script/FortniteGame.FortKismetLibrary)
  (TimeOfDay|SetTimeOfDaySpeed 0.0 :!class /Script/FortniteGame.FortKismetLibrary)
  (TimeOfDay|SetTimeOfDayManagerVisibility (TimeOfDay|GetContextualTimeOfDayManager) false)
  ;[1 1 1] Turn on necessary lights
  (Utilities|FlowControl|Sequence
    (:then_0
      (bind _outactors (Actor|GetAllActorsOfClass "/Game/Environments/FrontEnd/Blueprints/BP_Fortnite_CozyLodge_LightSwitcher.BP_Fortnite_CozyLodge_LightSwitcher_C"))
      (if (> (Utilities|Array|Length _outactors) 0)
        (for _array_element _outactors
          (Class|BPFortniteCozyLodgeLightSwitcher|EnableLobbySetLighting _array_element)
          (Development|PrintString "EnableCozylighitng"))))))

(event Actor|EventOnBecomeViewTarget
  (params (PC :type Player%20Controller%20Object%20Reference))
  (Development|PrintString "I am Activated")
  (Variables|Default|SetActive true)
  ;[1 1 1] TimeOfDayManager settings
  (TimeOfDay|SetTimeOfDay 12 :!class /Script/FortniteGame.FortKismetLibrary)
  (TimeOfDay|SetTimeOfDaySpeed 0.0 :!class /Script/FortniteGame.FortKismetLibrary)
  (TimeOfDay|SetTimeOfDayManagerVisibility (TimeOfDay|GetContextualTimeOfDayManager) false)
  ;[1 1 1] Turn on necessary lights
  (Utilities|FlowControl|Sequence
    (:then_0
      (bind _outactors (Actor|GetAllActorsOfClass "/Game/Environments/FrontEnd/Blueprints/BP_Fortnite_CozyLodge_LightSwitcher.BP_Fortnite_CozyLodge_LightSwitcher_C"))
      (if (> (Utilities|Array|Length _outactors) 0)
        (for _array_element _outactors
          (Class|BPFortniteCozyLodgeLightSwitcher|EnableLobbySetLighting _array_element)
          (Development|PrintString "EnableCozylighitng"))))))

(event Custom|ManualActivate
  (Variables|Default|SetActive true)
  ;[1 1 1] TimeOfDayManager settings
  (TimeOfDay|SetTimeOfDay 12 :!class /Script/FortniteGame.FortKismetLibrary)
  (TimeOfDay|SetTimeOfDaySpeed 0.0 :!class /Script/FortniteGame.FortKismetLibrary)
  (TimeOfDay|SetTimeOfDayManagerVisibility (TimeOfDay|GetContextualTimeOfDayManager) false)
  ;[1 1 1] Turn on necessary lights
  (Utilities|FlowControl|Sequence
    (:then_0
      (bind _outactors (Actor|GetAllActorsOfClass "/Game/Environments/FrontEnd/Blueprints/BP_Fortnite_CozyLodge_LightSwitcher.BP_Fortnite_CozyLodge_LightSwitcher_C"))
      (if (> (Utilities|Array|Length _outactors) 0)
        (for _array_element _outactors
          (Class|BPFortniteCozyLodgeLightSwitcher|EnableLobbySetLighting _array_element)
          (Development|PrintString "EnableCozylighitng"))))))

(event Camera|Fortnite|EventOnDeactivated
  (params (PlayerController :type Fort%20Player%20Controller%20Object%20Reference))
  (Variables|Default|SetActive false)
  ;[1 1 1] Turn off necessary lights
  (bind _outactors (Actor|GetAllActorsOfClass "/Game/Environments/FrontEnd/Blueprints/BP_Fortnite_CozyLodge_LightSwitcher.BP_Fortnite_CozyLodge_LightSwitcher_C"))
  (if (> (Utilities|Array|Length _outactors) 0)
    (for _array_element _outactors
      (Class|BPFortniteCozyLodgeLightSwitcher|DisableLobbySetLighting _array_element))))

(event Actor|EventOnEndViewTarget
  (params (PC :type Player%20Controller%20Object%20Reference))
  (Variables|Default|SetActive false)
  ;[1 1 1] Turn off necessary lights
  (bind _outactors (Actor|GetAllActorsOfClass "/Game/Environments/FrontEnd/Blueprints/BP_Fortnite_CozyLodge_LightSwitcher.BP_Fortnite_CozyLodge_LightSwitcher_C"))
  (if (> (Utilities|Array|Length _outactors) 0)
    (for _array_element _outactors
      (Class|BPFortniteCozyLodgeLightSwitcher|DisableLobbySetLighting _array_element))))

(fn ConstructionScript)
