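; Front-end camera Blueprint for the exterior of the Cozy Lodge environment, managing time of day and level lighting as it gains or loses the view.
; Camera activation, becoming view target and ManualActivate each set Active, freeze time at noon, hide the TimeOfDayManager and call UpdateLevelLightSwitch on.
; The three matching deactivation events only switch level lights off, leaving Active and time-of-day settings untouched.
; A Reset Location custom event restores the actor's saved transform.
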(event Custom|Reset%20Location
  (Transformation|SetActorTransform (Variables|Default|GetSavedTransform)))

(event Camera|Fortnite|EventOnActivated
  (params (PlayerController :type Fort%20Player%20Controller%20Object%20Reference))
  (Variables|Default|SetActive true)
  ;[1 1 1] TimeOfDayManager settings
  (TimeOfDay|SetTimeOfDay 12 :!class /Script/FortniteGame.FortKismetLibrary)
  (TimeOfDay|SetTimeOfDaySpeed 0.0 :!class /Script/FortniteGame.FortKismetLibrary)
  (TimeOfDay|SetTimeOfDayManagerVisibility (TimeOfDay|GetContextualTimeOfDayManager) false)
  (CallFunction|UpdateLevelLightSwitch true))

(event Actor|EventOnBecomeViewTarget
  (params (PC :type Player%20Controller%20Object%20Reference))
  (Variables|Default|SetActive true)
  ;[1 1 1] TimeOfDayManager settings
  (TimeOfDay|SetTimeOfDay 12 :!class /Script/FortniteGame.FortKismetLibrary)
  (TimeOfDay|SetTimeOfDaySpeed 0.0 :!class /Script/FortniteGame.FortKismetLibrary)
  (TimeOfDay|SetTimeOfDayManagerVisibility (TimeOfDay|GetContextualTimeOfDayManager) false)
  (CallFunction|UpdateLevelLightSwitch true))

(event Custom|ManualActivate
  (Variables|Default|SetActive true)
  ;[1 1 1] TimeOfDayManager settings
  (TimeOfDay|SetTimeOfDay 12 :!class /Script/FortniteGame.FortKismetLibrary)
  (TimeOfDay|SetTimeOfDaySpeed 0.0 :!class /Script/FortniteGame.FortKismetLibrary)
  (TimeOfDay|SetTimeOfDayManagerVisibility (TimeOfDay|GetContextualTimeOfDayManager) false)
  (CallFunction|UpdateLevelLightSwitch true))

(event Camera|Fortnite|EventOnDeactivated
  (params (PlayerController :type Fort%20Player%20Controller%20Object%20Reference))
  (CallFunction|UpdateLevelLightSwitch false))

(event Actor|EventOnEndViewTarget
  (params (PC :type Player%20Controller%20Object%20Reference))
  (CallFunction|UpdateLevelLightSwitch false))

(event Custom|ManualDeactivate
  (CallFunction|UpdateLevelLightSwitch false))

(fn ConstructionScript)
