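; Front-end party display manager that previews selected items and heroes and handles party pawn effects in the lobby.
; Showing an item resets visuals, loads assets, then fills a skeletal mesh, static mesh or prefab child actor, with special cases for ghost weapons.
; Level-up and evolution play particles, plus timeline pivot scaling for non-characters.
; Party pawns get teleport effects and in-progress hologram materials that copy vertex-hiding parameters and textures.
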(event EventTick
  (params (DeltaSeconds :type Float%20%28single-precision%29))
  (if (Variables|Default|GetPreviewRotation)
    (CallFunction|UpdatePreviewMeshTransforms)))

(event PartyDisplayManager|EventShowItem
  (params (ItemToView :type Fort%20Item%20Object%20Reference) (RequestId :type Guid%20Structure%20%28by%20ref%29))
  (bind _output_get (Variables|Default|SetItemToRepresent ItemToView))
  (CallFunction|ResetVisuals)
  (bind _returnvalue (PartyDisplayManager|GetHeroPlayerPawnforCurrentDisplayedItem))
  (Variables|Default|SetHeroPawn _returnvalue)
  (Utilities|IsValid _output_get
    (:Is%20Valid
      (Networking|SwitchHasAuthority
        (:Authority
          _)
        (:Remote
          (FortPawn|SetPawnVisibilityLocal (Variables|Default|GetHeroPawn) true true)))
      (PartyDisplayManager|BeginLoadingAssetsforItem (Variables|Default|GetItemToRepresent) (Item|GetPreviewAssetsfromContext (Variables|Default|GetItemToRepresent) (Variables|Default|GetContext%20Tags)) RequestId))
    (:Is%20Not%20Valid
      (Components|DestroyComponent (Variables|Default|GetTransitionFXComponent))
      (CallFunction|UpdatePreviewMeshTransforms)
      (PartyDisplayManager|HandleItemsShown (Utilities|Array|MakeArray ItemToView) RequestId))))

(event PartyDisplayManager|EventPlayLevelUpEffect
  (Networking|CanExecuteCosmeticEvents
    (:True
      (Variables|VFX|SetIsEvolve false)
      (CallFunction|CelebratewithFX))))

(event EventResOutPawnFromLobby
  (params (Pawn :type Fort%20Player%20Pawn%20Object%20Reference) (PartyIndex :type Integer))
  ;[0 0 0] Pawn removed, queue a res out if we are in athena (destruction is handled already in other modes)
  (if (not (== PartyIndex 0))
    (bind _asplayer_pawn_athena (Utilities|Casting|CastToPlayerPawn_Athena Pawn)
      (:then
        (if (Customization|IsinCharacterCustomization _asplayer_pawn_athena)
          (LobbyDisplay|QueueTeleportOut _asplayer_pawn_athena)
          (else
            (GameplayCue|ExecuteGameplayCue%28Burst%29Local _asplayer_pawn_athena (Variables|/Game/Athena/PlayerPawn_Athena.PlayerPawn_Athena_C|GetElimination%20Rez%20Sequence%20GC _asplayer_pawn_athena) (Ability|GameplayCue|MakeGameplayCueParameters 0.0)))))
      (:CastFailed
        _)))
  ;;[0 0 0] vault actor gets no FX on clear, wah wah
  (Actor|DestroyActor Pawn))

(event PartyDisplayManager|EventPlayEvolutionEffect
  (Networking|CanExecuteCosmeticEvents
    (:True
      (Variables|VFX|SetIsEvolve true)
      (CallFunction|CelebratewithFX))))

(event EventCharacterCustomizationFinished
  (params (Pawn :type Fort%20Player%20Pawn%20Object%20Reference))
  ;[0 0 0] Post Cosmetics Updated
  ;;[0 0 0] PIP - Override the pawn materials with hologram if in progress
  (CallFunction|PartyMemberinProgressCheck Pawn)
  ;;[0 0 0] any players mid-teleport need to override materials
  (CallFunction|RestoreTeleportMIDsinProgress Pawn))

(event Custom|CelebrateWithFX
  (bind _celebratoryfx (Variables|Default|GetCelebratoryFX))
  (bind _returnvalue (CallFunction|GetWeaponPlacementTransform))
  (Variables|VFX|SetIsCharacter (not (Utilities|IsValid (CallFunction|GetItemDefinitiontoShow) :!class /Script/Engine.KismetSystemLibrary)))
  (Variables|VFX|SetFXTargetLocation (.location _returnvalue))
  (Transformation|SetWorldLocation (Variables|Default|GetCelebratoryFX) (Variables|VFX|GetFXTargetLocation) false true)
  (if (Variables|VFX|GetIsEvolve)
    (if (Variables|VFX|GetIsCharacter)
      (Effects|Components|ParticleSystem|SetTemplate _celebratoryfx (Variables|VFX|GetCharEvolveUpFX))
      (else
        (Effects|Components|ParticleSystem|SetTemplate _celebratoryfx (Variables|VFX|GetEvolveUpFX))))
    (elif (Variables|VFX|GetIsCharacter)
      (Effects|Components|ParticleSystem|SetTemplate _celebratoryfx (Variables|VFX|GetCharLevelUpFX))
      (else
        (Effects|Components|ParticleSystem|SetTemplate _celebratoryfx (Variables|VFX|GetLevelUpFX)))))
  (Components|Activation|Activate (Variables|Default|GetCelebratoryFX) false)
  (if (not (Variables|VFX|GetIsCharacter))
    (CallFunction|PerformScaleMeshFX)))

(event Custom|PerformScaleMeshFX
  (Variables|Properties|SetIsPlayingCelebrateFX true)
  (Transformation|SetWorldLocation (Variables|Default|GetScalePivot) (Variables|VFX|GetFXTargetLocation) false true)
  (Transformation|SetRelativeLocation (Variables|Default|GetStaticMesh) (* (Variables|/Script/Engine.SceneComponent|GetRelativeLocation (Variables|Default|GetScalePivot)) -1) false true)
  (if (Variables|VFX|GetIsEvolve)
    (bind (_direction _scaling) (Timeline|ScaleAnimEvolve PlayFromStart)
      (:Update
        (Transformation|SetRelativeScale3D (Variables|Default|GetScalePivot) (Math|Conversions|ToVector(Float) _scaling)))
      (:Finished
        _))
    (else
      (bind (_direction_2 _scaling_2) (Timeline|ScaleAnimLevel PlayFromStart)
        (:Update
          (Transformation|SetRelativeScale3D (Variables|Default|GetScalePivot) (Math|Conversions|ToVector(Float) _scaling_2)))
        (:Finished
          _))))
  (Transformation|SetRelativeScale3D (Variables|Default|GetScalePivot) "1.000000,1.000000,1.000000")
  (CallFunction|ResetPreScaleLocations)
  (Variables|Properties|SetIsPlayingCelebrateFX false))

(event EventOnPlayerPawnAddedToLobby
  (params (Pawn :type Fort%20Player%20Pawn%20Object%20Reference))
  (bind (_outplayerindex _returnvalue) (CallFunction|GetPartyMemberIndexbyPlayerPawn Pawn))
  ;[0 0 0] On pawn added to lobby -NOTE: also occurs on cosmetics re-init
  (if _returnvalue
    (switch Utilities|FlowControl|Switch|SwitchonEFortPartyMemberLocation (PartyContext|GetPartyMemberLocation (BlueprintContext|GetFortPartyContext) (CallFunction|GetPlayerIdfromPartyMemberIndex _outplayerindex))
      (:PreLobby
        _)
      (:Lobby
        _))
    (bind _asplayer_pawn_athena (Utilities|Casting|CastToPlayerPawn_Athena Pawn))
    (LobbyDisplay|QueueTeleportIn _asplayer_pawn_athena)
    (elif (Utilities|Enum|Equal(Enum) (Fortnite|GetSubGame) "Athena")
      (Effects|SpawnEmitteratLocation (Variables|VFX|GetCharSwapTransitionFX) (Transformation|GetActorLocation Pawn))
      (Components|DestroyComponent (Variables|Default|GetTransitionFXComponent))
      (bind _returnvalue_2 (Effects|SpawnEmitteratLocation (Variables|VFX|GetCharSwapTransitionFXLoop) (Transformation|GetActorLocation)))
      (Variables|Default|SetTransitionFXComponent _returnvalue_2)
      (Utilities|FlowControl|Delay 1.0)
      (Utilities|FlowControl|WaitforBothExecutions true
        (:Then
          (Components|DestroyComponent (Variables|Default|GetTransitionFXComponent)))))))

(event EventBeginPlay
  (Utilities|IsValid (Variables|PartyInProgress|GetMID_HoloMaterial)
    (:Is%20Not%20Valid
      (bind _returnvalue (Rendering|Material|CreateDynamicMaterialInstance (Variables|MaterialFX|GetPlayerInMatchHoloMaterial) "None" "None" :!class /Script/Engine.KismetMaterialLibrary))
      (Variables|PartyInProgress|SetMID_HoloMaterial _returnvalue))))

(event PartyDisplayManager|EventHandleLoadingAssetsForItemCompleted
  (params (ItemWhoseAssetsWereLoaded :type Fort%20Item%20Object%20Reference) (LoadedAssets :type Array%20of%20Object%20References) (RequestId :type Guid%20Structure%20%28by%20ref%29))
  (bind _itemdefinition (CallFunction|GetItemDefinitiontoShow))
  (bind (_asset _returnvalue) (Frontend|GetPreviewSkeletalMeshAsset _itemdefinition))
  (bind (_asset_2 _returnvalue_2) (Frontend|GetPreviewStaticMeshAsset _itemdefinition))
  (bind _context_tags (Variables|Default|GetContext%20Tags))
  (bind (_asset_3 _returnvalue_3) (Frontend|GetPreviewPrefabAssetfromContext _itemdefinition _context_tags))
  (bind _output (Utilities|ResolveSoftReference _asset))
  (if (Utilities|Equal(Object) ItemWhoseAssetsWereLoaded (Variables|Default|GetItemToRepresent))
    (Networking|SwitchHasAuthority
      (:Authority
        _)
      (:Remote
        (FortPawn|SetPawnVisibilityLocal (Variables|Default|GetHeroPawn) false true)))
    (if _returnvalue
      (CallFunction|SetupSkeletalMeshVisuals _output)
      (if (Utilities|Equal(Object) _output "/Game/Weapons/FORT_Pistols/Mesh/SK_GhostPistol.SK_GhostPistol")
        (bind _returnvalue_4 (Effects|SpawnEmitterAttached "/Game/Weapons/FORT_Pistols/Effects/Weapon/M_GhostFlame.M_GhostFlame" (Variables|Default|GetSkeletalMesh) "None" "2.306265, 0, 0"))
        (Variables|Default|SetGhostPistolVFX _returnvalue_4)
        (elif (Utilities|Equal(Object) _output "/Game/Weapons/FORT_Melee/Meshes/SK_Sword_Ghost.SK_Sword_Ghost")
          (bind _returnvalue_5 (Effects|SpawnEmitterAttached "/Game/Weapons/FORT_Melee/Effects/GhostBlade/P_IdleFX.P_IdleFX" (Variables|Default|GetSkeletalMesh)))
          (Variables|Default|SetGhost%20Sword%20VFX _returnvalue_5)
          (bind _returnvalue_6 (AddComponent|NODE_AddStaticMeshComponent-0 :RelativeTransform (Transformation|GetRelativeTransform (Variables|Default|GetSkeletalMesh))))
          (bind _output_get (Variables|Default|SetGhost%20Sword%20Mesh _returnvalue_6))
          (Transformation|AttachComponentToComponent _output_get (Variables|Default|GetSkeletalMesh) "None" "KeepRelative" "KeepRelative" "KeepRelative" true)
          (CallFunction|UpdatePreviewMeshTransforms)
          (PartyDisplayManager|HandleItemsShown (Utilities|Array|MakeArray ItemWhoseAssetsWereLoaded) RequestId)))
      (else
        (if _returnvalue_2
          (CallFunction|SetupStaticMeshVisuals (Utilities|ResolveSoftReference _asset_2))
          (if (not (Utilities|IsValid (Item|GetSchematicCraftingResult (Variables|Default|GetItemToRepresent)) :!class /Script/Engine.KismetSystemLibrary))
            (CallFunction|SetUIOverrideVisuals (Utilities|ResolveSoftReference (Item|GetLargePreviewImage (Variables|Default|GetItemToRepresent))) (CallFunction|GetItemDefinitiontoShow) (Item|GetRarity (Variables|Default|GetItemToRepresent))))
          (else
            (if _returnvalue_3
              (CallFunction|SetCharacterPartsVisibility false)
              (ChildActorComponent|SetChildActorClass (Variables|Default|GetPrefabActorComp) (Utilities|ResolveSoftReference _asset_3))
              (bind (_item _requestid) (PartyDisplayManager|SetupPrefabVisuals self ItemWhoseAssetsWereLoaded RequestId)
                (:Completed
                  (CallFunction|PostSetupPrefabVisuals)
                  (CallFunction|UpdatePreviewMeshTransforms)
                  (PartyDisplayManager|HandleItemsShown (Utilities|Array|MakeArray _item) _requestid)))
              (else
                (PartyDisplayManager|HandleItemsShown (Utilities|Array|MakeArray ItemWhoseAssetsWereLoaded) RequestId)))))))))

(event PartyDisplayManager|EventSetupPrefabVisuals
  (bind _returnvalue (PartyDisplayManager|GetHeroPlayerPawnforCurrentDisplayedItem))
  (Utilities|IsValid _returnvalue
    (:Is%20Valid
      (bind _returnvalue_2 (Animation|Utilities|GetFrontEndAnimInstance (Variables|/Script/Engine.Character|GetMesh _returnvalue)))
      (Utilities|IsValid _returnvalue_2
        (:Is%20Valid
          (Variables|/Script/FortniteGame.FrontendAnimInstance|SetbIsSkydiving false _returnvalue_2))
        (:Is%20Not%20Valid
          _))
      (Animation|Montage|MontageStop (Components|SkeletalMesh|GetAnimInstance (Variables|/Script/Engine.Character|GetMesh _returnvalue)) 0.0 0)
      (bind _asplayer_pawn_athena (Utilities|Casting|CastToPlayerPawn_Athena _returnvalue)
        (:then
          (Components|Activation|Deactivate (Class|FortContrailsComponent|GetContrailFXComponent (Variables|/Game/Athena/PlayerPawn_Athena.PlayerPawn_Athena_C|GetFortContrails _asplayer_pawn_athena))))
        (:CastFailed
          _))
      (Transformation|SetWorldTransform (Variables|Default|GetPivotHolder) (Transformation|GetActorTransform) false true)
      (bind _asb_emote_preview_display (Utilities|Casting|CastToB_EmotePreviewDisplay (Variables|/Script/Engine.ChildActorComponent|GetChildActor (Variables|Default|GetPrefabActorComp)))
        (:then
          (Default|AssignOnEmoteSetupFinished _asb_emote_preview_display (delegate Custom|OnEmoteSetupFinished))
          (Frontend|CustomizePreviewPrefab (CallFunction|GetItemDefinitiontoShow) _asb_emote_preview_display)
          (Rendering|PrestreamTextures (Variables|/Script/Engine.ChildActorComponent|GetChildActor (Variables|Default|GetPrefabActorComp)) 5.0 true 0))
        (:CastFailed
          (Frontend|CustomizePreviewPrefab (CallFunction|GetItemDefinitiontoShow) (Variables|/Script/Engine.ChildActorComponent|GetChildActor (Variables|Default|GetPrefabActorComp)))
          (Rendering|PrestreamTextures (Variables|/Script/Engine.ChildActorComponent|GetChildActor (Variables|Default|GetPrefabActorComp)) 5.0 true 0)
          (CallFunction|NotifyPrefabChanged (Variables|/Script/Engine.ChildActorComponent|GetChildActor (Variables|Default|GetPrefabActorComp)) (Variables|Default|GetItemToRepresent))
          (PartyDisplayManager|CallOnItemDisplayReady 0))))
    (:Is%20Not%20Valid
      _)))

(event Custom|Set%20Character%20Parts%20Visibility
  (params (NewVisible :type Boolean))
  (bind _returnvalue (PartyDisplayManager|GetHeroPlayerPawnforCurrentDisplayedItem))
  (Utilities|IsValid _returnvalue
    (:Is%20Valid
      (for _array_element (Components|GetChildrenComponents (Variables|/Script/Engine.Character|GetMesh _returnvalue) true)
        (Utilities|Casting|CastToSkeletalMeshComponent _array_element)
        (Rendering|SetVisibility _array_element NewVisible false))
      (if NewVisible
        ;[0 1 0.04508] Prevents spamming multiple FX in Locker
        (Utilities|FlowControl|DoOnce false
          (:Completed
            (Utilities|FlowControl|Sequence
              (:then_0
                (Utilities|FlowControl|Delay 0.1))
              (:then_1
                _))
            (Effects|SpawnEmitteratLocation (Variables|VFX|GetCharSwapTransitionFX) (.location (Transformation|GetActorTransform _returnvalue)) (.rotation (Transformation|GetActorTransform _returnvalue)))
            (Variables|VFX|SetIsCharacterVisible true)))
        (elif (Variables|VFX|GetIsCharacterVisible)
          (Effects|SpawnEmitteratLocation "/Game/Effects/Fort_Effects/UI/Transition/P_FrontEndTransitionOut.P_FrontEndTransitionOut" (.location (Transformation|GetActorTransform _returnvalue)) (.rotation (Transformation|GetActorTransform _returnvalue)))
          (Variables|VFX|SetIsCharacterVisible false)))
      (Components|DestroyComponent (Variables|Default|GetTransitionFXComponent))
      (bind _returnvalue_2 (Effects|SpawnEmitteratLocation (Variables|VFX|GetCharSwapTransitionFXLoop) (Transformation|GetActorLocation)))
      (Variables|Default|SetTransitionFXComponent _returnvalue_2)
      (Utilities|FlowControl|Delay 1.0)
      (Utilities|FlowControl|WaitforBothExecutions true
        (:Then
          (Components|DestroyComponent (Variables|Default|GetTransitionFXComponent)))))))

(event Custom|OnEmoteSetupFinished
  (CallFunction|NotifyPrefabChanged (Variables|/Script/Engine.ChildActorComponent|GetChildActor (Variables|Default|GetPrefabActorComp)) (Variables|Default|GetItemToRepresent))
  (PartyDisplayManager|CallOnItemDisplayReady 0))

(event EventOnMeshLODStreamingFinished
  (Utilities|FlowControl|WaitforBothExecutions true
    (:Then
      (Components|DestroyComponent (Variables|Default|GetTransitionFXComponent)))))

(fn GetItemDefinitionToShow
  (bind _itemtorepresent (Variables|Default|GetItemToRepresent))
  (Utilities|IsValid _itemtorepresent
    (:Is%20Valid
      (return (select (Utilities|IsValid (Item|GetSchematicCraftingResult _itemtorepresent) :!class /Script/Engine.KismetSystemLibrary) (Item|GetSchematicCraftingResult _itemtorepresent) (Item|GetItemDefinition _itemtorepresent))))
    (:Is%20Not%20Valid
      (return 0))))

(fn GetItemPreviewOffset
  (params (ItemDefinition :type Fort%20Item%20Definition%20Object%20Reference))
  (return (Math|Transform|ComposeTransforms (Math|Transform|MakeTransform (.location (select (Math|Boolean|ANDBoolean (Math|Vector|Equal(Vector) (.location (Frontend|GetPreviewMeshOffset ItemDefinition)) "0, 0, 0" 0.0001) (Math|Rotator|Equal(Rotator) (.rotation (Frontend|GetPreviewMeshOffset ItemDefinition)) "0, 0, 0" 0.0001) (Math|Vector|Equal(Vector) (.scale (Frontend|GetPreviewMeshOffset ItemDefinition)) "1.000000,1.000000,1.000000" 0.0001)) (Math|Transform|MakeTransform "0.000000,-25.000000, 0" "0, 0,-40.000000") (Frontend|GetPreviewMeshOffset ItemDefinition)))) (Math|Transform|MakeTransform "0, 0, 0" (.rotation (select (Math|Boolean|ANDBoolean (Math|Vector|Equal(Vector) (.location (Frontend|GetPreviewMeshOffset ItemDefinition)) "0, 0, 0" 0.0001) (Math|Rotator|Equal(Rotator) (.rotation (Frontend|GetPreviewMeshOffset ItemDefinition)) "0, 0, 0" 0.0001) (Math|Vector|Equal(Vector) (.scale (Frontend|GetPreviewMeshOffset ItemDefinition)) "1.000000,1.000000,1.000000" 0.0001)) (Math|Transform|MakeTransform "0.000000,-25.000000, 0" "0, 0,-40.000000") (Frontend|GetPreviewMeshOffset ItemDefinition))) (.scale (select (Math|Boolean|ANDBoolean (Math|Vector|Equal(Vector) (.location (Frontend|GetPreviewMeshOffset ItemDefinition)) "0, 0, 0" 0.0001) (Math|Rotator|Equal(Rotator) (.rotation (Frontend|GetPreviewMeshOffset ItemDefinition)) "0, 0, 0" 0.0001) (Math|Vector|Equal(Vector) (.scale (Frontend|GetPreviewMeshOffset ItemDefinition)) "1.000000,1.000000,1.000000" 0.0001)) (Math|Transform|MakeTransform "0.000000,-25.000000, 0" "0, 0,-40.000000") (Frontend|GetPreviewMeshOffset ItemDefinition)))))))

(fn GetMeshForCurrentDisplayedItem
  (bind _skeletalmesh (Variables|Default|GetSkeletalMesh))
  (bind _staticmesh (Variables|Default|GetStaticMesh))
  (Utilities|IsValid (Components|SkeletalMesh|GetSkeletalMeshAsset _skeletalmesh)
    (:Is%20Valid
      (return _skeletalmesh))
    (:Is%20Not%20Valid
      (Utilities|IsValid (Variables|/Script/Engine.StaticMeshComponent|GetStaticMesh _staticmesh)
        (:Is%20Valid
          (return _staticmesh))
        (:Is%20Not%20Valid
          (bind _returnvalue (PartyDisplayManager|GetHeroPlayerPawnforCurrentDisplayedItem))
          (Utilities|IsValid _returnvalue
            (:Is%20Valid
              (return (Animation|GetCustomizationRootSkeletalMeshComponent _returnvalue)))
            (:Is%20Not%20Valid
              (bind _asfort_player_parachute (Utilities|Casting|CastToFortPlayerParachute (Variables|/Script/Engine.ChildActorComponent|GetChildActor (Variables|Default|GetPrefabActorComp)))
                (:then
                  (return (Class|FortPlayerParachute|GetParachuteRootMesh _asfort_player_parachute)))
                (:CastFailed
                  (return 0))))))))))

(fn GetPrefabActorForCurrentDisplayedItem
  (return (Variables|/Script/Engine.ChildActorComponent|GetChildActor (Variables|Default|GetPrefabActorComp))))

(fn GetPreviewMeshWorldTransform
  (bind _itemdefinition (CallFunction|GetItemDefinitiontoShow))
  (Utilities|IsValid _itemdefinition
    (:Is%20Valid
      (bind _returnvalue (CallFunction|GetWeaponPlacementTransform))
      (return (Math|Transform|ComposeTransforms (Math|Transform|ComposeTransforms (CallFunction|GetItemPreviewOffset _itemdefinition) (Math|Transform|MakeTransform "0, 0, 0" (Math|Rotator|MakeRotator 0.0 0.0 (select (Variables|Default|GetPreviewRotation) (Variables|Default|GetRotation) 180)))) _returnvalue)))
    (:Is%20Not%20Valid
      (return 0))))

(fn GetWeaponPlacementTransform
  (bind _vaultweaponplacementactor (Variables|Properties|GetVaultWeaponPlacementActor))
  (if (Utilities|IsValid _vaultweaponplacementactor :!class /Script/Engine.KismetSystemLibrary)
    (return (Transformation|GetActorTransform _vaultweaponplacementactor))
    (else
      (return (Math|Transform|ComposeTransforms (Transformation|GetActorTransform (Variables|Properties|GetVaultPlacementActor)) (Math|Transform|MakeTransform "0,-50.000000,65.000000"))))))

(fn HandlePartyMemberInProgressDisplayChanges
  (params (Pawn :type Fort%20Player%20Pawn%20Object%20Reference))
  (bind _hideelementsonlyconsidersredchannel (Variables|PartyinProgress|GetHideElementsOnlyConsidersRedChannel))
  (bind _hideelement_1 (Variables|PartyinProgress|GetHideElement_1))
  (bind _hideelement_2 (Variables|PartyinProgress|GetHideElement_2))
  (bind _hideelement_3 (Variables|PartyinProgress|GetHideElement_3))
  (bind _hideelement_4 (Variables|PartyinProgress|GetHideElement_4))
  (bind _hideelement_5 (Variables|PartyinProgress|GetHideElement_5))
  (bind _hideelement_6 (Variables|PartyinProgress|GetHideElement_6))
  (bind _hideelement_7 (Variables|PartyinProgress|GetHideElement_7))
  (bind _hideelement_8 (Variables|PartyinProgress|GetHideElement_8))
  (bind _hideelement_9 (Variables|PartyinProgress|GetHideElement_9))
  (bind _hideelement_10 (Variables|PartyinProgress|GetHideElement_10))
  (bind _playerinmatchholomaterial (Variables|MaterialFX|GetPlayerInMatchHoloMaterial))
  (bind _vertexrange (Variables|PartyinProgress|GetVertexRange))
  (for _array_element (Components|GetChildrenComponents (Variables|/Script/Engine.Character|GetMesh Pawn) true)
    ;[0 0 0] variant system can spawn particles, we dont want to include
    (Utilities|Casting|CastToParticleSystemComponent _array_element
      (:CastFailed
        (bind _asprimitive_component (Utilities|Casting|CastToPrimitiveComponent _array_element))
        (for _index (range (Rendering|Material|GetNumMaterials _asprimitive_component))
          (bind _asmaterial_instance_dynamic (Utilities|Casting|CastToMaterialInstanceDynamic (Rendering|Material|GetMaterial _asprimitive_component _index)))
          (Utilities|IsValid _playerinmatchholomaterial
            (:Is%20Valid
              ;[0 0 0] Ugly but necessary to swap over collapsed vertex material parameters - only sets necessary params
              (bind _returnvalue (Rendering|Material|CreateDynamicMaterialInstance _playerinmatchholomaterial "None" "None" :!class /Script/Engine.KismetMaterialLibrary))
              (Utilities|FlowControl|Sequence
                (:then_0
                  (bind _returnvalue_2 (Rendering|Material|GetScalarParameterValue _asmaterial_instance_dynamic _vertexrange :!class /Script/Engine.MaterialInstanceDynamic))
                  ;;[0 0 0] Vertex Hiding
                  (Rendering|Material|SetScalarParameterValue _returnvalue _vertexrange _returnvalue_2 :!class /Script/Engine.MaterialInstanceDynamic))
                (:then_1
                  (bind _returnvalue_3 (Rendering|Material|GetScalarParameterValue _asmaterial_instance_dynamic _hideelement_1 :!class /Script/Engine.MaterialInstanceDynamic))
                  (Rendering|Material|SetScalarParameterValue _returnvalue _hideelement_1 _returnvalue_3 :!class /Script/Engine.MaterialInstanceDynamic))
                (:then_2
                  (bind _returnvalue_4 (Rendering|Material|GetScalarParameterValue _asmaterial_instance_dynamic _hideelement_2 :!class /Script/Engine.MaterialInstanceDynamic))
                  (Rendering|Material|SetScalarParameterValue _returnvalue _hideelement_2 _returnvalue_4 :!class /Script/Engine.MaterialInstanceDynamic))
                (:then_3
                  (bind _returnvalue_5 (Rendering|Material|GetScalarParameterValue _asmaterial_instance_dynamic _hideelement_3 :!class /Script/Engine.MaterialInstanceDynamic))
                  (Rendering|Material|SetScalarParameterValue _returnvalue _hideelement_3 _returnvalue_5 :!class /Script/Engine.MaterialInstanceDynamic))
                (:then_4
                  (bind _returnvalue_6 (Rendering|Material|GetScalarParameterValue _asmaterial_instance_dynamic _hideelement_4 :!class /Script/Engine.MaterialInstanceDynamic))
                  (Rendering|Material|SetScalarParameterValue _returnvalue _hideelement_4 _returnvalue_6 :!class /Script/Engine.MaterialInstanceDynamic))
                (:then_5
                  (bind _returnvalue_7 (Rendering|Material|GetScalarParameterValue _asmaterial_instance_dynamic _hideelement_5 :!class /Script/Engine.MaterialInstanceDynamic))
                  (Rendering|Material|SetScalarParameterValue _returnvalue _hideelement_5 _returnvalue_7 :!class /Script/Engine.MaterialInstanceDynamic))
                (:then_6
                  (bind _returnvalue_8 (Rendering|Material|GetScalarParameterValue _asmaterial_instance_dynamic _hideelement_6 :!class /Script/Engine.MaterialInstanceDynamic))
                  (Rendering|Material|SetScalarParameterValue _returnvalue _hideelement_6 _returnvalue_8 :!class /Script/Engine.MaterialInstanceDynamic))
                (:then_7
                  (bind _returnvalue_9 (Rendering|Material|GetScalarParameterValue _asmaterial_instance_dynamic _hideelement_7 :!class /Script/Engine.MaterialInstanceDynamic))
                  (Rendering|Material|SetScalarParameterValue _returnvalue _hideelement_7 _returnvalue_9 :!class /Script/Engine.MaterialInstanceDynamic))
                (:then_8
                  (bind _returnvalue_10 (Rendering|Material|GetScalarParameterValue _asmaterial_instance_dynamic _hideelement_8 :!class /Script/Engine.MaterialInstanceDynamic))
                  (Rendering|Material|SetScalarParameterValue _returnvalue _hideelement_8 _returnvalue_10 :!class /Script/Engine.MaterialInstanceDynamic))
                (:then_9
                  (bind _returnvalue_11 (Rendering|Material|GetScalarParameterValue _asmaterial_instance_dynamic _hideelement_9 :!class /Script/Engine.MaterialInstanceDynamic))
                  (Rendering|Material|SetScalarParameterValue _returnvalue _hideelement_9 _returnvalue_11 :!class /Script/Engine.MaterialInstanceDynamic))
                (:then_10
                  (bind _returnvalue_12 (Rendering|Material|GetScalarParameterValue _asmaterial_instance_dynamic _hideelement_10 :!class /Script/Engine.MaterialInstanceDynamic))
                  (Rendering|Material|SetScalarParameterValue _returnvalue _hideelement_10 _returnvalue_12 :!class /Script/Engine.MaterialInstanceDynamic))
                (:then_11
                  (bind _returnvalue_13 (Rendering|Material|GetScalarParameterValue _asmaterial_instance_dynamic _hideelementsonlyconsidersredchannel :!class /Script/Engine.MaterialInstanceDynamic))
                  (Rendering|Material|SetScalarParameterValue _returnvalue _hideelementsonlyconsidersredchannel _returnvalue_13 :!class /Script/Engine.MaterialInstanceDynamic))
                (:then_12
                  ;;[0 0 0] textures
                  (bind _returnvalue_14 (Rendering|Material|GetTextureParameterValue _asmaterial_instance_dynamic "Normals" :!class /Script/Engine.MaterialInstanceDynamic))
                  (Rendering|Material|SetTextureParameterValue _returnvalue "Normals" _returnvalue_14))
                (:then_13
                  (bind _returnvalue_15 (Rendering|Material|GetTextureParameterValue _asmaterial_instance_dynamic "Diffuse" :!class /Script/Engine.MaterialInstanceDynamic))
                  (Rendering|Material|SetTextureParameterValue _returnvalue "Diffuse" _returnvalue_15))
                (:then_14
                  (Rendering|Material|SetMaterial _asprimitive_component _index _returnvalue))))
            (:Is%20Not%20Valid
              (Development|PrintString "HoloMaterial Property Not Set in Party Display Manager, Please Set to Valid Material ")))))))
  (Components|Timeline|PlayfromStart (Variables|PartyDisplayManagerBP|GetHoloMatColorPulse))
  (bind _asplayer_pawn_athena (Utilities|Casting|CastToPlayerPawn_Athena Pawn))
  (LobbyDisplay|SetPartyinProgress _asplayer_pawn_athena true))

(fn PartyMemberInProgressCheck
  (params (Pawn :type Fort%20Player%20Pawn%20Object%20Reference))
  (bind _playerindex (LocalVariables|GetPlayerIndex))
  (bind _returnvalue (CallFunction|GetPlayerIdfromPartyMemberIndex _playerindex))
  (bind (_outplayerindex _returnvalue_2) (CallFunction|GetPartyMemberIndexbyPlayerPawn Pawn))
  (bind _returnvalue_3 (BlueprintContext|GetFortPartyContext))
  (bind (_locationout _returnvalue_4) (PartyContext|GetPartyMemberLocation _returnvalue_3 _returnvalue))
  (bind _returnvalue_5 (Options|IsPartyinProgressEnabled))
  (if _returnvalue_5
    (if _returnvalue_2
      (LocalVariables|SetPlayerIndex _outplayerindex)
      (if _returnvalue_4
        (switch Utilities|FlowControl|Switch|SwitchonEFortPartyMemberLocation _locationout
          (:PreLobby
            _)
          (:ConnectingToLobby
            (CallFunction|HandlePartyMemberinProgressDisplayChanges (Variables|GetPawn)))
          (:Lobby
            _)
          (:JoiningGame
            _)
          (:ProcessingRejoin
            _)
          (:InGame
            _)
          (:WatchingReplay
            _)
          (:ReturningToFrontEnd
            _))
        (bind _asplayer_pawn_athena (Utilities|Casting|CastToPlayerPawn_Athena (Variables|GetPawn)))
        (LobbyDisplay|SetPartyinProgress _asplayer_pawn_athena false)))))

(fn PostSetupPrefabVisuals
  (CallFunction|SetCharacterPartsVisibility true))

(fn ResetPreScaleLocations
  (Transformation|SetRelativeLocation (Variables|Default|GetScalePivot) "0, 0, 0" false true)
  (Transformation|SetRelativeLocation (Variables|Default|GetStaticMesh) "0, 0, 0" false true)
  (Transformation|SetRelativeLocation (Variables|Default|GetSkeletalMesh) "0, 0, 0" false true)
  (Transformation|SetRelativeLocation (Variables|Default|GetPrefabActorComp) "0, 0, 0" false true))

(fn ResetVisuals
  (bind _ghost_sword_vfx (Variables|Default|GetGhost%20Sword%20VFX))
  (bind _staticmesh (Variables|Default|GetStaticMesh))
  (bind _ghost_sword_mesh (Variables|Default|GetGhost%20Sword%20Mesh))
  (bind _ghostpistolvfx (Variables|Default|GetGhostPistolVFX))
  (Utilities|FlowControl|Sequence
    (:then_0
      (Components|SkinnedMesh|SetSkinnedAssetandUpdate (Variables|Default|GetSkeletalMesh) 0 true))
    (:then_1
      (Components|StaticMesh|SetStaticMesh _staticmesh 0)
      (Rendering|Material|SetMaterial _staticmesh 0 0))
    (:then_2
      (ChildActorComponent|SetChildActorClass (Variables|Default|GetPrefabActorComp) 0)
      (CallFunction|NotifyPrefabChanged (Variables|/Script/Engine.ChildActorComponent|GetChildActor (Variables|Default|GetPrefabActorComp)) (Variables|Default|GetItemToRepresent)))
    (:then_3
      (if (not (Variables|Properties|GetIsPlayingCelebrateFX))
        (Transformation|SetRelativeTransform (Variables|Default|GetSkeletalMesh) (Math|Transform|MakeTransform "0, 0, 0") false true)))
    (:then_4
      (Utilities|FlowControl|Sequence
        (:then_0
          (Utilities|IsValid _ghostpistolvfx
            (:Is%20Valid
              (Components|DestroyComponent _ghostpistolvfx))))
        (:then_1
          (Utilities|IsValid _ghost_sword_vfx
            (:Is%20Valid
              (Components|DestroyComponent _ghost_sword_vfx))))
        (:then_2
          (Utilities|IsValid _ghost_sword_mesh
            (:Is%20Valid
              (Components|DestroyComponent _ghost_sword_mesh))))))
    (:then_5
      (Variables|Default|SetHasResetVisuals true))))

(fn RestoreTeleportMIDsInProgress
  (params (Pawn :type Fort%20Player%20Pawn%20Object%20Reference))
  (bind (_outplayerindex _returnvalue) (CallFunction|GetPartyMemberIndexbyPlayerPawn Pawn))
  (if _returnvalue
    (switch Utilities|FlowControl|Switch|SwitchonEFortPartyMemberLocation (PartyContext|GetPartyMemberLocation (BlueprintContext|GetFortPartyContext) (CallFunction|GetPlayerIdfromPartyMemberIndex _outplayerindex))
      (:PreLobby
        _)
      (:Lobby
        _))
    (bind _asplayer_pawn_athena (Utilities|Casting|CastToPlayerPawn_Athena (Variables|GetPawn)))
    (Utilities|IsValid (Variables|/Game/Athena/PlayerPawn_Athena.PlayerPawn_Athena_C|GetRezInFXActor _asplayer_pawn_athena)
      (:Is%20Valid
        (Class|DuplicateResOutMesh|RestoreinProgressRes (Variables|/Game/Athena/PlayerPawn_Athena.PlayerPawn_Athena_C|GetRezInFXActor _asplayer_pawn_athena)))
      (:Is%20Not%20Valid
        (Utilities|IsValid (Variables|/Game/Athena/PlayerPawn_Athena.PlayerPawn_Athena_C|GetRezOutFXActor _asplayer_pawn_athena)
          (:Is%20Valid
            (Class|DuplicateResOutMesh|RestoreinProgressRes (Variables|/Game/Athena/PlayerPawn_Athena.PlayerPawn_Athena_C|GetRezOutFXActor _asplayer_pawn_athena))))))))

(fn SetUIOverrideVisuals
  (params (Large%20Texture :type Texture%202D%20Object%20Reference) (Item :type Fort%20Item%20Definition%20Object%20Reference) (Rarity :type EFortRarity%20Enum :default "Common"))
  (bind _returnvalue (Item|GetPersistentName Item))
  (bind _uioverridemid (Variables|Default|GetUIoverrideMID))
  (bind _returnvalue_2 (Fortnite|GetRarityData Rarity))
  (bind _returnvalue_3 (Rendering|Material|CreateDynamicMaterialInstance (Variables|Default|GetStaticMesh) 0 0 "None" :!class /Script/Engine.PrimitiveComponent))
  (bind _output_get (Variables|Default|SetUIoverrideMID _returnvalue_3))
  (Rendering|Material|SetTextureParameterValue _output_get "WorkerPortrait" Large%20Texture)
  (Rendering|Material|SetVectorParameterValue _uioverridemid "RGBA0" (.Color3 (Utilities|Struct|BreakFortRarityItemData _returnvalue_2)) :!class /Script/Engine.MaterialInstanceDynamic)
  (Rendering|Material|SetVectorParameterValue _uioverridemid "RGBA1" (.Color4 (Utilities|Struct|BreakFortRarityItemData _returnvalue_2)) :!class /Script/Engine.MaterialInstanceDynamic)
  (if (Math|Conversions|ToBoolean(Integer) (Utilities|String|FindSubstring _returnvalue "constructor"))
    _
    (elif (Math|Conversions|ToBoolean(Integer) (Utilities|String|FindSubstring _returnvalue "ninja"))
      _
      (elif (Math|Conversions|ToBoolean(Integer) (Utilities|String|FindSubstring _returnvalue "soldier"))
        _
        (elif (Math|Conversions|ToBoolean(Integer) (Utilities|String|FindSubstring _returnvalue "outlander"))
          _
          (else
            (Rendering|Material|SetScalarParameterValue _uioverridemid "bFlipPortrait" 0.0 :!class /Script/Engine.MaterialInstanceDynamic))))))
  (Rendering|Material|SetScalarParameterValue _uioverridemid "bFlipPortrait" 1 :!class /Script/Engine.MaterialInstanceDynamic))

(fn SetupSkeletalMeshVisuals
  (params (NewMesh :type Skeletal%20Mesh%20Object%20Reference))
  (Transformation|SetWorldTransform (Variables|Default|GetPivotHolder) (Transformation|GetActorTransform) false true)
  (Components|SkinnedMesh|SetSkinnedAssetandUpdate (Variables|Default|GetSkeletalMesh) NewMesh true)
  (Class|FortItemFunctionLibrary|CustomizePreviewMesh (Variables|Default|GetSkeletalMesh)))

(fn SetupStaticMeshVisuals
  (params (NewMesh :type Static%20Mesh%20Object%20Reference))
  (Transformation|SetWorldTransform (Variables|Default|GetPivotHolder) (Transformation|GetActorTransform) false true)
  (Components|StaticMesh|SetStaticMesh (Variables|Default|GetStaticMesh) NewMesh)
  (Class|FortItemFunctionLibrary|CustomizePreviewMesh (Variables|Default|GetStaticMesh)))

(fn UpdatePreviewMeshTransforms
  (if (Variables|Default|GetPreviewRotation)
    (Variables|Default|SetRotation (Math|Float|%25%28Float%29 (+ (Variables|Default|GetRotation) 0.4) 360)))
  (Utilities|FlowControl|Sequence
    (:then_0
      (Transformation|SetWorldTransform (Variables|Default|GetPivotHolder) (CallFunction|GetPreviewMeshWorldTransform) false true))))

(fn ConstructionScript)
