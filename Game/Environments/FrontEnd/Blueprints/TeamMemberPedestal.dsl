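; Frontend lobby pedestal for one team member slot, presenting the player's pawn, pad and overhead prompts.
; Population, hover and squad events drive pad materials, lights and timeline fades through dynamic material instances, with squad colors and icons when several squads are populated.
; RefreshWidgets gates nameplate, social nudge, sign-in and take-control widgets by store state, camera and squad.
; It also applies in-match holograms, plays lobby animations and adjusts lighting per product.
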(event ClientEvent|HandleClientEvent:StoreTabClosed
  (params (EventSource :type Object%20Reference) (EventFocus :type Object%20Reference) (ClientEvent :type Fort%20Client%20Event%20Structure%20%28by%20ref%29))
  (CallFunction|RefreshWidgets))

(event ClientEvent|HandleClientEvent:StoreTabSelected
  (params (EventSource :type Object%20Reference) (EventFocus :type Object%20Reference) (ClientEvent :type Fort%20Client%20Event%20Structure%20%28by%20ref%29))
  (CallFunction|RefreshWidgets))

(event EventEndPlay
  (params (EndPlayReason :type EEndPlayReason%20Enum :default "Destroyed"))
  (CallFunction|UninitializeContextEvents))

(event EventBeginPlay
  (Utilities|FlowControl|Sequence
    (:then_0
      (CallFunction|SetDefaultWidgetComponentHeights)
      (Initialize|InitializeInputActionRefs)
      (CallFunction|HandleHoverInputActions (Variables|Default|GetbPlayerPodiumHovered))
      (CallFunction|RefreshWidgets)
      (Initialize|InitializeContextEvents)
      (Rendering|Components|Light|SetForceCachedShadowsforMovablePrimitives (Variables|Default|GetRimSpotLight) true)
      (bind _returnvalue (Rendering|Material|CreateDynamicMaterialInstance (Variables|Default|GetSM_Lobby_Character_Pad_Light_Ring) 0 (Rendering|Material|GetMaterial (Variables|Default|GetSM_Lobby_Character_Pad_Light_Ring) 0) "None" :!class /Script/Engine.PrimitiveComponent))
      (Variables|Materials|SetMID_Pad_Light_Ring _returnvalue)
      (bind _returnvalue_2 (Rendering|Material|CreateDynamicMaterialInstance (Variables|Default|GetSM_Lobby_Character_TopSpot) 0 (Rendering|Material|GetMaterial (Variables|Default|GetSM_Lobby_Character_TopSpot) 0) "None" :!class /Script/Engine.PrimitiveComponent))
      (Variables|Materials|SetMID_Top_Light_Ring _returnvalue_2)
      (if (CallFunction|IsinLocalSquadGroup)
        ;[0 0 0] Hardcoded offset for add player ui to address the last position + matchmaker ui
        (Transformation|SetRelativeLocation (Variables|FortTeamMemberPedestal|GetWidgetComponent_FillOpenSlot) (0.0 0.0 1.1e+02) false false)
        (bind _returnvalue_3 (Rendering|Material|CreateDynamicMaterialInstance (Variables|Default|GetSM_ObserverPlaceholder) 0 (Rendering|Material|GetMaterial (Variables|Default|GetSM_ObserverPlaceholder) (Rendering|Material|GetMaterialIndex (Variables|Default|GetSM_ObserverPlaceholder) "Screen")) "None" :!class /Script/Engine.PrimitiveComponent))
        (Variables|Materials|SetMID_ObserverScreen _returnvalue_3)
        (bind _returnvalue_4 (Rendering|Material|CreateDynamicMaterialInstance (Variables|Default|GetSM_ObserverPlaceholder) 1 0 "None" :!class /Script/Engine.PrimitiveComponent))
        (Variables|Materials|SetMID%20Observer%20Screen%20Frame _returnvalue_4)
        (CallFunction|TransitiontoLoadingScreen true)))
    (:then_1
      ;[1 1 1] Hide Pedestal if not local squad, it gets spawned when a player spawns
      (if (not (CallFunction|IsinLocalSquadGroup))
        (Rendering|SetVisibility (Variables|Default|GetSM_Lobby_Character_Pad) false false)
        (Rendering|Material|SetScalarParameterValue (Variables|Materials|GetMID_Pad) "Spawned" 0.0 :!class /Script/Engine.MaterialInstanceDynamic)))
    (:then_2
      ;[0 0.01298 0.1406] Initializing nameplate context menu variable
      (bind _asteam_member_nameplate_wrapper (Utilities|Casting|CastToTeamMemberNameplateWrapper (UserInterface|GetWidget (Variables|FortTeamMemberPedestal|GetWidgetComponent_MemberInfo))))
      (bind _output_get (Variables|Default|SetWBP_Nameplate (Variables|/Game/UI/Frontend/Lobby/Nameplates/TeamMemberNameplateWrapper.TeamMemberNameplateWrapper_C|GetWBP_Nameplate _asteam_member_nameplate_wrapper)))
      (Default|BindEventtoContextMenuClosedEventDispatcher _output_get (EventDispatchers|CreateEvent :!event Custom|On%20Context%20Menu%20Closed))
      (Default|BindEventtoOnMatchmakingStatusUpdatedEventDispatcher _output_get (EventDispatchers|CreateEvent :!event Custom|OnNameplateMatchmakingStatusUpdated)))
    (:then_3
      ;[0.1042 0 0.08067] Initializing social nudge variable
      (bind _asteam_member_social_nudge_wrapper (Utilities|Casting|CastToTeamMemberSocialNudgeWrapper (UserInterface|GetWidget (Variables|FortTeamMemberPedestal|GetWidgetComponent_FillOpenSlot))))
      (Variables|Default|SetWBP_SocialNudge (Variables|/Game/UI/Frontend/Lobby/SocialNudge/TeamMemberSocialNudgeWrapper.TeamMemberSocialNudgeWrapper_C|GetWBP_SocialNudge _asteam_member_social_nudge_wrapper))
      (Class|TeamMemberSocialNudgeWrapper|SetNudgePosition _asteam_member_social_nudge_wrapper (Variables|Default|GetLayerPositionInGroup) :!class /Game/UI/Frontend/Lobby/SocialNudge/TeamMemberSocialNudgeWrapper.TeamMemberSocialNudgeWrapper_C))))

(event EventOnTeamMemberSet
  (params (bHasValidTeamMember :type Boolean))
  ;[1 1 1] Show the scarecrow for team members that aren't playing FN
  (Rendering|SetVisibility (Variables|Default|GetSM_ObserverPlaceholder) false false)
  (CallFunction|RefreshWidgets)
  ;[0 0.000185 0.1615] Set Member Unique Id for Nameplate
  (Class|WBPNameplate|SetMemberUniqueId (Variables|Default|GetWBP_Nameplate) (CallFunction|GetCurrentTeamMemberUniqueId) :!class /Game/UI/Frontend/Lobby/Nameplates/WBP_Nameplate.WBP_Nameplate_C))

(event Custom|OnNameplateMatchmakingStatusUpdated
  (params (MatchmakingStatus :type EPedestalNameplateMatchmakingStatus%20Enum))
  ;[0 0.000185 0.1615] On Nameplate Matchmaking Status Updated
  (bind _returnvalue (CallFunction|SetNameplateMatchmakingStatus MatchmakingStatus))
  (if _returnvalue
    (Class|WBPNameplate|UpdateMatchmakingStatusTextDelegate (Variables|Default|GetWBP_Nameplate))))

(event EventOnDisconnectedToCampaignLobby
  (if (== (Variables|FortTeamMemberPedestal|GetAssignedMemberIndex) 0)
    (CallFunction|StopLobbyAnimation)
    (CallFunction|ClearPawnMaterialOverride)))

(event EventOnGameReadinessChanged
  (params (GameReadiness :type EGameReadiness%20Enum :default "NotReady"))
  ;[1 1 1] Hook up Event for Skip Match here
  (Rendering|Material|SetScalarParameterValue (Variables|Materials|GetMID_SkipMatchFX) "Visible" (Utilities|Select :Index GameReadiness :NotReady 0.0 :Ready 0.0 :Reevaluating 0.0 :SittingOut 1.0 :Away 1.0) :!class /Script/Engine.MaterialInstanceDynamic))

(event EventOnConnectedToCampaignLobby
  (if (== (Variables|FortTeamMemberPedestal|GetAssignedMemberIndex) 0)
    (CallFunction|PlayLobbyAnimation)))

(event EventOnPedestalIsPopulatedChanged
  (params (bIsPopulated :type Boolean))
  (bind _currentpawn (Variables|FortItemPlacementActor|GetCurrentPawn))
  (bind _mid_pad (Variables|Materials|GetMID_Pad))
  (bind _mid_pad_2 (Variables|Materials|GetMID_Pad))
  (Rendering|Material|SetScalarParameterValue (Variables|Materials|GetMID_Pad) "UseGroupPadColoration" (select (CallFunction|IsinLocalSquadGroup) 0.0 1.0) :!class /Script/Engine.MaterialInstanceDynamic)
  (CallFunction|RefreshWidgets)
  (Development|PrintString (Utilities|String|ToString(Boolean) bIsPopulated) true true "(R=0.000000,G=1.000000,B=0.012856,A=1.000000)" 5.0)
  (bind _output_get (Variables|Default|SetisPopulated bIsPopulated))
  (if _output_get
    (Rendering|SetVisibility (Variables|Default|GetSM_ObserverPlaceholder) false false)
    ;[1 1 1] Setup lights, materials for a player being on the pad
    (Utilities|FlowControl|Sequence
      (:then_0
        (CallFunction|SpawnPad)
        (Rendering|Material|SetScalarParameterValue _mid_pad "Occupied" 1 :!class /Script/Engine.MaterialInstanceDynamic)
        (Rendering|Material|SetScalarParameterValue _mid_pad "Selected" 0.0 :!class /Script/Engine.MaterialInstanceDynamic)
        (CallFunction|SetFakeBlobShadowVisiblity true)
        (if (CallFunction|IsinLocalSquadGroup)
          (Rendering|SetVisibility (Variables|Default|GetTopSpotLight) true false)))
      (:then_1
        (Rendering|SetVisibility (Variables|Default|GetPS-NoPlayer-Swirl02) false false)
        (if (and (CallFunction|IsinFrontend) (not (CallFunction|ShouldUsePreviewActor)))
          (bind _asplayer_pawn_athena (Utilities|Casting|CastToPlayerPawn_Athena (Variables|FortItemPlacementActor|GetCurrentPawn)))
          (AthenaPawn|OnResinDuringWarmup _asplayer_pawn_athena))))
    (else
      ;[1 0.8233 0 5] Reset pedestal visuals
      (Utilities|FlowControl|Sequence
        (:then_0
          (Rendering|SetVisibility (Variables|Default|GetUnderlightBlue01) false false)
          (Rendering|Material|SetScalarParameterValue _mid_pad_2 "Occupied" 0.0 :!class /Script/Engine.MaterialInstanceDynamic)
          (Rendering|Material|SetScalarParameterValue _mid_pad_2 "Selected" 0.0 :!class /Script/Engine.MaterialInstanceDynamic)
          (Components|Activation|Deactivate (Variables|Default|GetP_FE_Smoke))
          (Rendering|SetVisibility (Variables|Default|GetPS-NoPlayer-Swirl02) false false)
          (CallFunction|SetFakeBlobShadowVisiblity false)
          (bind _asplayer_pawn_athena_2 (Utilities|Casting|CastToPlayerPawn_Athena _currentpawn)
            (:then
              (GameplayCue|ExecuteGameplayCue%28Burst%29Local _asplayer_pawn_athena_2 (Variables|/Game/Athena/PlayerPawn_Athena.PlayerPawn_Athena_C|GetElimination%20Rez%20Sequence%20GC _asplayer_pawn_athena_2) (Ability|GameplayCue|MakeGameplayCueParameters 0.0)))
            (:CastFailed
              (if (not (CallFunction|ShouldUsePreviewActor))
                (Actor|DestroyActor _currentpawn)))))
        (:then_1
          (if (not (CallFunction|IsinLocalSquadGroup))
            (CallFunction|HidePad)
            (Rendering|Material|SetScalarParameterValue (Variables|Materials|GetMID_SkipMatchFX) "Visible" 0.0 :!class /Script/Engine.MaterialInstanceDynamic)))))))

(event Custom|SpawnPad
  ;[1 1 1] Hide and Spawn floating Pad
  (Rendering|SetVisibility (Variables|Default|GetSM_Lobby_Character_Pad) true false)
  (bind (_direction _newtrack_0) (Timeline|FadeInPodium Play)
    (:Update
      (Rendering|Material|SetScalarParameterValue (Variables|Materials|GetMID_Pad) "Spawned" _newtrack_0 :!class /Script/Engine.MaterialInstanceDynamic))
    (:Finished
      (if (Utilities|Select :Index _direction :Forward false :Backward true)
        (Rendering|SetVisibility (Variables|Default|GetSM_Lobby_Character_Pad) false false)))))

(event Custom|HidePad
  ;[1 1 1] Hide and Spawn floating Pad
  (Timeline|FadeInPodium Reverse))

(event EventOnAvatarChanged
  (params (AvatarImage :type Texture%202D%20Object%20Reference) (AvatarBackground :type Color%20Structure) (HighlightColor :type Color%20Structure))
  (bind _mid_observerscreen (Variables|Materials|GetMID_ObserverScreen))
  ;[1 1 1] Kairos Avatar scripting
  (Rendering|Material|SetTextureParameterValue _mid_observerscreen "SmallPreview" AvatarImage)
  (Rendering|Material|SetVectorParameterValue _mid_observerscreen "BackgroundRGB" (Math|Conversions|ToLinearColor(Color) AvatarBackground) :!class /Script/Engine.MaterialInstanceDynamic)
  (CallFunction|TransitiontoCharacterScreen true))

(event Custom|TransitionToCharacterScreen
  (params (bool :type Boolean))
  ;[1 1 1] Kairos Avatar scripting
  ;[1 1 1] Transitions for loading and chracter screen
  (if bool
    (bind (_direction _newtrack_0) (Timeline|Timeline_0 Play)
      (:Update
        (Rendering|Material|SetScalarParameterValue (Variables|Materials|GetMID_ObserverScreen) "CharacterScreen" _newtrack_0 :!class /Script/Engine.MaterialInstanceDynamic)))
    (else
      (Timeline|Timeline_0 Reverse))))

(event Custom|CallUpdateOnHovered
  ;[0 1 0] Hovered - On state
  (Rendering|Components|Light|SetIntensity (Variables|Default|GetTopSpotLight) (* (Variables|LightSettings|GetTopSpotLight_Brightness_Hovered) (Variables|LightSettings|GetLightBrightnessScale)))
  (bind _returnvalue (Rendering|Material|GetVectorParameterValue (Variables|Materials|GetMID_Pad) "GroupColor" :!class /Script/Engine.MaterialInstanceDynamic))
  (Rendering|Components|Light|SetLightColor (Variables|Default|GetUnderlightBlue01) (select (not (Variables|Default|GetMultiple%20Squads%20Populated)) "(R=0.000000,G=180.000000,B=255.000000,A=1.000000)" _returnvalue) true)
  (Rendering|Components|Light|SetIntensity (Variables|Default|GetUnderlightBlue02) (* (Variables|LightSettings|GetUnderlightBlueSpot_Brightness_Hovered) (Variables|LightSettings|GetLightBrightnessScale)))
  (Rendering|Components|Light|SetIntensity (Variables|Default|GetUnderlightBluePoint01) (* (Variables|LightSettings|GetUnderlightBluePoint_Brightness_Hovered) (Variables|LightSettings|GetLightBrightnessScale)))
  (Widget|SetVisibility (Variables|Default|GetHovered_InputAction) "SelfHitTestInvisible")
  (Class|WBPNameplate|OnPedestalHoveredChanged (Variables|Default|GetWBP_Nameplate) true))

(event EventOnHovered
  (bind _mid_pad (Variables|Materials|GetMID_Pad))
  ;[0 1 0] Hovered - On state
  (if (CallFunction|IsinLocalSquadGroup)
    (Variables|Default|SetbPlayerPodiumHovered true)
    (CallFunction|HandleHoverInputActions true)
    (Audio|PlayHoverSound)
    (Rendering|Material|SetScalarParameterValue _mid_pad "Occupied" 0.0 :!class /Script/Engine.MaterialInstanceDynamic)
    (Rendering|Material|SetScalarParameterValue _mid_pad "Selected" 1 :!class /Script/Engine.MaterialInstanceDynamic)
    (Rendering|SetVisibility (Variables|Default|GetUnderlightBlue01) true false)
    (Rendering|Components|Light|SetIntensity (Variables|Default|GetTopSpotLight) (* (Variables|LightSettings|GetTopSpotLight_Brightness_Hovered) (Variables|LightSettings|GetLightBrightnessScale)))
    (bind _returnvalue (Rendering|Material|GetVectorParameterValue (Variables|Materials|GetMID_Pad) "GroupColor" :!class /Script/Engine.MaterialInstanceDynamic))
    (Rendering|Components|Light|SetLightColor (Variables|Default|GetUnderlightBlue01) (select (not (Variables|Default|GetMultiple%20Squads%20Populated)) "(R=0.000000,G=180.000000,B=255.000000,A=1.000000)" _returnvalue) true)
    (Rendering|Components|Light|SetIntensity (Variables|Default|GetUnderlightBlue02) (* (Variables|LightSettings|GetUnderlightBlueSpot_Brightness_Hovered) (Variables|LightSettings|GetLightBrightnessScale)))
    (Rendering|Components|Light|SetIntensity (Variables|Default|GetUnderlightBluePoint01) (* (Variables|LightSettings|GetUnderlightBluePoint_Brightness_Hovered) (Variables|LightSettings|GetLightBrightnessScale)))
    (Widget|SetVisibility (Variables|Default|GetHovered_InputAction) "SelfHitTestInvisible")
    (Class|WBPNameplate|OnPedestalHoveredChanged (Variables|Default|GetWBP_Nameplate) true)))

(event Custom|TransitionToLoadingScreen
  (params (bool :type Boolean))
  ;[1 1 1] Kairos Avatar scripting
  ;[1 1 1] Transitions for loading and chracter screen
  (if bool
    (bind (_direction _newtrack_0) (Timeline|Timeline_1 Play)
      (:Update
        (Rendering|Material|SetScalarParameterValue (Variables|Materials|GetMID_ObserverScreen) "LoadingScreen" _newtrack_0 :!class /Script/Engine.MaterialInstanceDynamic)))
    (else
      (Timeline|Timeline_1 Reverse))))

(event Custom|TransitionfromCharactertoLoadingscreen
  (params (bool :type Boolean))
  (bind _mid_observerscreen (Variables|Materials|GetMID_ObserverScreen))
  ;[1 1 1] Kairos Avatar scripting
  ;[1 1 1] Transitions for loading and chracter screen
  (if bool
    (bind (_direction _newtrack_0 _newtrack_1) (Timeline|Timeline_2 Play)
      (:Update
        (Rendering|Material|SetScalarParameterValue _mid_observerscreen "CharacterScreen" _newtrack_0 :!class /Script/Engine.MaterialInstanceDynamic)
        (Rendering|Material|SetScalarParameterValue _mid_observerscreen "LoadingScreen" _newtrack_1 :!class /Script/Engine.MaterialInstanceDynamic)))
    (else
      (Timeline|Timeline_2 Reverse))))

(event EventOnItemClicked
  ;[0 0.02324 0.1771] Nameplate context menu open and close events
  (Audio|PlaySound2D "/Game/Sounds/Fort_UI_Sounds/FrontEnd/DefaultSounds/PressedButtons/MSS_UI_Button_Hover_Default.MSS_UI_Button_Hover_Default")
  (Class|WBPNameplate|OnPlayerCharacterClicked (Variables|Default|GetWBP_Nameplate)))

(event EventOnUnhovered
  (bind _mid_pad (Variables|Materials|GetMID_Pad))
  (bind _mid_pad_2 (Variables|Materials|GetMID_Pad))
  ;[0.12 0.12 0.12] Unhovered - Off State
  (Variables|Default|SetbPlayerPodiumHovered false)
  (CallFunction|HandleHoverInputActions false)
  (CallFunction|HoverAudioBuffer)
  (Widget|SetVisibility (Variables|Default|GetHovered_InputAction) "Hidden")
  (Rendering|SetVisibility (Variables|Default|GetSM_Lobby_Character_Pad_Light_Ring) false false)
  (Class|WBPNameplate|OnPedestalHoveredChanged (Variables|Default|GetWBP_Nameplate) false)
  (if (Utilities|IsValid (Variables|FortItemPlacementActor|GetCurrentPawn) :!class /Script/Engine.KismetSystemLibrary)
    ;;[1 1 1] Off state with player
    (Rendering|Material|SetScalarParameterValue _mid_pad_2 "Occupied" 1 :!class /Script/Engine.MaterialInstanceDynamic)
    (Rendering|Material|SetScalarParameterValue _mid_pad_2 "Selected" 0.0 :!class /Script/Engine.MaterialInstanceDynamic)
    (Rendering|SetVisibility (Variables|Default|GetTopSpotLight) true false)
    (Rendering|Components|Light|SetIntensity (Variables|Default|GetTopSpotLight) (* (Variables|LightSettings|GetTopSpotLight_Brightness) (Variables|LightSettings|GetLightBrightnessScale)))
    (Rendering|Components|Light|SetIntensity (Variables|Default|GetUnderlightBlue02) (* (Variables|LightSettings|GetUnderlightBlueSpot_Brightness) (Variables|LightSettings|GetLightBrightnessScale)))
    (Rendering|Components|Light|SetIntensity (Variables|Default|GetUnderlightBluePoint01) (* (Variables|LightSettings|GetUnderlightBluePoint_Brightness) (Variables|LightSettings|GetLightBrightnessScale)))
    (else
      ;;[1 1 1] Off state with no player
      (Rendering|Material|SetScalarParameterValue _mid_pad "Occupied" 0.0 :!class /Script/Engine.MaterialInstanceDynamic)
      (Rendering|Material|SetScalarParameterValue _mid_pad "Selected" 0.0 :!class /Script/Engine.MaterialInstanceDynamic)
      (Rendering|SetVisibility (Variables|Default|GetUnderlightBlue01) false false)
      (bind _wbp_socialnudge (Variables|Default|GetWBP_SocialNudge)
        (:then
          (Class|WBPSocialNudge|NavigateOut _wbp_socialnudge))
        (:else)))))

(event Custom|On%20Context%20Menu%20Closed
  ;[0 0.02324 0.1771] Nameplate context menu open and close events
  (CallFunction|HandleReturnfromNameplateContextMenu))

(event EventOnPartySuggestionChanged
  (params (SuggestedSocialUserId :type Unique%20Net%20Id%20Structure))
  (bind _wbp_socialnudge (Variables|Default|GetWBP_SocialNudge))
  (bind _fortteammemberpedestalsocialnudgeviewmodel (Variables|/Game/UI/Frontend/Lobby/SocialNudge/WBP_SocialNudge.WBP_SocialNudge_C|GetFortTeamMemberPedestalSocialNudgeViewModel _wbp_socialnudge))
  ;[0.03746 0 0.07227] Set member unique id on party suggestion changed event
  (Class|WBPSocialNudge|SetMemberUniqueId _wbp_socialnudge SuggestedSocialUserId :!class /Game/UI/Frontend/Lobby/SocialNudge/WBP_SocialNudge.WBP_SocialNudge_C)
  (Utilities|IsValid _fortteammemberpedestalsocialnudgeviewmodel
    (:Is%20Valid
      (Utilities|IsValid (Class|FortTeamMemberPedestalSocialNudgeViewModel|GetSocialUser _fortteammemberpedestalsocialnudgeviewmodel)
        (:Is%20Valid
          (Audio|PlaySound2D (Variables|Sound|GetPlayerSuggesstionSound)))))))

(event EventOnOwningPedestalGroupEstablished
  (params (SquadIdx :type Integer) (bSquadHasLocalPlayer :type Boolean) (bIsPopulated :type Boolean))
  (Utilities|FlowControl|Sequence
    (:then_0
      (if (not (CallFunction|IsinLocalSquadGroup))
        (Rendering|SetVisibility (Variables|Default|GetTopSpotLight) false false)
        (Rendering|SetVisibility (Variables|FortTeamMemberPedestal|GetWidgetComponent_MemberInfo) false false)))
    (:then_1
      (if (or bSquadHasLocalPlayer bIsPopulated)
        (CallFunction|SpawnPad)
        (else
          (CallFunction|HidePad)))
      (CallFunction|RefreshWidgets))))

(event EventOnNewSquadEstablished
  (params (bMultipleSquadsPopulated :type Boolean) (bOwningSquadContainsLocalPlayer :type Boolean) (OwningSquadStyle :type Athena%20Team%20Display%20Info%20Structure%20%28by%20ref%29))
  (bind (_icon _ribbon _name _textcolor _pedestalcolor) (Utilities|Struct|BreakAthenaTeamDisplayInfo OwningSquadStyle))
  (bind _output_get (Variables|Default|SetMultiple%20Squads%20Populated bMultipleSquadsPopulated))
  (Utilities|FlowControl|Sequence
    (:then_0
      (Rendering|Material|SetTextureParameterValue (Variables|Materials|GetMID_Pad) "IconTexture" "/Game/Global/Textures/Default/Blanks/T_Blank_White_ZeroAlpha_sRGBColor.T_Blank_White_ZeroAlpha_sRGBColor")
      (Rendering|Material|SetScalarParameterValue (Variables|Materials|GetMID_Pad) "UseGroupPadColoration" (select _output_get 1.0 0.0) :!class /Script/Engine.MaterialInstanceDynamic)
      (if (not _output_get)
        (Rendering|Material|SetVectorParameterValue (Variables|Materials|GetMID_Pad) "EmissiveColorX" "(R=0.000000,G=1.000000,B=1.000000,A=150.000000)" :!class /Script/Engine.MaterialInstanceDynamic)))
    (:then_1
      (if _output_get
        (bind _object (Utilities|AsyncLoadAsset _icon)
          (:Completed
            (bind _astexture (Utilities|Casting|CastToTexture _object))
            (Rendering|Material|SetTextureParameterValue (Variables|Materials|GetMID_Pad) "IconTexture" _astexture)))))
    (:then_2
      (Rendering|Material|SetVectorParameterValue (Variables|Materials|GetMID_Pad) "GroupColor" (select _output_get _pedestalcolor "(R=0.000000,G=0.301101,B=0.776042,A=1.000000)") :!class /Script/Engine.MaterialInstanceDynamic))
    (:then_3
      (CallFunction|RefreshWidgets)
      (bind _outactors (Actor|GetAllActorsOfClass "/Game/Environments/FrontEnd/Blueprints/CameraAnimationTransition.CameraAnimationTransition_C"))
      (for _array_element _outactors
        (if _output_get
          (Class|CameraAnimationTransition|MoveCameraUp _array_element)
          (else
            (Class|CameraAnimationTransition|MoveCameraDown _array_element)))))))

(event EventOnHeroPawnSetupCompleted
  (if (Math|Boolean|ANDBoolean (CallFunction|IsinFrontend) (Utilities|Enum|Equal(Enum) (Fortnite|GetSubGame) "Campaign") (== (Variables|FortTeamMemberPedestal|GetAssignedMemberIndex) 0))
    (CallFunction|PlayLobbyAnimation)))

(event EventOnIsCurrentlyInMatchChanged
  (params (bIsInMatch :type Boolean))
  (CallFunction|EnableHologram bIsInMatch))

(event EventOnPreviewPawnSpawned
  (bind _returnvalue (CallFunction|IsinFrontend))
  (CallFunction|EnableHologram (not _returnvalue))
  (if _returnvalue
    (Audio|PlaySound2D "/Game/Sounds/Fort_UI_Sounds/S13_Frontend/BattlePass_ResIn_Cue.BattlePass_ResIn_Cue")))

(event EventOnPreviewPawnShown
  ; Old Preview Flow Only
  (if (not (CallFunction|ShouldUsePreviewActor))
    (CallFunction|OutroLoadingFX)))

(event EventOnSetNameplateHeight
  (params (Height :type Float%20%28single-precision%29))
  ;[1 1 1] Support overriding the default name plate and suggestion widget heights. Using 1 as a threshold for sanity and floating point precision errors.
  (Transformation|SetRelativeLocation (Variables|FortTeamMemberPedestal|GetWidgetComponent_MemberInfo) (Math|Vector|MakeVector 0.0 0.0 (Math|Float|SelectFloat Height (Variables|Default|GetDefaultNameplateHeight) (> Height 1.0))) false false))

(event EventOnSetSuggestionHeight
  (params (Height :type Float%20%28single-precision%29))
  ;[1 1 1] Support overriding the default name plate and suggestion widget heights. Using 1 as a threshold for sanity and floating point precision errors.
  (Transformation|SetRelativeLocation (Variables|FortTeamMemberPedestal|GetWidgetComponent_FillOpenSlot) (Math|Vector|MakeVector 0.0 0.0 (Math|Float|SelectFloat Height (Variables|Default|GetDefaultSuggestionHeight) (> Height 1.0))) false false))

(event Custom|HoverAudioBuffer
  ;[0.1073 0.2812 1] Prevents Hover from Firing if Unhover was clicked this same frame
  (Variables|Audio|SetisHoverSoundDisabled true)
  (Utilities|FlowControl|DelayUntilNextTick)
  (Variables|Audio|SetisHoverSoundDisabled false))

(event EventOnSignInPromptVisibilityChanged
  (params (bVisible :type Boolean) (BindingHandle :type UI%20Action%20Binding%20Handle%20Structure))
  (if (!= bVisible (Variables|Default|GetSignInPromptVisible))
    (Variables|Default|SetSignInPromptVisible bVisible)
    (CallFunction|RefreshWidgets)
    (bind _aswbp_splitscreen_prompt_join (Utilities|Casting|CastToWBP_Splitscreen_Prompt_Join (UserInterface|GetWidget (Variables|FortTeamMemberPedestal|GetWidgetComponent_SignIn))))
    (Class|WBPSplitscreenPromptJoin|SetBindingHandle _aswbp_splitscreen_prompt_join BindingHandle)))

(event EventOnTakeControlPromptVisibilityChanged
  (params (bVisible :type Boolean) (BindingHandle :type UI%20Action%20Binding%20Handle%20Structure))
  (if (!= bVisible (Variables|Default|GetTakeControlPromptVisible))
    (Variables|Default|SetTakeControlPromptVisible bVisible)
    (CallFunction|RefreshWidgets)
    (bind _asfort_pedestal_take_control_widget (Utilities|Casting|CastToFortPedestalTakeControlWidget (UserInterface|GetWidget (Variables|FortTeamMemberPedestal|GetWidgetComponent_TakeControl))))
    (ActionWidget|SetActionBinding _asfort_pedestal_take_control_widget BindingHandle)))

(event EventOnFrontendCameraChanged
  (params (NewCamera :type EFrontEndCamera%20Enum :default "Invalid") (OldCamera :type EFrontEndCamera%20Enum :default "Invalid"))
  (CallFunction|RefreshWidgets)
  (if (and (or (Utilities|Enum|Equal(Enum) NewCamera "HeroSelect") (Utilities|Enum|Equal(Enum) OldCamera "LobbyCentered")) (Utilities|NotEqual(Object) (Variables|FortItemPlacementActor|GetCurrentPawn) (Variables|Default|GetLastAnimatedPlayerPawn)))
    (CallFunction|PlayLobbyAnimation)))

(event EventOnOwningPlayerChanged
  (params (LocalPlayer :type Local%20Player%20Object%20Reference))
  (CallFunction|RefreshWidgets))

(fn ApplyHoloMaterialOverrideToPawn
  (Utilities|IsValid (Variables|FortItemPlacementActor|GetCurrentPawn)
    (:Is%20Valid
      (Class|FortPlayerPawn|RemoveMaterialOverrideLocal (Variables|FortItemPlacementActor|GetCurrentPawn) (Variables|Default|GetMaterialOverride))
      (bind _returnvalue (Class|FortPlayerPawn|ApplyMaterialOverrideLocalwithFlags (Variables|FortItemPlacementActor|GetCurrentPawn) (Utilities|ResolveSoftReference (Variables|Materials|GetHoloMaterialInst)) (Variables|Default|GetMaterial%20Params%20to%20Copy) 1.0 15))
      (Variables|Default|SetMaterialOverride _returnvalue)
      (bind _asplayer_pawn_athena_generic (Utilities|Casting|CastToPlayerPawn_Athena_Generic (Variables|FortItemPlacementActor|GetCurrentPawn)))
      (Variables|/Script/FortniteGame.FortPlayerPawn|SetbIsInFrontEndHologram true _asplayer_pawn_athena_generic))))

(fn ClearPawnMaterialOverride
  (Utilities|IsValid (Variables|FortItemPlacementActor|GetCurrentPawn)
    (:Is%20Valid
      (Class|FortPlayerPawn|RemoveMaterialOverrideLocal (Variables|FortItemPlacementActor|GetCurrentPawn) (Variables|Default|GetMaterialOverride))
      (bind _asplayer_pawn_athena_generic (Utilities|Casting|CastToPlayerPawn_Athena_Generic (Variables|FortItemPlacementActor|GetCurrentPawn)))
      (Variables|/Script/FortniteGame.FortPlayerPawn|SetbIsInFrontEndHologram false _asplayer_pawn_athena_generic))))

(fn EnableHologram
  (params (Enabled :type Boolean))
  (if Enabled
    (CallFunction|ApplyHoloMaterialOverridetoPawn)
    (else
      (CallFunction|ClearPawnMaterialOverride))))

(fn HandleHoverInputActions
  (params (bIsHovered :type Boolean))
  (bind _currentpawn (Variables|FortItemPlacementActor|GetCurrentPawn))
  (bind _returnvalue (Utilities|IsValid _currentpawn :!class /Script/Engine.KismetSystemLibrary))
  (CallFunction|UpdateAddFriendButtonInteractionEnabled bIsHovered)
  (if bIsHovered
    (Widget|SetVisibility (Variables|Default|GetHovered_WithPlayer_InputAction) (select _returnvalue "SelfHitTestInvisible" "Hidden"))
    (Widget|SetVisibility (Variables|Default|GetHovered_InputAction) (select _returnvalue "Hidden" "SelfHitTestInvisible"))
    (Widget|SetVisibility (Variables|Default|GetHovered_WithPlayer_InputAction_AddFriend) (select (and _returnvalue (not (CallFunction|IsCurrentMemberLocal))) "HitTestInvisible" "Hidden"))
    (else
      (Widget|SetVisibility (Variables|Default|GetHovered_InputAction) "Hidden"))))

(fn HandleNavigationDown
  (bind _wbp_socialnudge (Variables|Default|GetWBP_SocialNudge))
  (bind _hassuggestedsocialuser (CallFunction|HasSuggestedSocialUser))
  (if (and _hassuggestedsocialuser (Variables|/Script/FortniteGame.FortRuntimeOptions|GetbUseNewNameplatesAndSocialNudges (Options|GetRuntimeOptions)))
    (if (Utilities|Enum|Equal(Enum) (Variables|/Game/UI/Frontend/Lobby/SocialNudge/WBP_SocialNudge.WBP_SocialNudge_C|GetCurrentFocusedWidget _wbp_socialnudge) "NewEnumerator2")
      (return false)
      (else
        (Default|NavigationSocialNudgeandPlusButtons _wbp_socialnudge "Down")
        (return true)))
    (else
      (return false))))

(fn HandleNavigationUp
  (bind _wbp_socialnudge (Variables|Default|GetWBP_SocialNudge))
  (bind _hassuggestedsocialuser (CallFunction|HasSuggestedSocialUser))
  (if (and _hassuggestedsocialuser (Variables|/Script/FortniteGame.FortRuntimeOptions|GetbUseNewNameplatesAndSocialNudges (Options|GetRuntimeOptions)))
    (if (Utilities|Enum|Equal(Enum) (Variables|/Game/UI/Frontend/Lobby/SocialNudge/WBP_SocialNudge.WBP_SocialNudge_C|GetCurrentFocusedWidget _wbp_socialnudge) "NewEnumerator1")
      (return false)
      (else
        (Default|NavigationSocialNudgeandPlusButtons _wbp_socialnudge "Up")
        (return true)))
    (else
      (return false))))

(fn HandleProductLightingSettings
  (params (Selection :type Gameplay%20Tag%20Structure))
  (switch Utilities|FlowControl|Switch|SwitchonGameplayTag Selection
    (:Default
      _)
    (:Product.BR
      _)
    (:Product.Juno
      (CallFunction|UpdateLightValues 0.07 1.5 0.15)))
  (CallFunction|UpdateLightValues 0.1 0.9 1.0))

(fn HasSuggestedSocialUser
  (bind _returnvalue (UniqueNetID|IsValidNetID (CallFunction|GetSuggestedSocialUserId)))
  (return _returnvalue))

(fn InitializeContextEvents
  (UITeam|BindEventtoOnTeamMemberStateChanged (PartyContext|GetLocalPlayerTeam (BlueprintContext|GetFortPartyContext)) (EventDispatchers|CreateEvent :!event OnTeamMemberStateChanged))
  (PartyContext|BindEventtoOnPlayerStateChanged (BlueprintContext|GetFortPartyContext) (EventDispatchers|CreateEvent :!event OnPartyDataChanged))
  (Class|FortIslandUtils|BindonCurrentContextProductChanged (EventDispatchers|CreateEvent :!event OnCurrentContextProductChanged)))

(fn InitializeInputActionRefs
  (bind _aswbp_lobby_add_player_bubble (Utilities|Casting|CastToWBP_Lobby_AddPlayerBubble (UserInterface|GetUserWidgetObject (Variables|FortTeamMemberPedestal|GetWidgetComponent_FillOpenSlot))))
  (Variables|Default|SetHovered_InputAction (Variables|/Game/UI/Frontend/Lobby/PlayerPad/WBP_Lobby_AddPlayerBubble.WBP_Lobby_AddPlayerBubble_C|GetOverlayInputAction _aswbp_lobby_add_player_bubble))
  (bind _asteam_member_pedestal_nameplate (Utilities|Casting|CastToTeamMemberPedestal_Nameplate (UserInterface|GetUserWidgetObject (Variables|FortTeamMemberPedestal|GetWidgetComponent_MemberInfo))))
  (Variables|Default|SetHovered_WithPlayer_InputAction (Variables|/Game/UI/Frontend/Lobby/PlayerPad/TeamMemberPedestal_Nameplate.TeamMemberPedestal_Nameplate_C|GetSB_InputAction _asteam_member_pedestal_nameplate))
  (bind _asteam_member_pedestal_nameplate_add_friend_button (Utilities|Casting|CastToTeamMemberPedestal_Nameplate_AddFriendButton (Variables|/Script/FortniteUI.FortTeamMemberPedestalNameplate|GetButton_AddFriend _asteam_member_pedestal_nameplate)))
  (Variables|Default|SetHovered_WithPlayer_InputAction_AddFriend (Variables|/Game/UI/Frontend/Lobby/PlayerPad/TeamMemberPedestal_Nameplate_AddFriendButton.TeamMemberPedestal_Nameplate_AddFriendButton_C|GetSB_InputAction _asteam_member_pedestal_nameplate_add_friend_button))
  (CallFunction|InitializePlaceholderMesh (Variables|Default|GetSM_ObserverPlaceholder)))

(fn InitilizeLightValuesFromContext
  (CallFunction|HandleProductLightingSettings (Class|FortIslandUtils|GetCurrentContextProductTag)))

(fn OnContextMenuClosed
  (CallFunction|HandleReturnfromNameplateContextMenu))

(fn OnCurrentContextProductChanged
  ;[1 1 1] TODO: Handle product change in a non hardcoded way
  (CallFunction|HandleProductLightingSettings (Class|FortProductConfig|GetCurrentCosmeticProductTag))
  (if (Variables|FortItemPlacementActor|GetbIsActorHovered)
    (CallFunction|CallUpdateonHovered)))

(fn OnPartyDataChanged
  (params (Member%20Info :type Fort%20Team%20Member%20Info%20Structure%20%28by%20ref%29))
  (CallFunction|RefreshWidgets))

(fn OnTeamMemberStateChanged
  (params (Team%20Member%20Info :type Fort%20Team%20Member%20Info%20Structure))
  (bind (_memberuniqueid _partyleaderuniqueid _consoleuniqueid _platform _playername _bpartyleader _bisinzone _bhasboostxp _bhasrestxp _bannericonid _bannercolorid _bbattlepasspurchased _battlepasslevel _numplayersinparty _playerindex) (Utilities|Struct|BreakFortTeamMemberInfo Team%20Member%20Info))
  (if (== (Variables|FortTeamMemberPedestal|GetAssignedMemberIndex) _playerindex)
    (Variables|Default|SetTeamMemberInfo Team%20Member%20Info)
    (CallFunction|RefreshWidgets)))

(fn PlayHoverSound
  (if (not (Variables|Audio|GetisHoverSoundDisabled))
    (Audio|PlaySound2D "/Game/Sounds/Fort_UI_Sounds/FrontEnd/DefaultSounds/PressedButtons/MSS_UI_Button_Hover_Default.MSS_UI_Button_Hover_Default")))

(fn PlayLobbyAnimation
  (bind _output_get (Variables|Default|SetLastAnimatedPlayerPawn (Variables|FortItemPlacementActor|GetCurrentPawn)))
  (Utilities|IsValid _output_get
    (:Is%20Valid
      (if (FortGlobals|IsinAthena)
        (Animation|PlayAnimationsMatchingQuery(Message) _output_get (GameplayTags|MakeGameplayTagQuery "(TokenStreamVersion=0,TagDictionary=((TagName=\"Animation.FrontEnd.Lobby\"),(TagName=\"Homebase.Class.IsCommando.Athena\")),QueryTokenStream=(0,1,2,2,0,1),UserDescription=\"\",AutoDescription=\" ALL( Animation.FrontEnd.Lobby, Homebase.Class.IsCommando.Athena )\")"))
        (else
          (Animation|PlayAnimationsMatchingQuery(Message) _output_get (GameplayTags|MakeGameplayTagQuery "(TokenStreamVersion=0,TagDictionary=((TagName=\"Animation.FrontEnd.Lobby\")),QueryTokenStream=(0,1,2,1,0),UserDescription=\"\",AutoDescription=\" ALL( Animation.FrontEnd.Lobby )\")")))))))

(fn RefreshWidgetComponentsVisibility
  (Utilities|FlowControl|Sequence
    (:then_0
      (Rendering|SetVisibility (Variables|FortTeamMemberPedestal|GetWidgetComponent_MemberInfo) (CallFunction|IsValidTeamMember) false))
    (:then_1
      (Rendering|SetVisibility (Variables|FortTeamMemberPedestal|GetWidgetComponent_TakeControl) (and (CallFunction|IsValidTeamMember) (Variables|Default|GetTakeControlPromptVisible)) false))
    (:then_2
      (Rendering|SetVisibility (Variables|FortTeamMemberPedestal|GetWidgetComponent_FillOpenSlot) (and (not (CallFunction|IsValidTeamMember)) (not (Variables|Default|GetSignInPromptVisible))) false))
    (:then_3
      (Rendering|SetVisibility (Variables|FortTeamMemberPedestal|GetWidgetComponent_SignIn) (Variables|Default|GetSignInPromptVisible) false))
    (:then_4
      (Refresh|Refresh (Variables|Default|GetGadgets) (Variables|Default|GetTeamMemberInfo))
      (Widget|SetVisibility (Variables|Default|GetGadgets) (select (Math|Boolean|ANDBoolean (CallFunction|IsCurrentMemberLocal) (MatchmakingContext|IsinLobby (BlueprintContext|GetFortMatchmakingContext)) (CallFunction|IsValidTeamMember)) "SelfHitTestInvisible" "Hidden")))))

(fn RefreshWidgetComponentsVisibility_STW
  (Utilities|FlowControl|Sequence
    (:then_0
      (Rendering|SetVisibility (Variables|FortTeamMemberPedestal|GetWidgetComponent_MemberInfo) (and (CallFunction|IsValidTeamMember) (Variables|FortTeamMemberPedestal|GetbShowWidgetComponents)) false))
    (:then_1
      (Rendering|SetVisibility (Variables|FortTeamMemberPedestal|GetWidgetComponent_TakeControl) (Math|Boolean|ANDBoolean (CallFunction|IsValidTeamMember) (Variables|Default|GetTakeControlPromptVisible) (Variables|FortTeamMemberPedestal|GetbShowWidgetComponents)) false))
    (:then_2
      (Rendering|SetVisibility (Variables|FortTeamMemberPedestal|GetWidgetComponent_FillOpenSlot) (and (and (not (CallFunction|IsValidTeamMember)) (not (Variables|Default|GetSignInPromptVisible))) (Variables|FortTeamMemberPedestal|GetbShowWidgetComponents)) false))
    (:then_3
      (Rendering|SetVisibility (Variables|FortTeamMemberPedestal|GetWidgetComponent_SignIn) (and (Variables|Default|GetSignInPromptVisible) (Variables|FortTeamMemberPedestal|GetbShowWidgetComponents)) false))
    (:then_4
      (Refresh|Refresh (Variables|Default|GetGadgets) (Variables|Default|GetTeamMemberInfo))
      (Widget|SetVisibility (Variables|Default|GetGadgets) (select (Math|Boolean|ANDBoolean (CallFunction|IsCurrentMemberLocal) (MatchmakingContext|IsinLobby (BlueprintContext|GetFortMatchmakingContext)) (CallFunction|IsValidTeamMember) (Variables|FortTeamMemberPedestal|GetbShowWidgetComponents)) "SelfHitTestInvisible" "Hidden")))))

(fn RefreshWidgets
  (Widget|SetVisibility (Variables|Default|GetGadgets) "Collapsed")
  (bind _returnvalue (CallFunction|GetFrontendCamera))
  (if (Math|Boolean|ANDBoolean (Utilities|Enum|Equal(Enum) (StoreContext|GetStoreState (BlueprintContext|GetFortStoreContext)) "Closed") (CallFunction|IsinLocalSquadGroup) (Math|Boolean|ORBoolean (Utilities|Enum|Equal(Enum) _returnvalue "HeroSelect") (Utilities|Enum|Equal(Enum) _returnvalue "LobbyCentered") (Utilities|Enum|Equal(Enum) _returnvalue "MultipleSquads") (Utilities|Enum|Equal(Enum) _returnvalue "LargeSquad")))
    ; Refresh widget component visibility based on the lobby widget activations. This is to prevent problems with seeing the widget components trough the walls in STW.
    (if (Class|BlueprintLibrarySTW|IsinHestiaBeauty)
      (CallFunction|RefreshWidgetComponentsVisibilitySTW)
      (else
        (CallFunction|RefreshWidgetComponentsVisibility)))
    (else
      (Rendering|SetVisibility (Variables|FortTeamMemberPedestal|GetWidgetComponent_MemberInfo) false false))))

(fn RefreshWidgetsVisibility
  (if (Class|BlueprintLibrarySTW|IsinHestiaBeauty)
    (CallFunction|RefreshWidgets))
  (return false))

(fn SetDefaultWidgetComponentHeights
  (bind (_relativelocation_x _relativelocation_y _relativelocation_z) (Math|Vector|BreakVector (Variables|/Script/Engine.SceneComponent|GetRelativeLocation_X (Variables|FortTeamMemberPedestal|GetWidgetComponent_FillOpenSlot))))
  (bind (_relativelocation_x_2 _relativelocation_y_2 _relativelocation_z_2) (Math|Vector|BreakVector (Variables|/Script/Engine.SceneComponent|GetRelativeLocation_X (Variables|FortTeamMemberPedestal|GetWidgetComponent_MemberInfo))))
  (Variables|Default|SetDefaultNameplateHeight _relativelocation_z_2)
  (Variables|Default|SetDefaultSuggestionHeight _relativelocation_z))

(fn SetFakeBlobShadowVisiblity
  (params (Visible :type Boolean))
  (bind _returnvalue (TouchPlatform|IsNintendoSwitch))
  (bind _returnvalue_2 (Fortnite|IsMobilePlatform))
  (bind _returnvalue_3 (Settings|GetGameUserSettings))
  (Rendering|SetVisibility (Variables|Default|GetSM_Character_ShadowBlob) (and (Math|Boolean|ORBoolean _returnvalue_2 _returnvalue (<= (Settings|GetVisualEffectQuality _returnvalue_3) 1)) Visible) false)
  (Rendering|Material|SetScalarParameterValue (Variables|Materials|GetMID_FakeBlobShadow) "ShowShadows" (Math|Float|SelectFloat 1.0 0.0 (and (Math|Boolean|ORBoolean _returnvalue_2 _returnvalue (<= (Settings|GetVisualEffectQuality _returnvalue_3) 1)) Visible)) :!class /Script/Engine.MaterialInstanceDynamic))

(fn SetFNCrew
  (bind _mid_pad (Variables|Materials|GetMID_Pad))
  (Rendering|Material|SetVectorParameterValue _mid_pad "EmissiveColorX" (LocalVariables|GetCrewColor1) :!class /Script/Engine.MaterialInstanceDynamic)
  (Rendering|Material|SetScalarParameterValue _mid_pad "UseGroupPadColoration" 0.0 :!class /Script/Engine.MaterialInstanceDynamic)
  (Rendering|Material|SetVectorParameterValue (Variables|Materials|GetMID_Pad_Light_Ring) "LightEmissiveColor_1" (LocalVariables|GetCrewColor2) :!class /Script/Engine.MaterialInstanceDynamic))

(fn SetSparkleParticlesEnabled
  (params (IsEnabled :type Boolean))
  (Development|SetHiddeninGame (Variables|Default|GetPS-NoPlayer-Sparkle) (not IsEnabled) false))

(fn StopLobbyAnimation
  (bind _output_get (Variables|Default|SetLastAnimatedPlayerPawn (Variables|FortItemPlacementActor|GetCurrentPawn)))
  (Utilities|IsValid _output_get
    (:Is%20Valid
      (Utilities|IsValid (Variables|/Script/Engine.Character|GetMesh _output_get)
        (:Is%20Valid
          (if (not (FortGlobals|IsinAthena))
            (Animation|Montage|MontageStop (Components|SkeletalMesh|GetAnimInstance (Variables|/Script/Engine.Character|GetMesh _output_get)) 0.2 0)))))))

(fn UninitializeContextEvents
  (Class|FortIslandUtils|UnbindonCurrentContextProductChanged))

(fn UpdateAddFriendButtonInteractionEnabled
  (params (IsHovered :type Boolean))
  (bind _returnvalue (CallFunction|IsCurrentMemberLocal))
  (bind _returnvalue_2 (not _returnvalue))
  (bind _returnvalue_3 (and IsHovered _returnvalue_2))
  ;[1 1 0] Get a reference to the Nameplate widget
  (bind _asteam_member_pedestal_nameplate (Utilities|Casting|CastToTeamMemberPedestal_Nameplate (UserInterface|GetUserWidgetObject (Variables|FortTeamMemberPedestal|GetWidgetComponent_MemberInfo))))
  (if (GlobalUIContext|IsUsingGamepad (BlueprintContext|GetFortGlobalUIContext))
    ;[1 1 0 1 12] Toggling IsInteractionEnabled allows us to control which button is currently active so that we can determine which one we are pressing. If we are using MnK, we want all of the buttons always interactable (we will hide the ones that should not be interactable), so that the player can click them without hovering over this pedestal first.
    (CommonButton|Setters|SetIsInteractionEnabled (Variables|/Script/FortniteUI.FortTeamMemberPedestalNameplate|GetButton_AddFriend _asteam_member_pedestal_nameplate) _returnvalue_3)
    (if _returnvalue_3
      ;[1 1 0 1 12] Activate the button's parent widget, the activable container, so the button is enabled properly.
      (ActivatableWidget|ActivateWidget (Variables|/Game/UI/Frontend/Lobby/PlayerPad/TeamMemberPedestal_Nameplate.TeamMemberPedestal_Nameplate_C|GetTeamMemberPedestal_ActivatableContainer _asteam_member_pedestal_nameplate))
      (else
        (ActivatableWidget|DeactivateWidget (Variables|/Game/UI/Frontend/Lobby/PlayerPad/TeamMemberPedestal_Nameplate.TeamMemberPedestal_Nameplate_C|GetTeamMemberPedestal_ActivatableContainer _asteam_member_pedestal_nameplate))))
    (else
      (CommonButton|Setters|SetIsInteractionEnabled (Variables|/Script/FortniteUI.FortTeamMemberPedestalNameplate|GetButton_AddFriend _asteam_member_pedestal_nameplate) true))))

(fn UpdateLightValues
  (params (PadLightRingOpacityScale :type Float%20%28single-precision%29) (PadLightRingConePower :type Float%20%28single-precision%29) (LightBrightnessScale :type Float%20%28double-precision%29))
  (bind _mid_pad_light_ring (Variables|Materials|GetMID_Pad_Light_Ring))
  (Utilities|FlowControl|Sequence
    (:then_0
      (Rendering|Material|SetScalarParameterValue _mid_pad_light_ring "OpacityScale" PadLightRingOpacityScale :!class /Script/Engine.MaterialInstanceDynamic))
    (:then_1
      (Rendering|Material|SetScalarParameterValue _mid_pad_light_ring "ConePow" PadLightRingConePower :!class /Script/Engine.MaterialInstanceDynamic))
    (:then_2
      (Variables|LightSettings|SetLightBrightnessScale LightBrightnessScale))))

(fn ConstructionScript
  (bind _sm_character_shadowblob (Variables|Default|GetSM_Character_ShadowBlob))
  (bind _debughoveredlighting (Variables|LightSettings|GetdebugHoveredLighting))
  (Utilities|FlowControl|Sequence
    (:then_0
      ;[1 1 1] Setup for new Lobby layout Kairos_Once everything works replace asset directly and remove this
      (if (Variables|Default|GetNewLobbyLayout)
        (bind _returnvalue (Rendering|Material|CreateDynamicMaterialInstance (Variables|Default|GetSM_Lobby_Character_Pad) 0 "/Game/UI/Foundation/Lobby/materials/M_Lobby_Pad_Floating.M_Lobby_Pad_Floating" "None" :!class /Script/Engine.PrimitiveComponent))
        (bind _output_get (Variables|Materials|SetMID_Pad _returnvalue))
        (Rendering|Material|SetMaterial (Variables|Default|GetSM_Lobby_Character_Pad) 0 _output_get)
        (bind _returnvalue_2 (Rendering|Material|CreateDynamicMaterialInstance (Variables|Default|GetSM_LobbyLightDisc_Floating_FX) 0 0 "None" :!class /Script/Engine.PrimitiveComponent))
        (Variables|Materials|SetMID%20Pad%20FX _returnvalue_2)))
    (:then_1
      ;[1 1 1] Disable visual elements for all podiums but the first 4 players
      (Rendering|SetVisibility (Variables|Default|GetTopSpotLight) false false)
      (Rendering|SetVisibility (Variables|FortTeamMemberPedestal|GetWidgetComponent_MemberInfo) false false))
    (:then_2
      (bind _returnvalue_3 (Rendering|Material|CreateDynamicMaterialInstance _sm_character_shadowblob 0 0 "None" :!class /Script/Engine.PrimitiveComponent))
      (bind _output_get_2 (Variables|Materials|SetMID_FakeBlobShadow _returnvalue_3))
      (Rendering|Material|SetScalarParameterValue _output_get_2 "ShowShadows" 0.0 :!class /Script/Engine.MaterialInstanceDynamic)
      (Rendering|SetVisibility _sm_character_shadowblob false false))
    (:then_3
      (Initialize|InitilizeLightValuesfromContext))
    (:then_4
      (Rendering|Components|Light|SetIntensity (Variables|Default|GetRimSpotLight) (* (Variables|LightSettings|GetRimSpotLight_Brightness) (Variables|LightSettings|GetLightBrightnessScale)))
      (Rendering|Components|Light|SetIntensity (Variables|Default|GetTopSpotLight) (* (Variables|LightSettings|GetLightBrightnessScale) (Math|Float|SelectFloat (Variables|LightSettings|GetTopSpotLight_Brightness_Hovered) (Variables|LightSettings|GetTopSpotLight_Brightness) _debughoveredlighting)))
      (Rendering|Components|Light|SetIntensity (Variables|Default|GetUnderlightBlue01) (* (Variables|LightSettings|GetLightBrightnessScale) (Math|Float|SelectFloat (Variables|LightSettings|GetUnderlightBlueSpot_Brightness_Hovered) (Variables|LightSettings|GetUnderlightBlueSpot_Brightness) _debughoveredlighting)))
      (Rendering|Components|Light|SetIntensity (Variables|Default|GetUnderlightBluePoint02) (* (Variables|LightSettings|GetLightBrightnessScale) (Math|Float|SelectFloat (Variables|LightSettings|GetUnderlightBluePoint_Brightness_Hovered) (Variables|LightSettings|GetUnderlightBluePoint_Brightness) _debughoveredlighting))))
    (:then_5
      ;[1 1 1] Skip Match FX setup
      (bind _returnvalue_4 (Rendering|Material|CreateDynamicMaterialInstance (Variables|Default|GetSM_LobbyFXSkipMatch) 0 0 "None" :!class /Script/Engine.PrimitiveComponent))
      (bind _output_get_3 (Variables|Materials|SetMID_SkipMatchFX _returnvalue_4))
      (Rendering|Material|SetScalarParameterValue _output_get_3 "Visible" 0.0 :!class /Script/Engine.MaterialInstanceDynamic))
    (:then_6
      (bind _asteam_member_nameplate_wrapper (Utilities|Casting|CastToTeamMemberNameplateWrapper (UserInterface|GetWidget (Variables|FortTeamMemberPedestal|GetWidgetComponent_MemberInfo))))
      (bind _output_get_4 (Variables|Default|SetWBP_Nameplate (Variables|/Game/UI/Frontend/Lobby/Nameplates/TeamMemberNameplateWrapper.TeamMemberNameplateWrapper_C|GetWBP_Nameplate _asteam_member_nameplate_wrapper)))
      (Default|BindEventtoContextMenuClosedEventDispatcher _output_get_4 (EventDispatchers|CreateEvent :!event OnContextMenuClosed)))))
