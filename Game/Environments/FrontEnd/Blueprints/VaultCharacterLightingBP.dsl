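; VaultCharacterLightingBP is a front-end lighting rig that switches between PC, low-detail PC and mobile light setups, apparently for vault character display.
; LightControl picks the setup from platform and shadow quality, toggling lights and post-process components unless AlwaysOn is set.
; A game user settings callback re-selects the setup while active.
; When active, LightControl writes sunlight vector and color to FortniteMaterialParameters; construction script offers debug previews.
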(event Custom|OVERRIDELIGHTING
  (Variables|Default|SetDebugLightingPC true))

(event Custom|ToggleLighting
  ;[1 1 1] Toggle Lighting Event
  (if (Variables|Default|GetIsActive)
    (Variables|Default|SetIsActive false)
    (else
      (Variables|Default|SetIsActive true))))

(event EventBeginPlay
  (if (Variables|Default|GetAlwaysOn)
    (Development|PrintString "ALWAYS ON")
    (else
      (Utilities|FlowControl|Sequence
        (:then_0
          (Rendering|Components|Light|SetForceCachedShadowsforMovablePrimitives (Variables|Default|GetRimLowerRight) true))
        (:then_1
          ;[1 1 1] Force disable on start up
          (CallFunction|SwitchPCLighting false)
          (CallFunction|SwitchMobileLighting false)
          (CallFunction|SwitchPCLightingLOWDetail false))
        (:then_2
          ;[1 1 1] Switch lgihting if settings are switched while screen is already active
          (bind _returnvalue (Settings|GetGameUserSettings))
          (EventDispatchers|BindEventtoOnGameUserSettingsUINeedsUpdate _returnvalue (delegate Custom|UpdateSettings)))))))

(event Custom|UpdateSettings
  (bind _returnvalue (Fortnite|IsMobilePlatform))
  ;[1 1 1] Switch lgihting if settings are switched while screen is already active
  (if (Variables|Default|GetIsActive)
    (if (Math|Boolean|ORBoolean (== (Settings|GetShadowQuality (stale EventBeginPlay _returnvalue)) 0) _returnvalue (TouchPlatform|IsNintendoSwitch))
      (if _returnvalue
        ;;[1 1 1] Enable MOBILE
        (CallFunction|SwitchPCLighting false)
        (CallFunction|SwitchPCLightingLOWDetail false)
        (CallFunction|SwitchMobileLighting true)
        (else
          ;;[1 1 1] Enable LOW DETAIL PC
          (CallFunction|SwitchPCLighting false)
          (CallFunction|SwitchMobileLighting false)
          (CallFunction|SwitchPCLightingLOWDetail true)))
      (else
        ;;[1 1 1] Disable all LOW and enbale HIGH END
        (CallFunction|SwitchMobileLighting false)
        (CallFunction|SwitchPCLightingLOWDetail false)
        (CallFunction|SwitchPCLighting true)))))

(event Custom|LightControl
  (params (NewParam :type Boolean))
  (bind _returnvalue (Fortnite|IsMobilePlatform))
  ;[1 1 1] Turn on/off lights
  (if (not (Variables|Default|GetAlwaysOn))
    (Variables|Default|SetIsActive NewParam)
    (bind _returnvalue_2 (Settings|GetGameUserSettings))
    (if (Math|Boolean|ORBoolean (== (Settings|GetShadowQuality _returnvalue_2) 0) _returnvalue (TouchPlatform|IsNintendoSwitch))
      (if (or _returnvalue (TouchPlatform|IsNintendoSwitch))
        (CallFunction|SwitchMobileLighting NewParam)
        (else
          (CallFunction|SwitchPCLightingLOWDetail NewParam)))
      (else
        (CallFunction|SwitchPCLighting NewParam)))
    (if (Variables|Default|GetIsActive)
      ;[1 1 1] Manually set MPC values necessary for various materials.
      ; Also being set in:
      ; Lobby: BP_FortniteLobbyLightSwitcher
      ; Battlepass: BattlePassScreen (widget BP)
      (Rendering|Material|SetVectorParameterValue "/Game/Packages/Fortress_SharedMaterials/GlobalMaterialParameters/FortniteMaterialParameters.FortniteMaterialParameters" "SunlightDirectionalVector" (Variables|Default|GetMPC-ManualSunlightVector) :!class /Script/Engine.KismetMaterialLibrary)
      (Rendering|Material|SetVectorParameterValue "/Game/Packages/Fortress_SharedMaterials/GlobalMaterialParameters/FortniteMaterialParameters.FortniteMaterialParameters" "SunLightColor" (Variables|Default|GetMPC-ManualSunlightColor) :!class /Script/Engine.KismetMaterialLibrary))))

(fn SwitchMobileLighting
  (params (Visibilty :type Boolean))
  (Rendering|SetVisibility (Variables|Default|GetKeyLight) false false)
  (Rendering|SetVisibility (Variables|Default|GetDirectionalLightMobile) Visibilty false)
  (Variables|/Script/Engine.PostProcessComponent|SetbEnabled Visibilty (Variables|Default|GetPostProcess_Mobile)))

(fn SwitchPCLighting
  (params (Visibility :type Boolean))
  (Rendering|SetVisibility (Variables|Default|GetBounceRear) Visibility false))

(fn SwitchPCLighting_LOWDetail
  (params (Visibility :type Boolean))
  (Rendering|SetVisibility (Variables|Default|GetDirectionalLight_LOWPC) Visibility false)
  (Variables|/Script/Engine.PostProcessComponent|SetbEnabled Visibility (Variables|Default|GetPostProcess_LOWPC)))

(fn ConstructionScript
  (bind _debuglightingpc (Variables|Default|GetDebugLightingPC))
  (bind _debugmobile_lighting (Variables|Default|GetDebugMobile%20Lighting))
  (bind _debuglighting_lowdetailpc (Variables|Default|GetDebugLighting_LOWDetailPC))
  (bind _debuglightingpc_2 (Variables|Default|GetDebugLightingPC))
  (if (Variables|Default|GetAlwaysOn)
    (if _debuglightingpc
      _)
    (Rendering|SetVisibility (Variables|Default|GetBounceRear) _debuglightingpc false)
    (else
      (Utilities|FlowControl|Sequence
        (:then_0
          (Rendering|SetVisibility (Variables|Default|GetKeyLight) false false)
          (Rendering|SetVisibility (Variables|Default|GetDirectionalLightMobile) false false))
        (:then_1
          (if _debugmobile_lighting
            (CallFunction|SwitchMobileLighting _debugmobile_lighting)
            (elif _debuglightingpc_2
              (CallFunction|SwitchPCLighting _debuglightingpc_2)
              (elif _debuglighting_lowdetailpc
                (CallFunction|SwitchPCLightingLOWDetail _debuglighting_lowdetailpc)))))))))
