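; Front-end helper actor that repositions itself for the active menu camera to place the character on Battle Royale (Athena) screens.
; It caches its transform at construction, and again at BeginPlay in an Athena zone to correct for level streaming.
; Camera-change events apply per-screen offsets to that transform when the subgame-driven Athena flag is set, otherwise zero offsets.
; Tilted rotation offsets sit behind a hardcoded false condition.
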(event EventBeginPlay
  (bind _returnvalue (BlueprintContext|GetFortFrontEndContext))
  (Utilities|FlowControl|Sequence
    (:then_0
      ;[1 1 1] If we were level-streamed into the BR map, we need to update our original transform at BeginPlay. Outside the editor, streamed-in actors start at the origin and then get fixed up.
      (if (and (GlobalUIContext|IsinZone (BlueprintContext|GetFortGlobalUIContext)) (Utilities|Enum|Equal(Enum) (GlobalUIContext|SubGame|GetSubGame (BlueprintContext|GetFortGlobalUIContext)) "Athena"))
        (Variables|Default|SetOriginalTransform (Transformation|GetActorTransform self))))
    (:then_1
      (GlobalUIContext|SubGame|AssignSubGameChanged (BlueprintContext|GetFortGlobalUIContext) (delegate Custom|SubGameChanged_Event_0)))
    (:then_2
      (Utilities|IsValid _returnvalue
        (:Is%20Valid
          (FrontEndCamera|BindEventtoOnFrontEndCameraChanged _returnvalue (delegate Custom|OnFrontEndCameraChanged)))))))

(event Custom|SubGameChanged_Event_0
  (params (SubGame :type ESubGame%20Enum :default "Campaign"))
  (switch Utilities|FlowControl|Switch|SwitchonESubGame (GlobalUIContext|SubGame|GetSubGame (BlueprintContext|GetFortGlobalUIContext))
    (:Campaign
      _)
    (:Athena
      (Variables|Default|SetAthena true))
    (:Invalid
      _))
  (Transformation|SetActorTransform (Variables|Default|GetOriginalTransform))
  (Variables|Default|SetAthena false))

(event Custom|OnFrontEndCameraChanged
  (params (NewCamera :type EFrontEndCamera%20Enum :default "Invalid") (OldCamera :type EFrontEndCamera%20Enum :default "Invalid"))
  (CallFunction|UpdatePositionforCamera NewCamera))

(event EventEndPlay
  (params (EndPlayReason :type EEndPlayReason%20Enum :default "Destroyed"))
  (bind _returnvalue (BlueprintContext|GetFortFrontEndContext))
  (Utilities|IsValid _returnvalue
    (:Is%20Valid
      (FrontEndCamera|UnbindEventfromOnFrontEndCameraChanged _returnvalue (delegate Custom|OnFrontEndCameraChanged)))))

(fn UpdatePositionForCamera
  (params (FrontendCamera :type EFrontEndCamera%20Enum :default "Invalid"))
  (bind _returnvalue (not true))
  (bind _frontendcamera (Variables|GetFrontendCamera))
  (if (Variables|Default|GetAthena)
    (Utilities|FlowControl|Sequence
      (:then_0
        ;[1 1 1] Offsets for the characters position on the different menu screens.
        (if (Utilities|Enum|Equal(Enum) _frontendcamera "Invalid")
          _
          (elif (Utilities|Enum|Equal(Enum) _frontendcamera "Cosmetics")
            (Variables|Default|SetOffsetTranslate "65.000000,10.000000,10.000000") ; Offset Locker Female
            (if _returnvalue
              (Variables|Default|SetOffsetRotate "15.000000, 0,-2.000000")
              (else
                (Variables|Default|SetOffsetRotate "0, 0, 0")))
            (elif (Utilities|Array|ContainsItem (Utilities|Array|MakeArray "Home" "Store" "StoreItemInspect") _frontendcamera)
              (Variables|Default|SetOffsetTranslate "85.000000,35.000000,20.000000") ; Offset Shops Female
              (if _returnvalue
                (Variables|Default|SetOffsetRotate "15.000000, 0,-2.000000")
                (else
                  _))
              (elif (Utilities|Enum|Equal(Enum) _frontendcamera "SmallCosmetics")
                (Variables|Default|SetOffsetTranslate "80.000000,28.000000,-10.000000") ; Offset Battle Pass Female
                (if _returnvalue
                  (Variables|Default|SetOffsetRotate "5.000000, 0,-2.000000")
                  (else
                    _)))))))
      (:then_1
        _))
    (else
      (Variables|Default|SetOffsetTranslate "0, 0, 0") ; Zero out the character placement actor
      (Variables|Default|SetOffsetRotate "0, 0, 0")))
  (Transformation|SetActorTransform (Math|Transform|MakeTransform (+ (.location (Variables|Default|GetOriginalTransform)) (Variables|Default|GetOffsetTranslate)) (Math|Rotator|CombineRotators (.rotation (Variables|Default|GetOriginalTransform)) (Variables|Default|GetOffsetRotate)) (.scale (Variables|Default|GetOriginalTransform)))))

(fn ConstructionScript
  (Variables|Default|SetOriginalTransform (Transformation|GetActorTransform self)))
