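; Front-end helper Blueprint for the Vault screen that plays a temporary swirl particle effect when the camera switches to the Vault.
; It binds to the front-end context's camera-changed event on BeginPlay and unbinds on EndPlay.
; Entering the Vault shows PS-NoPlayer-Swirl01 and calls ItemRez, whose retriggerable delay hides it after a few seconds.
; Any other camera hides the swirl immediately; GetFrontendAnimInstance and the construction script are empty.
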(event EventBeginPlay
  (bind _returnvalue (BlueprintContext|GetFortFrontEndContext))
  ;[1 1 1] BeginPlay
  (Utilities|IsValid _returnvalue
    (:Is%20Valid
      (FrontEndCamera|BindEventtoOnFrontEndCameraChanged _returnvalue (EventDispatchers|CreateEvent :!event OnFrontendCameraChanged)))))

(event EventEndPlay
  (params (EndPlayReason :type EEndPlayReason%20Enum :default "Destroyed"))
  (bind _returnvalue (BlueprintContext|GetFortFrontEndContext))
  (Utilities|IsValid _returnvalue
    (:Is%20Valid
      (FrontEndCamera|UnbindEventfromOnFrontEndCameraChanged _returnvalue (EventDispatchers|CreateEvent :!event OnFrontendCameraChanged)))))

(event Custom|ItemRez
  (Utilities|FlowControl|RetriggerableDelay 3)
  (Rendering|SetVisibility (Variables|Default|GetPS-NoPlayer-Swirl01) false false))

(fn GetFrontendAnimInstance)

(fn OnFrontendCameraChanged
  (params (NewCamera :type EFrontEndCamera%20Enum :default "Invalid") (OldCamera :type EFrontEndCamera%20Enum :default "Invalid"))
  (bind _ps_noplayer_swirl01 (Variables|Default|GetPS-NoPlayer-Swirl01))
  (if (Utilities|Enum|Equal(Enum) NewCamera "Vault")
    ;[1 1 1] Lights can be turned
    ; on/off here if need be
    (Rendering|SetVisibility _ps_noplayer_swirl01 true false)
    (CallFunction|ItemRez)
    (else
      (Rendering|SetVisibility _ps_noplayer_swirl01 false false))))

(fn ConstructionScript)
