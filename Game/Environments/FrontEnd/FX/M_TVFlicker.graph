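# Light function material that gives a light a TV-style flicker, stored with the front-end environment effects.
# Time drives sine and cosine waves of different periods: a sine plus a cosine is multiplied by a slower sine, then added to a half-strength fast sine.
# The sum is clamped to a high floor, so the flicker only dims the light partway and never turns it off.
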Material /Game/Environments/FrontEnd/FX/M_TVFlicker
graph {
 properties {materialDomain=MD_LightFunction, pixelDepthOffsetMode=PDOM_Legacy, shadingModel=MSM_Unlit, usedShadingModels=Unlit}
 nodes [
  n1 = Time()
  n2 = Add(A=Sine(None=@n1) {period=0.5}, B=Cosine(None=@n1) {period=2})
  n3 = Sine(None=@n1) {period=3}
  n4 = Sine(None=@n1) {period=0.15}
  n5 = Constant() {r=0.5}
  Clamp(None=Add(A=Multiply(A=@n2, B=@n3), B=Multiply(A=@n4, B=@n5))) {minDefault=0.7}
 ]
}
