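# Lit floor material for the front-end lobby backplate, themed through MPC_LobbyBackground color, fade, surface and stylised controls.
# Layered floor, noise and dirt textures form the base; a stylised blend adds quantized world-space noise.
# Emissive fakes reflections by projecting blurred render-target and default-background textures on a virtual plane with Fresnel, dampened on deferred shading.
# Roughness follows a horizontal screen mask and global illumination/reflection quality; forward/mobile adds tint.
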Material /Game/Environments/FrontEnd/Materials/M_BackPlate_Floor
graph {
 properties {bUseMaterialAttributes=true, pixelDepthOffsetMode=PDOM_Legacy, usedShadingModels=Default Lit}
 nodes [
  n1 = ScalarParameter() {defaultValue=1823.9929, parameterName=StylisedNoiseTiling}
  Multiply(A=TextureCoordinate(), B=@n1)
  n2 = ScalarParameter() {defaultValue=462.7858, parameterName=StylisedNoiseTiling2}
  Multiply(A=TextureCoordinate(), B=@n2)
  n3 = ScreenPosition()
  NamedRerouteDeclaration(None=SphereMask(
    A=MaterialFunctionCall(Float2=@n3.ViewportUV) {materialFunction=&BreakOutFloat2Components}.R
    B=Constant() {r=0.5}
    Radius=ScalarParameter() {defaultValue=0.6, parameterName=ScreenMaskRadius}
    Hardness=ScalarParameter() {defaultValue=0.6, parameterName=ScreenMaskHardness}
   )) {name=ScreenMask.U, nodeColor=<rgba 0.6641 0 1 1>}
  n4 = CollectionParameter() {collection=/Game/Environments/FrontEnd/BackPlates/Materials/MPC_LobbyBackground, parameterName=FloorColor}
  n5 = VectorParameter() {parameterName=Color, defaultValue=<rgba 0.3021 0.3021 0.3021 1>}
  n6 = VectorParameter() {parameterName=Color2, defaultValue=<rgba 0.02 0.02 0.02 1>}
  n7 = TextureSample(UVs=Multiply(A=TextureCoordinate(), B=ScalarParameter() {defaultValue=0.5, parameterName=Tiling})) {texture=/Game/UI/Foundation/Lobby/textures/T_Lobby_Floor_C}
  n8 = TextureSample(UVs=Multiply(A=TextureCoordinate(), B=ScalarParameter() {defaultValue=0.4093, parameterName=Tiling3})) {samplerType=SAMPLERTYPE_LinearGrayscale, texture=/Game/Global/Textures/Default/Noise/T_SwirlNoise_H2}
  n9 = Multiply(A=@n7.R, B=Power(Base=@n8.R, Exp=ScalarParameter() {defaultValue=0.7234, parameterName=NoiseContrast}))
  n10 = Multiply(A=TextureCoordinate(), B=ScalarParameter() {defaultValue=0.1524, parameterName=Tiling2})
  n11 = LinearInterpolate(
   A=Constant3Vector()
   B=LinearInterpolate(
    A=LinearInterpolate(A=@n5.RGB, B=@n6.RGB, Alpha=@n9)
    B=VectorParameter() {parameterName=Color3, defaultValue=<rgba 0.0365 0.0365 0.0365 1>}.RGB
    Alpha=TextureSample(UVs=@n10) {texture=/Game/Environments/Props/DriveIn/Textures/T_FilmLensDirt}.R
   )
   Alpha=Saturate(None=CollectionParameter() {collection=&MPC_LobbyBackground, parameterName=Fade_Floor_to_Black})
  )
  n12 = Multiply(A=VectorParameter() {parameterName=TintFloor, defaultValue=<rgba 0.5543 0.7095 0.7656 1>}.RGB, B=@n11)
  n13 = LinearInterpolate(
   A=ShadingPathSwitch(Default=@n11, Deferred=@n11, Forward=@n12, Mobile=@n12)
   B=ComponentMask(None=@n4) {b=true, g=true, r=true}
   Alpha=ComponentMask(None=@n4) {a=true}
  )
  n14 = ScalarParameter() {parameterName=StylisedBaseColorContribution}
  n15 = WorldPosition()
  n16 = TextureSample(UVs=Divide(A=@n15.XY, B=@n2)) {texture=&/WildEstate_Landscape_Grass_Overlay_D}
  n17 = LinearInterpolate(A=ScalarParameter() {defaultValue=0.8876, parameterName=MinDarkness2}, Alpha=@n16.G)
  n18 = ScalarParameter() {defaultValue=0.652, parameterName=MinDarkness}
  n19 = TextureSample(UVs=Divide(A=@n15.XY, B=@n1)) {samplerType=SAMPLERTYPE_LinearGrayscale, texture=&T_SimplePerlin_Greyscale}
  n20 = ScalarParameter() {defaultValue=4, parameterName=Quantize}
  n21 = Divide(A=Round(None=Multiply(A=@n19.R, B=@n20)), B=@n20)
  n22 = Multiply(A=@n13, B=Multiply(A=Multiply(A=@n17, B=@n16.R), B=LinearInterpolate(A=@n18, Alpha=@n21)))
  n23 = ComponentMask(None=CollectionParameter() {collection=&MPC_LobbyBackground, parameterName=FloorStylised}) {r=true}
  n24 = CollectionParameter() {collection=&MPC_LobbyBackground, parameterName=FloorMetallic}
  n25 = ScalarParameter() {defaultValue=561651648, parameterName=NoiseContrast2}
  n26 = Multiply(
   A=MaterialFunctionCall(In=Power(Base=@n7.R), Contrast=ScalarParameter() {defaultValue=0.5, parameterName=SpecContrast}) {materialFunction=&CheapContrast}.Result
   B=OneMinus(None=Power(Base=@n8.R, Exp=@n25))
  )
  n27 = LinearInterpolate(
   A=Constant()
   B=LinearInterpolate(
    A=ScalarParameter() {defaultValue=0.1, parameterName=Roughness}
    B=ScalarParameter() {defaultValue=0.4, parameterName=Roughness2}
    Alpha=@n26
   )
   Alpha=CollectionParameter() {collection=&MPC_LobbyBackground, parameterName=FloorRoughness}
  ) {constAlpha=1}
  n28 = Add(A=ScalarParameter() {defaultValue=0.23, parameterName=RoughnessAddforLowQualityReflections}, B=@n27)
  n29 = CollectionParameter() {
   collection=&FortniteMaterialParameters
   parameterName=GlobalIlluminationEnabled
  }
  n30 = CollectionParameter() {collection=&FortniteMaterialParameters, parameterName=ReflectionQuality}
  n31 = Subtract(A=@n30)
  n32 = Saturate(None=Min(A=@n29, B=Saturate(None=@n31)))
  n33 = MaterialFunctionCall(
   Plane Normal Axis=Constant3Vector() {constant=<rgba 0 1 0 0>}
   Plane X Axis=Constant3Vector() {constant=<rgba 1 0 0 0>}
   Plane Center=Add(A=ActorPositionWS(), B=VectorParameter() {parameterName=ReflectionOffset, defaultValue=<rgba 49.2157 0 -451.2493 1>}.RGB)
   UV Scale=AppendVector(
    A=ScalarParameter() {defaultValue=1548.4684, parameterName=ScaleReflX}
    B=ScalarParameter() {defaultValue=1022.4915, parameterName=ScaleReflY}
   )
  ) {materialFunction=&VirtualPlaneCoordinates}
  n34 = LinearInterpolate(
   A=TextureSample(UVs=@n33.UVs) {texture=/Game/Environments/FrontEnd/BackPlates/Materials/RT_Frontend_Backplate_Blur}.RGB
   B=TextureSample(UVs=@n33.UVs) {texture=/Game/Environments/FrontEnd/BackPlates/Textures/T_Background_LobbyDefault}.RGB
   Alpha=Saturate(None=CollectionParameter() {collection=&MPC_LobbyBackground, parameterName=Fade})
  )
  n35 = ScalarParameter() {defaultValue=1.4067, parameterName=ReflectionContrast}
  n36 = ScalarParameter() {parameterName=ReflectionBirightness}
  n37 = ScalarParameter() {defaultValue=3, parameterName=ReflectionBirightness2}
  n38 = Fresnel(ExponentIn=ScalarParameter() {defaultValue=3.863, parameterName=ReflectionExponent}, Normal=VertexNormalWS()) {baseReflectFraction=0}
  n39 = Multiply(A=Multiply(A=Power(Base=@n34, Exp=@n35), B=LinearInterpolate(A=@n36, B=@n37, Alpha=@n38)), B=ComponentMask(None=@n11) {r=true})
  n40 = LinearInterpolate(
   A=Constant()
   B=@n39
   Alpha=Max(A=ScalarParameter() {parameterName=UseFakeReflections}, B=ScalarParameter() {defaultValue=0.2, parameterName=UseFakeReflectionsHighEnd})
  ) {constAlpha=1}
  n41 = ShadingPathSwitch(Default=@n40, Deferred=@n40, Forward=@n39, Mobile=@n39)
  n42 = ScalarParameter() {defaultValue=-0.5, parameterName=Specular}
  n43 = ScalarParameter() {defaultValue=1, parameterName=Specular2}
  n44 = CollectionParameter() {collection=&MPC_LobbyBackground, parameterName=FloorSpecular}
  n45 = Multiply(A=TextureCoordinate(), B=ScalarParameter() {defaultValue=0.25, parameterName=TilingNormal})
  SetMaterialAttributes!(
   MaterialAttributes
   Base Color=LinearInterpolate(A=@n13, B=Multiply(A=@n14, B=@n22), Alpha=@n23)
   Metallic=Multiply(A=ScalarParameter() {defaultValue=0.5, parameterName=Metal}, B=ComponentMask(None=@n24) {r=true})
   Specular=Multiply(A=LinearInterpolate(A=@n42, B=@n43, Alpha=@n26), B=ComponentMask(None=@n44) {r=true})
   Roughness=LinearInterpolate(B=LinearInterpolate(A=@n28, B=@n27, Alpha=@n32) {constAlpha=1, desc=Roughness}, Alpha=NamedRerouteUsage()) {constA=1}
   Anisotropy
   Emissive Color=LinearInterpolate(A=@n41, B=Add(A=@n22, B=@n41), Alpha=@n23)
   Opacity
   Opacity Mask
   Normal=MaterialFunctionCall(
    Normal=TextureSample(UVs=@n45) {samplerType=SAMPLERTYPE_Normal, texture=/Game/Environments/Sets/Neo/Architecture/Textures/T_Neo_Ext_Floortiles_02_N}.RGB
    Flatness=ScalarParameter() {defaultValue=1, parameterName=NormalFlatten}
   ) {materialFunction=&FlattenNormal}.Result
  ) {attributeSetTypes [
    #1
    #2
    #3
    #4
    #5
    #6
    #7
    #8
    #9
   ]}
 ]
}
