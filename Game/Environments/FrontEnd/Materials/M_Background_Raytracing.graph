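# Unlit front-end background material that shows a cubemap (Apollo day sky by default) or a 2D texture, and casts no ray-traced shadows.
# A static switch picks between the cubemap, sampled along the negated reflection vector, and the texture on the second UV channel.
# The brightness-scaled result blends with a fade color through a UV-centered sphere mask with adjustable radius and hardness, forming a radial vignette.
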Material /Game/Environments/FrontEnd/Materials/M_Background_Raytracing
graph {
 properties {bCastRayTracedShadows=false, pixelDepthOffsetMode=PDOM_Legacy, shadingModel=MSM_Unlit, usedShadingModels=Unlit}
 nodes [
  Constant3Vector() {constant=<rgba 1 1 -1 0>}
  n1 = TextureSampleParameterCube!(UVs=Multiply(A=ReflectionVectorWS(), B=Constant() {r=-1}), Apply View MipBias) {parameterName=Cubemap, texture=/Game/Packages/Fortress_Sky/SkyDome/MaterialInstances/SkyDome_Apollo/Cubemap_Apollo_Day}
  n2 = TextureSampleParameter2D(UVs=TextureCoordinate() {coordinateIndex=1}) {parameterName=Texture}
  LinearInterpolate(
   A=VectorParameter() {parameterName=FadeColor, defaultValue=<rgba 0 0 0 1>}.RGB
   B=Multiply(
    A=StaticSwitchParameter(True=@n1.RGB, False=@n2.RGB) {defaultValue=true, parameterName=CubeMap}
    B=ScalarParameter() {defaultValue=1, parameterName=Brightness}
   )
   Alpha=SphereMask(
    A=TextureCoordinate() {coordinateIndex=1}
    B=Constant2Vector() {g=0.5, r=0.5}
    Radius=ScalarParameter() {defaultValue=0.5, parameterName=R}
    Hardness=ScalarParameter() {defaultValue=0.5, parameterName=H}
   )
  )
 ]
}
