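# Translucent unlit highlight material for Cozy Lodge front-end objects, drawing an outline that PreSelect blends into a pulsing, panning shimmer randomized by object position.
# Vertices push along normals by a depth-scaled screen thickness, plus a camera-axis offset.
# Opacity compares pixel depth to custom depth, or scene depth on mobile, with an optional texture mask.
# Static switches add tree animation, ribbon and sock vertex shaders; scene-depth edge detection sits unconnected.
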Material /Game/Environments/FrontEnd/Materials/M_CozyLodge_Highlight
graph {
 properties {bUseMaterialAttributes=true, blendMode=BLEND_Translucent, pixelDepthOffsetMode=PDOM_Legacy, shadingModel=MSM_Unlit, usedShadingModels=Unlit}
 nodes [
  n1 = ObjectPositionWS()
  n2 = ComponentMask(None=@n1) {r=true}
  n3 = ComponentMask(None=@n1) {g=true}
  n4 = Multiply(A=Frac(None=Add(A=@n2, B=@n3)), B=ScalarParameter() {defaultValue=1, parameterName=Random})
  n5 = Add(A=ScreenPosition().ViewportUV, B=@n4)
  n6 = Panner(
   Coordinate=Multiply(A=@n5, B=ScalarParameter() {defaultValue=1, parameterName=TilingSineMask})
   Speed=AppendVector(A=ScalarParameter() {defaultValue=0.1, parameterName=Pansine}, B=Constant())
  )
  ComponentMask(None=@n6) {g=true, r=true}
  n7 = ScalarParameter() {defaultValue=1, parameterName=VertexOffsetBias}
  n8 = ComponentMask(None=Multiply(A=Custom!(None) {code="GetCotanHalfFieldOfView().x", desc="GetCotanHalfFieldOfView()", outputType=CMOT_Float2}, B=ViewSize())) {r=true}
  n9 = ScalarParameter() {defaultValue=4, parameterName=OutlineThickness}
  n10 = Divide(A=Divide(A=Multiply(A=@n8) {constB=0.5}, B=PixelDepth().R), B=ComponentMask(None=@n9) {r=true})
  n11 = ScalarParameter() {parameterName=PreSelect}
  n12 = Multiply(A=VertexNormalWS(), B=LinearInterpolate(A=Divide(A=@n7, B=@n10), B=Constant(), Alpha=@n11))
  n13 = Add(A=WorldPosition().XYZ, B=@n12)
  n14 = CameraPositionWS()
  n15 = Multiply(A=Normalize(VectorInput=Subtract(A=@n13, B=@n14)), B=ScalarParameter() {defaultValue=-5, parameterName=OffestInside})
  Multiply(B=@n15)
  n16 = SceneDepth()
  n17 = Multiply(A=@n9, B=SceneTexelSize()) {constA=2}
  n18 = Constant()
  n19 = AppendVector(A=ComponentMask(None=@n17) {desc=x, r=true}, B=@n18)
  n20 = Subtract(A=@n16.R, B=SceneDepth!(OffsetFraction=@n19) {inputMode=OffsetFraction}.R)
  n21 = AppendVector(A=@n18, B=ComponentMask(None=@n17) {desc=y, g=true}) {desc=y+}
  n22 = Subtract(A=@n16.R, B=SceneDepth(OffsetFraction=@n21) {inputMode=OffsetFraction}.R)
  n23 = Constant() {r=-1}
  n24 = Subtract(A=@n16.R, B=SceneDepth(OffsetFraction=Multiply(A=@n19, B=@n23)) {inputMode=OffsetFraction}.R)
  n25 = Subtract(A=@n16.R, B=SceneDepth(OffsetFraction=Multiply(A=@n23, B=@n21)) {inputMode=OffsetFraction}.R)
  n26 = Divide(A=Add(A=Add(A=@n20, B=@n22), B=Add(A=@n24, B=@n25)), B=ScalarParameter() {defaultValue=23.4474})
  Saturate(None=Power(Base=Saturate(None=@n26), Exp=ScalarParameter() {defaultValue=3, parameterName=Contrast}))
  SceneDepth!(UVs)
  n27 = TextureSampleParameter2D!(UVs, Level=ScalarParameter() {defaultValue=2.0372, parameterName=Level}, Apply View MipBias) {
   mipValueMode=TMVM_MipLevel
   parameterName=Mask
   samplerType=SAMPLERTYPE_Masks
   texture=/Game/Environments/Apollo/Foliage/PineTree/Textures/T_Apollo_PineBranch_01_M
  }
  n28 = ScalarParameter() {defaultValue=39.3254, parameterName=MaskOpacity}
  n29 = Saturate(None=Power(Base=Multiply(A=@n27.R, B=@n28), Exp=ScalarParameter() {defaultValue=50, parameterName=ContrastOpacityMask}))
  n30 = SceneTexture() {sceneTextureId=PPI_CustomDepth}
  n31 = Divide(
   A=Subtract(A=ComponentMask(None=@n30.Color) {r=true}, B=PixelDepth().R)
   B=ScalarParameter() {defaultValue=30.2221, parameterName=CustomDepth}
  )
  n32 = OneMinus(None=Power(Base=Saturate(None=@n31), Exp=ScalarParameter() {defaultValue=50, parameterName=ContrastCore}))
  n33 = Divide(A=Subtract(A=SceneDepth().R, B=PixelDepth().R), B=ScalarParameter() {defaultValue=30.2221, parameterName=CustomDepth})
  n34 = OneMinus(None=Power(Base=Saturate(None=@n33), Exp=ScalarParameter() {defaultValue=50, parameterName=ContrastCore}))
  n35 = Saturate(None=Subtract(A=Constant() {r=1}, B=FeatureLevelSwitch(Default=@n32, ES2=@n34, ES3_1=@n34, SM4_REMOVED=@n32)))
  n36 = ScalarParameter() {parameterName=MinHighlight}
  n37 = Panner(
   Coordinate=Multiply(A=@n5, B=ScalarParameter() {defaultValue=2, parameterName=Tiling})
   Speed=AppendVector(A=ScalarParameter() {defaultValue=0.4, parameterName=Pan}, B=Constant())
  )
  n38 = ScalarParameter() {defaultValue=1.7003, parameterName=rot}
  n39 = Power(
   Base=TextureSample(UVs=Rotator(Coordinate=@n37, Time=@n38)) {texture=/Game/Global/Textures/Default/Masks_and_Gradients/WeaponGradients/T_Gradient_DesertSky_Desaturated}.R
   Exp=ScalarParameter() {defaultValue=2.7082, parameterName=ShimmerContrast}
  )
  n40 = ScalarParameter() {defaultValue=0.05, parameterName=Shimmer}
  n41 = Add(A=Time(), B=@n4)
  n42 = ScalarParameter() {defaultValue=0.2, parameterName=Speed}
  n43 = LinearInterpolate(
   A=ScalarParameter() {defaultValue=-0.6762, parameterName=c1}
   B=ScalarParameter() {defaultValue=1, parameterName=c2}
   Alpha=Sine(None=Multiply(A=@n41, B=@n42))
  )
  n44 = LinearInterpolate(A=Saturate(None=Add(A=@n35, B=@n36)), B=Multiply(A=Multiply(A=@n39, B=@n40), B=Saturate(None=@n43)), Alpha=@n11)
  n45 = SetMaterialAttributes!(
   MaterialAttributes
   Emissive Color=LinearInterpolate(
    A=VectorParameter() {parameterName=OutlineColor, defaultValue=<rgba 1 0.6816 0.5086 1>}.RGB
    B=VectorParameter() {parameterName=PreSelectColor, defaultValue=<rgba 1 0.6604 0.1499 1>}.RGB
    Alpha=@n11
   )
   Opacity=StaticSwitchParameter(True=Multiply(A=@n29, B=@n44), False=@n44) {parameterName=UseOpacityMaskTexture}
  ) {attributeSetTypes [#1, #2]}
  n46 = MaterialFunctionCall!(In=@n45, OpaqueCanopyMask) {materialFunction=/Game/Environments/Apollo/Foliage/MaterialFunctions/MF_TreeAnim_Apollo_Optimized}
  n47 = GetMaterialAttributes(None=@n46.Result) {attributeGetTypes [#3]}
  n48 = Add(A=@n15, B=@n12)
  n49 = Subtract(A=ObjectPositionWS(), B=Add(A=@n48, B=WorldPosition().XYZ))
  n50 = VertexColor()
  n51 = Constant() {r=0.3}
  n52 = Constant() {r=1}
  n53 = Constant()
  n54 = If(A=@n50.A, B=Constant() {r=0.75}, "A > B"=@n53, "A == B"=@n53, "A < B"=@n52)
  n55 = LinearInterpolate(
   A=Multiply(A=If(A=@n50.A, B=@n51, "A > B"=@n52, "A == B"=@n52, "A < B"=@n53), B=@n54)
   B=OneMinus(None=@n54)
   Alpha=ScalarParameter() {parameterName=RibbonType}
  )
  n56 = StaticSwitchParameter(True=Add(A=@n48, B=Multiply(A=@n49, B=@n55)), False=@n48) {parameterName=UseRibbonVertexshader}
  n57 = ScalarParameter() {parameterName=ProgressJiggle}
  n58 = Constant3Vector() {constant=<rgba 1 0 0 0>}
  n59 = Multiply(A=Time(), B=ScalarParameter() {defaultValue=2, parameterName=SpeedRot})
  n60 = RotateAboutAxis(
   NormalizedRotationAxis=Normalize(VectorInput=Transform(None=@n58) {transformSourceType=TRANSFORMSOURCE_Local})
   RotationAngle=Multiply(A=Sine(None=@n59), B=ScalarParameter() {defaultValue=0.02, parameterName=RotJiggle})
   PivotPoint=TransformPosition(None=VectorParameter() {parameterName=pivot, defaultValue=<rgba 0 13.7973 -9.4825 1>}.RGB) {transformType=TRANSFORMPOSSOURCE_World}
   Position=Add(A=@n56, B=WorldPosition().XYZ)
  )
  n61 = StaticSwitchParameter(True=Add(A=@n56, B=Multiply(A=@n57, B=@n60)), False=@n56) {parameterName=UseSockVertexShader}
  SetMaterialAttributes!(
   MaterialAttributes=StaticSwitchParameter(True=@n46.Result, False=@n45) {parameterName=UseTreeAnim}
   World Position Offset=StaticSwitchParameter(True=Add(A=@n47.World Position Offset, B=@n61), False=@n61) {parameterName=UseTreeAnim}
  ) {attributeSetTypes [#3]}
  Sine(None=ComponentMask(None=Rotator(Coordinate=@n6, Time=@n38)) {r=true})
  TextureCoordinate()
 ]
}
