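# M_Frontend_Snow is a Default Lit snow surface material for the front-end environment, built through material attributes.
# World XY position divided by a Size parameter gives planar UVs for the snow normal map, while a tile multiplier scales them for the roughness mask's red channel.
# Base color is a flat Color parameter, and the normal's X and Y are scaled down to soften relief.
# Two LandscapeLayerCoords nodes and a second mask sample are unconnected.
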Material /Game/Environments/FrontEnd/Materials/M_Frontend_Snow
graph {
 properties {bUseMaterialAttributes=true, pixelDepthOffsetMode=PDOM_Legacy, usedShadingModels=Default Lit}
 nodes [
  LandscapeLayerCoords() {mappingScale=50}
  LandscapeLayerCoords() {mappingScale=30}
  n1 = Divide(A=ComponentMask(None=WorldPosition().XYZ) {g=true, r=true}, B=ScalarParameter() {defaultValue=1299.4639, parameterName=Size})
  n2 = TextureSample(UVs=@n1) {
   samplerSource=SSM_Wrap_WorldGroupSettings
   samplerType=SAMPLERTYPE_Normal
   texture=/Game/Global/Textures/Default/Tileable_Patterns/T_Snow01_N
  }
  n3 = TextureSample(UVs=Multiply(A=ScalarParameter() {defaultValue=3, parameterName=SnowRoughTile}, B=@n1)) {
   samplerSource=SSM_Wrap_WorldGroupSettings
   samplerType=SAMPLERTYPE_LinearColor
   texture=/Game/Athena/Environments/Landscape/Textures/T_Athena_Snow_01_M
  }
  SetMaterialAttributes!(
   MaterialAttributes
   Base Color=VectorParameter() {parameterName=Color, defaultValue=<rgba 0.9062 0.9062 0.9062 1>}.RGB
   Roughness=@n3.R
   Normal=Multiply(A=@n2.RGB, B=Constant3Vector() {constant=<rgba 0.25 0.25 1 0>})
  ) {attributeSetTypes [#1, #2, #3]}
  TextureSample(UVs=@n1) {
   samplerSource=SSM_Wrap_WorldGroupSettings
   samplerType=SAMPLERTYPE_LinearColor
   texture=&T_Athena_Snow_01_M
  }
 ]
}
