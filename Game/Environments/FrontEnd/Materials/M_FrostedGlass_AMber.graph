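# Translucent unlit amber frosted-glass material in the front-end environment, with index-of-refraction distortion.
# An ice-lake snow texture and vertex-color green drive a HeightLerp mask that blends two opacity levels and varies refraction between flat and frost-modulated values.
# A night-sky cubemap sampled by reflection vector is added to the frost texture, both brightness-scaled by depth fade.
# An amber GlassTint multiplies that sum to form the glass color.
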Material /Game/Environments/FrontEnd/Materials/M_FrostedGlass_AMber.M_FrostedGlass_Amber
graph {
 properties {
  blendMode=BLEND_Translucent
  pixelDepthOffsetMode=PDOM_Legacy
  refractionMethod=RM_IndexOfRefraction
  shadingModel=MSM_Unlit
  usedShadingModels=Unlit
 }
 nodes [
  n1 = TextureSample() {texture=/Game/Packages/Fortnite_UniqueMaterials_1/Textures/T_Ice_Lake_Snow_D}
  n2 = MaterialFunctionCall(Transition Phase=VertexColor().G, Height Texture=@n1.R, Contrast=ScalarParameter() {defaultValue=0.3145, parameterName=Contrast}) {materialFunction=&HeightLerp}
  LinearInterpolate(
   A=ScalarParameter() {defaultValue=0.5, parameterName=Opacity1}
   B=ScalarParameter() {defaultValue=0.55, parameterName=Opacity2}
   Alpha=@n2.Alpha
  )
  LinearInterpolate(
   A=ScalarParameter() {defaultValue=1, parameterName=R}
   B=LinearInterpolate(A=ScalarParameter() {defaultValue=1.01, parameterName=R2}, B=ScalarParameter() {defaultValue=0.99, parameterName=R3}, Alpha=@n1.R)
   Alpha=@n2.Alpha
  )
  MaterialFunctionCall() {materialFunction=/Game/Environments/FrontEnd/Materials/Cozy/MF_CozyWindow}
  n3 = TextureSample(UVs=ReflectionVectorWS()) {texture=/Game/Packages/Fortress_Sky/SkyDome/MaterialInstances/SkyDome_Apollo/Cubemap_Apollo_Night}
  n4 = Saturate(None=DepthFade(Opacity=Constant() {r=1}, FadeDistance=ScalarParameter() {defaultValue=2282.6501, parameterName=DepthFade}))
  n5 = LinearInterpolate(
   A=ScalarParameter() {parameterName=Reflection Brightness2}
   B=ScalarParameter() {defaultValue=0.1746, parameterName=Reflection Brightness}
   Alpha=@n4
  )
  n6 = LinearInterpolate(A=ScalarParameter() {parameterName=IceBrightness}, B=ScalarParameter() {defaultValue=0.1796, parameterName=IceBrightness2}, Alpha=@n4)
  Multiply(
   A=VectorParameter() {parameterName=GlassTint, defaultValue=<rgba 1 0.4013 0.1085 1>}.RGB
   B=Add(A=Multiply(A=@n3.RGB, B=@n5), B=Multiply(A=@n6, B=@n1.R))
  )
  TextureSample() {texture=/Game/Global/Textures/Default/Noise/T_UI_Ice_Noise}
 ]
}
