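# M_LobbyFlares is an unlit additive material for flare glows in the front-end lobby.
# FakeHDR tints and weights the flare texture's RGB channels separately; otherwise one color's brightness lerps by the red channel, optionally modulated by panning cloud noise (IsAnimated).
# PoliceLight replaces the opacity mask with a blinking one built from time-driven sines and a horizontal UV falloff.
# An optional depth fade softens the resulting opacity at intersections.
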Material /Game/Environments/FrontEnd/Materials/M_LobbyFlares
graph {
 properties {bUseTranslucencyVertexFog=false, blendMode=BLEND_Additive, pixelDepthOffsetMode=PDOM_Legacy, shadingModel=MSM_Unlit, usedShadingModels=Unlit}
 nodes [
  n1 = VectorParameter() {parameterName=Color, defaultValue=<rgba 0.5 0.5179 1 1>}
  n2 = TextureSampleParameter2D() {parameterName=Opacity, texture=/Game/Environments/FrontEnd/Textures/T_EFX_CameraFlare_02}
  n3 = Add(A=Multiply(A=@n1.RGB, B=@n2.R), B=Multiply(A=VectorParameter() {parameterName=Color_1, defaultValue=<rgba 0.5 0.5179 1 1>}.RGB, B=@n2.G))
  n4 = Multiply(A=VectorParameter() {parameterName=Color_2, defaultValue=<rgba 0.5 0.5179 1 1>}.RGB, B=@n2.B)
  n5 = ScalarParameter() {defaultValue=52.6634, parameterName=Brightness_2}
  n6 = ScalarParameter() {parameterName=Brightness_1}
  n7 = TextureCoordinate()
  n8 = ScalarParameter() {defaultValue=1, parameterName=ScaleAnimatedCloud}
  n9 = ScalarParameter() {defaultValue=1, parameterName=PanSpeedX}
  n10 = ScalarParameter() {parameterName=PanSpeedY}
  n11 = TextureSampleParameter2D(UVs=Panner(Coordinate=Multiply(A=@n7, B=@n8), Time=Time(), Speed=AppendVector(A=@n9, B=@n10)) {bFractionalPart=true}) {parameterName=Opacity_2, samplerType=SAMPLERTYPE_LinearGrayscale, texture=/Game/Packages/Fortress_Sky/SkyDome/Master/T_SkyDome_Cloud_Voroni}
  n12 = StaticSwitchParameter(True=Multiply(A=@n11.R, B=@n2.R), False=@n2.R) {parameterName=IsAnimated}
  StaticSwitchParameter(True=Multiply(A=Add(A=@n3, B=@n4), B=@n5), False=Multiply(A=@n1.RGB, B=LinearInterpolate(A=@n6, B=@n5, Alpha=@n12))) {parameterName=FakeHDR}
  n13 = ScalarParameter() {defaultValue=1, parameterName=Opacity}
  n14 = ScalarParameter() {parameterName=DepthFade}
  n15 = ComponentMask(None=TextureCoordinate()) {r=true}
  n16 = Time()
  n17 = ScalarParameter() {defaultValue=7, parameterName=Speed_1}
  n18 = Power(Base=Multiply(A=@n15, B=OneMinus(None=@n15)), Exp=Sine(None=Multiply(A=@n16, B=@n17)))
  n19 = Multiply(A=@n16, B=ScalarParameter() {defaultValue=2.91, parameterName=Speed})
  n20 = MaterialFunctionCall(Multiplier=ScalarParameter() {parameterName=OffsetTime}) {materialFunction=&Pi}
  n21 = Multiply(A=Sine(None=Add(A=@n19, B=@n20.Result)), B=ScalarParameter() {defaultValue=2.3082, parameterName=Blinking Amp})
  n22 = Add(
   A=Multiply(A=ScalarParameter() {defaultValue=1, parameterName=Opacity1}, B=@n2.R)
   B=Multiply(A=ScalarParameter() {defaultValue=1, parameterName=Opacity2}, B=@n2.G)
  )
  n23 = Multiply(A=ScalarParameter() {defaultValue=1, parameterName=Opacity3}, B=@n2.B)
  n24 = StaticSwitchParameter(
   True=Multiply(A=Multiply(A=@n2.R, B=@n18), B=Clamp(None=@n21))
   False=StaticSwitchParameter(True=Add(A=@n22, B=@n23), False=@n2.R) {parameterName=FakeHDR}
  ) {parameterName=PoliceLight}
  StaticSwitchParameter(True=Multiply(A=DepthFade(Opacity=@n13, FadeDistance=@n14), B=@n24), False=@n24) {parameterName=Use Depth fade?}
 ]
}
