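# Opaque two-sided material for the front-end environment's monitor frame, enabled for Nanite and instanced static meshes.
# A tiled, rotated noise mask with subtract and multiply contrast controls blends Col_01 toward a second color, itself mixed by vertex red between Col_01_1 and darkened Col_01.
# Specular and roughness are scalar parameters over a flat normal.
# The attributes pass through the destructible damage function, with hit effects on by default, then the Creative move tool function.
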Material /Game/Environments/FrontEnd/Materials/M_Monitor_Frame
graph {
 properties {
  bUseMaterialAttributes=true
  bUsedWithInstancedStaticMeshes=true
  bUsedWithNanite=true
  pixelDepthOffsetMode=PDOM_Legacy
  twoSided=true
  usedShadingModels=Default Lit
 }
 nodes [
  n1 = VectorParameter() {parameterName=Col_01, defaultValue=<rgba 0.0177 0.0286 0.0365 1>}
  n2 = VertexColor()
  n3 = TextureCoordinate() {uTiling=6, vTiling=6}
  n4 = ScalarParameter() {defaultValue=0.4952, parameterName=Rotate}
  n5 = Subtract(
   A=TextureSample(UVs=Rotator(Coordinate=@n3, Time=@n4)) {texture=/Game/Global/Textures/Default/Noise/TankMoteNoise}.R
   B=ScalarParameter() {defaultValue=0.921, parameterName=Subtract}
  )
  n6 = ScalarParameter() {defaultValue=21.7978, parameterName=Multiply}
  n7 = LinearInterpolate(
   A=@n1.RGB
   B=LinearInterpolate(
    A=VectorParameter() {parameterName=Col_01_1, defaultValue=<rgba 0.0208 0.013 0.0087 1>}.RGB
    B=Multiply(A=@n1.RGB) {constB=0.75}
    Alpha=Clamp(None=@n2.R)
   )
   Alpha=Clamp(None=Multiply(A=@n5, B=@n6) {constB=64})
  )
  n8 = ScalarParameter() {defaultValue=0.35, parameterName=Spec}
  n9 = ScalarParameter() {defaultValue=0.5, parameterName=Gloss}
  n10 = TextureSample() {samplerType=SAMPLERTYPE_Normal, texture=&FlatNormalBlue}
  MaterialFunctionCall(MA In=MaterialFunctionCall(
    Material_In=MakeMaterialAttributes!(
     BaseColor=@n7
     Metallic
     Specular=@n8
     Roughness=@n9
     Anisotropy
     EmissiveColor
     Opacity
     OpacityMask
     Normal=@n10.RGB
     Tangent
     WorldPositionOffset
     SubsurfaceColor
     ClearCoat
     ClearCoatRoughness
     AmbientOcclusion
     Refraction
     CustomizedUVs_0
     CustomizedUVs_1
     CustomizedUVs_2
     CustomizedUVs_3
     CustomizedUVs_4
     CustomizedUVs_5
     CustomizedUVs_6
     CustomizedUVs_7
     PixelDepthOffset
     ShadingModel
     Displacement
    )
    UseDestructibleDamage=StaticBoolParameter() {defaultValue=true, group=z_code, parameterName=Code_UseHitEffects}
   ) {materialFunction=&MF_FN_DestructibleDamage}.Material_Out) {materialFunction=&MF_CreativeMoveTool}
 ]
}
