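# Translucent unlit material for a monitor screen, showing a stylized animated background or gadget video with an icon layered on top.
# A two-color radial gradient and faint icon-set texture, distorted by panning line streaks, blend toward video via IsGadget, while Transition adds scrolling bands and UV wobble.
# The desaturated, X-ray-tinted icon gets stepped noise jitter, streak distortion and camera-distance mip blur.
# The composite passes through Fortnite_HitEffects, driven by custom primitive data.
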Material /Game/Environments/FrontEnd/Materials/M_Monitor_Screen_V
graph {
 properties {
  bScreenSpaceReflections=true
  bUseMaterialAttributes=true
  bUsedWithNanite=true
  blendMode=BLEND_Translucent
  numCustomizedUVs=1
  pixelDepthOffsetMode=PDOM_Legacy
  shadingModel=MSM_Unlit
  usedShadingModels=Unlit
 }
 nodes [
  Constant()
  n1 = VectorParameter() {parameterName=Col_02, defaultValue=<rgba 0.0722 0.4483 1.8646 1>}
  n2 = AppendVector(
   A=ScalarParameter() {defaultValue=0.5, parameterName=UScale_RadialGad}
   B=ScalarParameter() {defaultValue=0.8, parameterName=VScale_RadialGrad}
  )
  n3 = MaterialFunctionCall(XY=Multiply(A=TextureCoordinate(), B=ScalarParameter() {defaultValue=0.9284, parameterName=DistortionScale})) {materialFunction=&Swizzle}
  n4 = TextureSample(UVs=Panner(Coordinate=@n3.YX, Speed=ScalarParameter() {defaultValue=0.2571, parameterName=PanSpeed})) {samplerType=SAMPLERTYPE_Masks, texture=/Game/Environments/Sets/Vulcano/Textures/T_LineStreaks_M}
  n5 = TextureCoordinate()
  n6 = ComponentMask(None=@n5) {r=true}
  n7 = MaterialFunctionCall(Float2=@n5) {materialFunction=&BreakOutFloat2Components}
  n8 = Panner(Coordinate=AppendVector(A=@n7.G, B=@n7.G)) {speedX=-0.25}
  n9 = Add(
   A=ComponentMask(None=@n5) {g=true}
   B=Clamp(None=TextureSample(UVs=@n8) {samplerType=SAMPLERTYPE_LinearColor, texture=/Game/Effects/Fort_Effects/BluGlow/BluGlowAnimation_Normals}.R) {maxDefault=0.125}
  )
  n10 = ScalarParameter() {parameterName=Transition}
  n11 = Add(
   A=Multiply(A=Subtract(A=@n4.R) {constB=0.5}, B=ScalarParameter() {defaultValue=-0.0268, parameterName=Distortion2})
   B=LinearInterpolate(A=TextureCoordinate(), B=AppendVector(A=@n6, B=@n9), Alpha=@n10)
  )
  n12 = LinearInterpolate(
   A=VectorParameter() {parameterName=Col_01, defaultValue=<rgba 0.0155 0.0536 0.1094 1>}.RGB
   B=@n1.RGB
   Alpha=MaterialFunctionCall(
    UVs=MaterialFunctionCall(UVs=@n11, Texture Scale=@n2) {materialFunction=&ScaleUVsByCenter}.UVs
    Radius=ScalarParameter() {defaultValue=1.2952, parameterName=Radius, sliderMax=2}
    Density=ScalarParameter() {defaultValue=0.8571, parameterName=Density, sliderMax=10}
   ) {materialFunction=&RadialGradientExponential}.RadialGradientExponential
  )
  n13 = Multiply(A=@n1.RGB) {constB=1.5}
  n14 = AppendVector(A=ScalarParameter() {defaultValue=0.7452, parameterName=UScale}, B=ScalarParameter() {defaultValue=1.3803, parameterName=VScale})
  n15 = Multiply(
   A=TextureSampleParameter2D(UVs=MaterialFunctionCall(UVs=@n11, Texture Scale=@n14) {materialFunction=&ScaleUVsByCenter}.UVs) {parameterName=Upgrade-IconSet, texture=/Game/Athena/TestAssets/Textures/Monitor_Bg}.R
   B=ScalarParameter() {defaultValue=0.0667, parameterName=IconLockOpacity}
  )
  n16 = Panner(Coordinate=@n11) {speedY=-0.1}
  n17 = Panner(Coordinate=@n11) {speedX=0.05}
  n18 = LinearInterpolate(
   A=LinearInterpolate(
    A=LinearInterpolate(A=@n12, B=@n13, Alpha=@n15)
    B=TextureSampleParameter2D() {parameterName=GadgetVid, samplerSource=SSM_Wrap_WorldGroupSettings, texture=/Game/UI/Foundation/Movie/DefaultMediaTexture}.RGB
    Alpha=ScalarParameter() {parameterName=IsGadget, sliderMax=1}
   )
   B=Add(
    A=TextureSample(UVs=@n16) {texture=/Game/Environments/Homebase/Textures/T_RoadFade_M}.G
    B=TextureSample(UVs=@n17) {texture=&T_RoadFade_M}.R
   )
   Alpha=LinearInterpolate(Alpha=@n10) {constB=0.1}
  )
  n19 = ScalarParameter() {defaultValue=1, parameterName=Emissive, sliderMax=1}
  n20 = VectorParameter() {parameterName=XRayColor1, defaultValue=<rgba 0.0576 0.1051 0.3906 1>}
  n21 = VectorParameter() {parameterName=XRayColor2, defaultValue=<rgba 1.0938 1.4813 3 1>}
  n22 = Multiply(A=TextureCoordinate(), B=Constant2Vector() {g=0.5, r=1}) {constA=2}
  n23 = ScalarParameter() {defaultValue=2.2852, group=ControlIcon, parameterName=IconRotate}
  n24 = MaterialFunctionCall(
   UVs=Multiply(A=Rotator(Coordinate=@n22, Time=@n23), B=Constant2Vector() {g=2, r=1}) {constA=2}
   Texture Scale=ScalarParameter() {defaultValue=1.1043, group=ControlIcon, parameterName=IconScale}
  ) {materialFunction=&ScaleUVsByCenter}
  n25 = TextureCoordinate()
  n26 = Time()
  n27 = ScalarParameter() {defaultValue=1.3617, parameterName=RandomOffsetSpeed}
  n28 = Multiply(A=Ceil(None=Multiply(A=@n26, B=@n27)), B=ScalarParameter() {defaultValue=0.0381, parameterName=OffsetAmount})
  n29 = Multiply(
   A=ScalarParameter() {defaultValue=-0.0476, group=ControlIcon, parameterName=IconShift}
   B=TextureSample(UVs=Add(A=@n25, B=@n28)) {samplerType=SAMPLERTYPE_LinearColor, texture=/Game/Global/Textures/Default/Noise/T_Atlas_DigitalNoise_01}.R
  )
  n30 = ScalarParameter() {defaultValue=-0.019, group=ControlIcon, parameterName=IconMasterBlur}
  n31 = LinearInterpolate(B=ScalarParameter() {defaultValue=0.0571, parameterName=Distortion}, Alpha=@n30)
  n32 = Subtract(A=Add(A=TextureSample(UVs=@n3.YX) {samplerType=SAMPLERTYPE_Masks, texture=&T_LineStreaks_M}.R, B=@n4.R)) {constB=0.5}
  n33 = ScalarParameter() {defaultValue=5, parameterName=Blur}
  n34 = ScalarParameter() {defaultValue=7.5, parameterName=BlurFar}
  n35 = MaterialFunctionCall(Distance=ScalarParameter() {defaultValue=2000, parameterName=FadeDistanceBlur}) {materialFunction=&Near_Camera_Fade}
  n36 = TextureSampleParameter2D!(
   UVs=Add(A=Add(A=@n24.UVs, B=@n29), B=Multiply(A=@n31, B=@n32))
   Level=LinearInterpolate(B=LinearInterpolate(A=@n33, B=@n34, Alpha=@n35.Result), Alpha=@n30)
   Apply View MipBias
  ) {mipValueMode=TMVM_MipLevel, parameterName=Item, texture=/Game/Athena/TestAssets/Textures/Monitor_Icon_1}
  n37 = Power(Base=Desaturation(None=@n36.RGB, Fraction=Constant() {r=1}), Exp=ScalarParameter() {defaultValue=1.0557, parameterName=Contrast})
  n38 = ScalarParameter() {defaultValue=0.4286, group=ControlIcon, parameterName=IconOpacity}
  n39 = MaterialFunctionCall!(
   WSDeformationCenter=VectorParameter() {bUseCustomPrimitiveData=true, group=z_code, parameterName=WSDeformationCenter, primitiveDataIndex=20}.RGB
   Code_DestructionAmount=ScalarParameter() {bUseCustomPrimitiveData=true, group=z_code, parameterName=Code_DestructionAmount, primitiveDataIndex=13}
   Code_UseUvForDamage
   Code_UseHitEffects=StaticBoolParameter() {defaultValue=true, group=z_code, parameterName=Code_UseHitEffects}
   Code_Emissive
   WSDamageRadius=ScalarParameter() {bUseCustomPrimitiveData=true, defaultValue=1, group=z_code, parameterName=WSDamageRadius, primitiveDataIndex=26}
   CrackDamageShader
   Diffuse=LinearInterpolate(A=Multiply(A=@n18, B=@n19), B=LinearInterpolate(A=@n20.RGB, B=@n21.RGB, Alpha=@n37), Alpha=Multiply(A=@n36.A, B=@n38))
   ScalarModDamageCracks
   ScalarOffsetforDamageCracks
   ExposeLerpFunctionForInvertingTheHitDistanceCalc
   InvertDamageFalloff
   CrackWorldPosition
   WorldPosition
   DistanceCullFadeAnimationOffset=ScalarParameter() {bUseCustomPrimitiveData=true, group=z_code, parameterName=DistanceCullFadeAnimationOffset, primitiveDataIndex=12}
  ) {materialFunction=/Game/Packages/jl_tests/Fortnite_HitEffects}
  n40 = Constant()
  n41 = Time()
  n42 = LinearInterpolate(B=Clamp(None=Sine(None=@n41) {period=0.125}) {maxDefault=0.25, minDefault=-0.25}, Alpha=@n10)
  SetMaterialAttributes!(
   MaterialAttributes
   Base Color=Constant3Vector() {constant=<rgba 0.0084 0.0091 0.0104 0>}
   Specular=ScalarParameter() {defaultValue=0.25, parameterName=Spec}
   Roughness=ScalarParameter() {defaultValue=0.35, parameterName=Gloss}
   Emissive Color=@n39.Diffuse
   Normal=TextureSample() {samplerType=SAMPLERTYPE_Normal, texture=&FlatNormalBlue}.RGB
   Customized UV 0=Add(A=TextureCoordinate(), B=AppendVector(A=@n40, B=@n42))
   World Position Offset=@n39.Diffuse
  ) {attributeSetTypes [
    #1
    #2
    #3
    #4
    #5
    #6
    #7
   ]}
  VectorNoise()
 ]
}
