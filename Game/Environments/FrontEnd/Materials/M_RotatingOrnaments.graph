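# Default Lit material for snowflake ornaments in the front end that makes each one swing back and forth around the vertical axis at its actor position.
# A pseudo-random value from the actor's X position offsets time and blends two speed parameters, so ornaments move out of sync.
# The resulting sine angle drives RotateAboutAxis and corrects vertex normals, which are interpolated and blended with the normal map.
# Diffuse and specular textures are parameters.
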Material /Game/Environments/FrontEnd/Materials/M_RotatingOrnaments
graph {
 properties {bUsedWithNanite=true, pixelDepthOffsetMode=PDOM_Legacy, usedShadingModels=Default Lit}
 nodes [
  n1 = Constant3Vector() {constant=<rgba 0 0 1 0>}
  n2 = ActorPositionWS()
  n3 = MaterialFunctionCall(Float3=@n2) {materialFunction=&BreakOutFloat3Components}
  n4 = Frac(None=Divide(A=Add(A=@n3.R, B=@n3.R)) {constB=213.34})
  n5 = Add(A=Time(), B=Multiply(A=@n4, B=ScalarParameter() {defaultValue=109.2879, parameterName=Random}))
  n6 = LinearInterpolate(
   A=ScalarParameter() {defaultValue=0.0952, parameterName=RotSpeed}
   B=ScalarParameter() {defaultValue=0.03, parameterName=RotSpeed2}
   Alpha=@n4
  )
  n7 = Multiply(A=Sine(None=Multiply(A=@n5, B=@n6)), B=ScalarParameter() {defaultValue=0.1, parameterName=RotAngle})
  MaterialFunctionCall(
   BaseNormal=VertexInterpolator(VS=MaterialFunctionCall(Rotation Axis=@n1, Rotation Angle=@n7) {materialFunction=&FixRotateAboutAxisNormals}.New Tangent Space Vertex Normal).PS
   AdditionalNormal=TextureSampleParameter2D() {parameterName=Normal, samplerType=SAMPLERTYPE_Normal, texture=/Game/Environments/Props/Winter/Textures/T_SnowflakeOrnaments_N}.RGB
  ) {materialFunction=&BlendAngleCorrectedNormals}
  RotateAboutAxis(NormalizedRotationAxis=@n1, RotationAngle=@n7, PivotPoint=@n2, Position=WorldPosition().XYZ)
  TextureSampleParameter2D() {parameterName=Diffuse, samplerType=SAMPLERTYPE_LinearColor, texture=/Game/Environments/Props/Winter/Textures/T_SnowflakeOrnaments_D}
  TextureSampleParameter2D() {parameterName=Specular, samplerType=SAMPLERTYPE_Masks, texture=/Game/Environments/Props/Winter/Textures/T_SnowflakeOrnaments_S}
 ]
}
